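; Camera lens Niagara actor Blueprint for the bus warp warm-up effect in the lobby island space environment.
; On BeginPlay it computes the viewport's width-to-height ratio, clamps it to at least 1, and passes it to the Niagara component's ScreenSizScalar parameter.
; The ratio is sampled only once, and a timer deactivates the Niagara component four seconds later.
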(event EventBeginPlay
  (bind (_x _y) (Math|Vector2D|BreakVector2D (Viewport|GetViewportSize)))
  (|CallParentFunction)
  (Effects|Components|ParticleSystem|SetFloatParameter (Variables|NiagaraActor|GetNiagaraComponent) "ScreenSizScalar" (Math|Float|Max(Float) (/ _x _y) 1.0))
  (Utilities|Time|SetTimerbyEvent (delegate Custom|DeactivateSelf) 4.0))

(event Custom|DeactivateSelf
  (Components|Activation|Deactivate (Variables|NiagaraActor|GetNiagaraComponent)))

(fn ConstructionScript
  (|CallParentFunction))
