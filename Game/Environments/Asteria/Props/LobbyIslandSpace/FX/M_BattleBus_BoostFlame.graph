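# Additive, unlit, two-sided material for the Battle Bus boost flame in the lobby island environment, built for Niagara mesh particles.
# Panning blurred noise perturbs the V coordinate used to sample a color curve atlas, which is tinted by particle color.
# Depth-faded opacity blends a sine-shaped, view-angle-faded mask with an axis-aligned fresnel.
# Separate branches compute a vertex-color-masked normal offset and a parameter-scaled party bus animation WPO.
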Material /Game/Environments/Asteria/Props/LobbyIslandSpace/FX/M_BattleBus_BoostFlame
graph {
 properties {
  bUsedWithNiagaraMeshParticles=true
  blendMode=BLEND_Additive
  pixelDepthOffsetMode=PDOM_Legacy
  shadingModel=MSM_Unlit
  twoSided=true
  usedShadingModels=Unlit
 }
 nodes [
  Constant3Vector()
  n1 = TextureCoordinate()
  n2 = TextureCoordinate()
  n3 = VectorParameter() {parameterName=UV_Tile, defaultValue=<rgba 5 0.5 0 0>}
  n4 = AppendVector(A=@n3.R, B=@n3.R)
  n5 = TextureSample(UVs=Panner(Coordinate=Multiply(A=@n2, B=@n4)) {speedX=0.2, speedY=2}) {samplerType=SAMPLERTYPE_Grayscale, texture=&LowResBlurredNoise}
  n6 = CurveAtlasRowParameter(CurveTime=Add(
    A=MaterialFunctionCall(Float2=@n1) {materialFunction=&BreakOutFloat2Components}.G
    B=Divide(A=@n5.G) {constB=8}
   )) {
   atlas=/Game/Athena/Items/AwfulDuck/FX/CA_AwfulDuck
   curve=&/Curve_BoostFlame
   defaultValue=12
   parameterName=ColorCurve
  }
  n7 = ParticleColor()
  n8 = Multiply(A=MaterialFunctionCall(Float2=TextureCoordinate()) {materialFunction=&BreakOutFloat2Components}.R) {constB=0.5}
  n9 = Multiply(
   A=Power(Base=Sine(None=@n8), Exp=ScalarParameter() {defaultValue=1, parameterName=Param_2})
   B=MaterialFunctionCall("FadeSensitivity (optional)"=ScalarParameter() {defaultValue=2, parameterName=Param_4}) {materialFunction=&FadeBasedOnViewAngle}.Result
  )
  n10 = MaterialFunctionCall!(Axis=Constant3Vector() {constant=<rgba 1 0 0 0>}, Exponent=ScalarParameter() {defaultValue=2, parameterName=Param_1}, Normal) {materialFunction=/Game/Effects/Fort_Effects/Materials/_Functions/MF_AxisAlignedFrensnel}
  n11 = ScalarParameter() {defaultValue=1, parameterName=s}
  DepthFade(
   Opacity=Multiply(A=Multiply(A=@n6.A, B=@n7.A), B=LinearInterpolate(A=@n9, B=@n10.Fresnel, Alpha=@n11) {constA=1})
   FadeDistance=ScalarParameter() {defaultValue=100, parameterName=Param_3}
  )
  MaterialFunctionCall(Emissive=Multiply(A=@n6, B=@n7.RGB)) {materialFunction=&MF_FX_EmissiveAdjustments}
  n12 = VertexColor()
  Multiply(A=Multiply(A=VertexNormalWS(), B=OneMinus(None=@n12.R)), B=ScalarParameter() {defaultValue=-150})
  Multiply(
   A=MaterialFunctionCall() {materialFunction=/Game/Athena/Aircraft/PartyBus_RT/Materials/MF_PartyBusAnimationWPO}.Result
   B=ScalarParameter() {parameterName=UsePartyBusAnimWPO}
  )
  VectorParameter() {defaultValue=<rgba 0.6389 0.2181 1 0>}
 ]
}
