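# Translucent, unlit, two-sided Niagara mesh-particle material for Battle Bus booster rim-burst flames on the Asteria lobby island.
# A transformed axis vector scaled by U and BurstAmount forms an offset.
# A FalloffPos sphere mask along U shapes a vertex-red-masked gradient and sets smoothstep thresholds eroding two summed smoke-noise samples, multiplied by vertex red and particle alpha.
# Dynamic parameter T drives both panners, whose speeds use only SpeedX for both axes.
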Material /Game/Environments/Asteria/Props/LobbyIslandSpace/FX/M_FX_BattleBusBoostOnRimFlames
graph {
 properties {
  bUsedWithNiagaraMeshParticles=true
  blendMode=BLEND_Translucent
  pixelDepthOffsetMode=PDOM_Legacy
  previewMesh=&/SM_FX_Battlebus_Booster_RimBurst
  shadingModel=MSM_Unlit
  twoSided=true
  usedShadingModels=Unlit
 }
 nodes [
  n1 = Constant3Vector() {constant=<rgba 1 0 0 0>}
  n2 = TextureCoordinate()
  Multiply(
   A=Multiply(A=Transform(None=@n1), B=ComponentMask(None=@n2) {r=true})
   B=DynamicParameter() {defaultValue=<rgba 50 0 1 1>, paramNames [BurstAmount,FalloffPos,T,Param4]}.BurstAmount
  ) {constB=100}
  n3 = ComponentMask(None=TextureCoordinate()) {r=true}
  n4 = Power(Base=SphereMask(A=@n3, B=DynamicParameter() {defaultValue=<rgba 50 0 1 1>, paramNames [BurstAmount,FalloffPos,T,Param4]}.FalloffPos) {attenuationRadius=1, hardnessPercent=0.5})
  Multiply(A=Multiply(A=OneMinus(None=@n3), B=@n4), B=OneMinus(None=VertexColor().R))
  n5 = VectorParameter() {group=1, parameterName=Tex2_Xfm, channelNames {a=SpeedY, b=SpeedX, g=TileY, r=TileX}, defaultValue=<rgba 0 0 0 1>}
  n6 = Multiply(A=TextureCoordinate(), B=ComponentMask(None=@n5) {g=true, r=true})
  n7 = DynamicParameter() {defaultValue=<rgba 50 0 1 1>, paramNames [BurstAmount,FalloffPos,T,Param4]}
  n8 = AppendVector(A=@n5.SpeedX, B=@n5.SpeedX)
  n9 = TextureObjectParameter() {parameterName=Noise, texture=/Game/Global/Textures/Default/Tileable_Patterns/T_Smoke_OUT}
  n10 = VectorParameter() {group=1, parameterName=Tex1_Xfm, channelNames {a=SpeedY, b=SpeedX, g=TileY, r=TileX}, defaultValue=<rgba 0 0 0 1>}
  n11 = Multiply(A=TextureCoordinate(), B=ComponentMask(None=@n10) {g=true, r=true})
  n12 = AppendVector(A=@n10.SpeedX, B=@n10.SpeedX)
  n13 = Add(
   A=TextureSample(UVs=Panner(Coordinate=@n6, Time=@n7.T, Speed=@n8) {bFractionalPart=true, speedY=1}, Tex=@n9) {texture=&T_Smoke_OUT}.R
   B=TextureSample(UVs=Panner(Coordinate=@n11, Time=@n7.T, Speed=@n12) {bFractionalPart=true, speedY=1}, Tex=@n9) {texture=&T_Smoke_OUT}.R
  )
  n14 = Add(A=@n4, B=ScalarParameter() {parameterName=p})
  n15 = SmoothStep(Min=@n14, Max=Add(A=@n14, B=ScalarParameter() {parameterName=w}), Value=OneMinus(None=Saturate(None=@n13)))
  Multiply(A=ParticleColor().A, B=Multiply(A=Multiply(A=@n13, B=@n15), B=VertexColor().R))
 ]
}
