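# Additive unlit Niagara mesh-particle material for the Battle Bus boost effect on the bus interior in the lobby island space scene.
# A heat mask's green channel drives a smoothstep reveal of its red channel, tinted by FireHeatColor.
# A radial light mask blends into panning plasma along UV1, colored by hue-shifted particle color.
# Dynamic parameters control heat, radius, hue and edge fade; vertex red and particle alpha shape the fade.
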Material /Game/Environments/Asteria/Props/LobbyIslandSpace/FX/M_FX_BattleBusBoost_Interior
graph {
 properties {
  bUseTranslucencyVertexFog=false
  bUsedWithNiagaraMeshParticles=true
  blendMode=BLEND_Additive
  pixelDepthOffsetMode=PDOM_Legacy
  shadingModel=MSM_Unlit
  translucencyPass=MTP_BeforeDOF
  usedShadingModels=Unlit
 }
 nodes [
  n1 = TextureSample() {texture=&/T_FX_Battlebus_HeatMask}
  Add(A=Multiply(B=@n1.R) {constA=32}, B=Multiply())
  n2 = DynamicParameter() {parameterIndex=1, paramNames [LightRadius/FireHeat, HueShiftAmnt/HeatmapReveal, EnergyPulsing, EdgeFade]}
  n3 = Add(A=@n2.HueShiftAmnt/HeatmapReveal) {constB=2}
  n4 = Multiply(A=@n1.R, B=Saturate(None=SmoothStep(Min=@n2.HueShiftAmnt/HeatmapReveal, Max=@n3, Value=@n1.G)))
  n5 = TextureCoordinate()
  n6 = Constant() {r=0.5}
  n7 = DynamicParameter() {paramNames [LightRadius/FireHeat, HueShiftAmnt, EnergyPulsing, EdgeFade]}
  n8 = Panner(Coordinate=TextureCoordinate() {coordinateIndex=1, uTiling=0.25, vTiling=3}) {speedX=-1.25}
  n9 = LinearInterpolate(
   A=Add(A=Constant(), B=SphereMask(A=@n5, B=@n6, Radius=@n7.LightRadius/FireHeat) {attenuationRadius=0.5, hardnessPercent=0})
   B=Multiply(
    A=TextureSample(UVs=@n8) {samplerType=SAMPLERTYPE_LinearGrayscale, texture=/Game/Effects/Fort_Effects/Emotes/UltraLight/Texture/T_Plasma_C_Greyscale}.R
    B=ScalarParameter() {defaultValue=4, parameterName=PlasmaStr}
   ) {constB=4}
   Alpha=ComponentMask(None=TextureCoordinate() {coordinateIndex=1}) {r=true}
  )
  n10 = TextureCoordinate() {coordinateIndex=1}
  n11 = DynamicParameter() {paramNames [LightRadius/FireHeat, HueShiftAmnt, EnergyPulsing, EdgeFade]}
  n12 = ParticleColor()
  n13 = MaterialFunctionCall(Hue Shift Percentage=Multiply(A=ComponentMask(None=@n10) {r=true}, B=@n11.HueShiftAmnt), Texture=@n12.RGB) {materialFunction=&HueShift}
  Add(
   A=Multiply(A=Multiply(A=@n4, B=@n2.LightRadius/FireHeat), B=VectorParameter() {parameterName=FireHeatColor, defaultValue=<rgba 0 0 0 1>}.RGB)
   B=Add(A=Multiply(A=@n9, B=@n13.Result), B=@n13.Result)
  )
  Multiply(A=Power(Base=VertexColor().R, Exp=@n11.EdgeFade), B=@n12.A)
 ]
}
