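# Translucent unlit material for Battle Bus flames on the Asteria lobby island, enabled for Niagara mesh particles.
# Two panning noise layers blend by FlameMaturity, are shaped by a dynamic-parameter sphere mask and vertex red, then smoothstepped into opacity.
# Particle color is hue-shifted by ColorGrad, while a panned flame pattern forms a source edge mask.
# Perlin noise drives normal-direction wobble, and a parameter blends in party bus animation offset.
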Material /Game/Environments/Asteria/Props/LobbyIslandSpace/FX/M_FX_BattleBusFlames_Sprite
graph {
 properties {
  bUsedWithNiagaraMeshParticles=true
  blendMode=BLEND_Translucent
  pixelDepthOffsetMode=PDOM_Legacy
  shadingModel=MSM_Unlit
  usedShadingModels=Unlit
 }
 nodes [
  Constant2Vector() {g=1, r=0.5}
  LinearInterpolate(
   A=NamedRerouteUsage()
   B=MaterialFunctionCall() {materialFunction=/Game/Athena/Aircraft/PartyBus_RT/Materials/MF_PartyBusAnimationWPO}.Result
   Alpha=ScalarParameter() {parameterName=UsePartyBusAnimWPO}
  )
  n1 = Add(B=ScalarParameter() {parameterName=SSTEP_P})
  n2 = Multiply(
   A=DynamicParameter() {parameterIndex=1, paramNames [FlameReveal/Softness, Wobble, FlameMaturity, ColorGrad]}.FlameReveal/Softness
   B=ScalarParameter() {parameterName=SSTEP_W}
  )
  n3 = ComponentMask(None=NamedRerouteUsage()) {r=true}
  n4 = SphereMask(
   A=Abs(None=Subtract(A=@n3) {constB=0.5})
   B=Constant() {r=0.5}
   Radius=DynamicParameter() {paramNames [FlameReveal/Softness, Wobble, FlameMaturity, ColorGrad]}.FlameReveal/Softness
  ) {hardnessPercent=0}
  n5 = Multiply(A=ScalarParameter() {parameterName=Flames_DistortionStr}, B=Multiply(A=@n4, B=VertexColor().R))
  n6 = VectorParameter() {parameterName=FlamesEarly_Tiling, defaultValue=<rgba 5 0.5 0 0>}
  n7 = Multiply(A=TextureCoordinate(), B=AppendVector(A=@n6.R, B=@n6.R))
  n8 = Add(A=Time(), B=ParticleRandom())
  n9 = VectorParameter() {parameterName=UV_Tile, defaultValue=<rgba 5 0.5 0 0>}
  n10 = Multiply(A=TextureCoordinate(), B=AppendVector(A=@n9.R, B=@n9.R))
  n11 = Add(A=Time(), B=ParticleRandom())
  n12 = LinearInterpolate(
   A=TextureSample(UVs=Panner(Coordinate=@n7, Time=@n8) {speedX=0.2, speedY=2}) {samplerType=SAMPLERTYPE_Grayscale, texture=&LowResBlurredNoise}.G
   B=TextureSample(UVs=Panner(Coordinate=@n10, Time=@n11) {speedX=0.2, speedY=2}) {samplerType=SAMPLERTYPE_Grayscale, texture=&LowResBlurredNoise}.G
   Alpha=DynamicParameter() {paramNames [FlameReveal/Softness, Wobble, FlameMaturity, ColorGrad]}.FlameMaturity
  )
  n13 = Add(A=ComponentMask(None=NamedRerouteUsage()) {g=true}, B=Add(A=ScalarParameter() {parameterName=FlamesEdge}, B=Multiply(A=@n5, B=@n12)))
  n14 = NamedRerouteUsage()
  n15 = Multiply(A=SmoothStep(Min=@n1, Max=Add(A=@n1, B=@n2), Value=Multiply(A=@n13, B=@n14)), B=NamedRerouteUsage())
  LinearInterpolate(A=@n15, B=Multiply(A=ParticleColor().A, B=@n15), Alpha=ScalarParameter() {parameterName=UseParticleColorToHandleOpacity})
  n16 = NamedRerouteUsage()
  MaterialFunctionCall(
   Hue Shift Percentage=Multiply(A=ComponentMask(None=@n16) {g=true}, B=DynamicParameter() {paramNames [FlameReveal/Softness, Wobble, FlameMaturity, ColorGrad]}.ColorGrad)
   Texture=ParticleColor().RGB
  ) {materialFunction=&HueShift}
  MaterialFunctionCall() {materialFunction=&Swizzle}
  NamedRerouteDeclaration(None=Fmod(A=TextureCoordinate(), B=Constant() {r=1})) {name=LoopingUV, nodeColor=<rgba 1 0.8203 0 1>}
  NamedRerouteDeclaration(None=@n4) {name=DynamicReveal, nodeColor=<rgba 0 0.4609 1 1>}
  n17 = Panner() {speedY=1}
  n18 = Add(
   A=Multiply(A=TextureSample(UVs=@n17) {samplerType=SAMPLERTYPE_Masks, texture=/Game/Global/Textures/Default/Masks_and_Gradients/FlamePattern}.B) {constB=0.1}
   B=ScalarParameter() {defaultValue=0.5, group=Sourcemask, parameterName=Source_Mask_P}
  ) {constB=0.55}
  n19 = Add(A=@n18, B=ScalarParameter() {defaultValue=0.5, parameterName=Source_Mask_W}) {constB=0.25}
  n20 = ComponentMask(None=NamedRerouteUsage()) {g=true}
  NamedRerouteDeclaration(None=OneMinus(None=SmoothStep(Min=@n18, Max=@n19, Value=@n20))) {name=SourceEdgeMask, nodeColor=<rgba 0.6406 0 1 1>}
  n21 = ComponentMask(None=NamedRerouteUsage()) {g=true}
  n22 = VectorParameter() {parameterName=WobbleTiling, channelNames {a="", b=Strength, g=Y Tile, r=X Tile}, defaultValue=<rgba 0 0 0 1>}
  n23 = Multiply(A=TextureCoordinate(), B=ComponentMask(None=@n22) {g=true, r=true})
  n24 = Multiply(A=TextureSample(UVs=Panner(Coordinate=@n23) {speedX=1}) {samplerType=SAMPLERTYPE_Masks, texture=/Game/Global/Textures/Default/Noise/T_Generic_Perlin}.R) {constB=2}
  n25 = Multiply(A=Multiply(A=Subtract(A=@n24), B=VertexNormalWS()), B=@n22.Strength)
  n26 = DynamicParameter() {paramNames [FlameReveal/Softness, Wobble, FlameMaturity, ColorGrad]}
  NamedRerouteDeclaration(None=Multiply(A=OneMinus(None=@n21), B=Multiply(A=@n25, B=@n26.Wobble))) {name=FireWobble, nodeColor=<rgba 0.6406 0 1 1>}
  TextureCoordinate()
  TextureCoordinate()
 ]
}
