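# Translucent unlit smoke-sheet material for Niagara mesh particles, named for the Battle Bus warmup ground smoke in the lobby island.
# Two panning smoke textures on particle-offset time are summed and eroded by a UV-gradient smoothstep, then scaled by vertex red and particle alpha.
# A dynamic-parameter sphere mask around the particle position fakes lighting on the tinted color.
# SplineThicken supplies optional world position offset; depth fade applies only at High and Epic quality.
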Material /Game/Environments/Asteria/Props/LobbyIslandSpace/FX/M_FX_BattlebusWarmup_GroundSmokeSheet
graph {
 properties {
  bUsedWithNiagaraMeshParticles=true
  blendMode=BLEND_Translucent
  pixelDepthOffsetMode=PDOM_Legacy
  shadingModel=MSM_Unlit
  usedShadingModels=Unlit
 }
 nodes [
  n1 = DynamicParameter() {paramNames [LightRadius, LightingBrightness, DepthFadeDistance, Param4]}
  n2 = Power(Base=SphereMask(A=WorldPosition().XYZ, B=ParticlePositionWS(), Radius=@n1.LightRadius) {hardnessPercent=0})
  Add(
   A=Multiply(A=ParticleColor().RGB, B=Multiply(A=@n2, B=@n1.LightingBrightness))
   B=VectorParameter() {parameterName=DiffuseTint, defaultValue=<rgba 1 1 1 1>}.RGB
  )
  Constant() {r=0.5}
  Multiply(
   A=ScalarParameter() {parameterName=UseSplineThicken}
   B=MaterialFunctionCall(WidthBase=ScalarParameter() {parameterName=Thickness}, WidthTip=ScalarParameter() {parameterName=Endthickness}) {materialFunction=&SplineThicken}.WorldPositionOffset
  )
  NamedRerouteDeclaration(None=Add(A=Time(), B=ParticleRandom())) {name=PrtTime, nodeColor=<rgba 0 1 0.0156 1>}
  n3 = VectorParameter() {group=1, parameterName=Tex2_Xfm, channelNames {a=SpeedY, b=SpeedX, g=TileY, r=TileX}, defaultValue=<rgba 0 0 0 1>}
  n4 = Multiply(A=TextureCoordinate(), B=ComponentMask(None=@n3) {g=true, r=true})
  n5 = NamedRerouteUsage()
  n6 = AppendVector(A=@n3.SpeedX, B=@n3.SpeedX)
  n7 = VectorParameter() {group=1, parameterName=Tex1_Xfm, channelNames {a=SpeedY, b=SpeedX, g=TileY, r=TileX}, defaultValue=<rgba 0 0 0 1>}
  n8 = Multiply(A=TextureCoordinate(), B=ComponentMask(None=@n7) {g=true, r=true})
  n9 = AppendVector(A=@n7.SpeedX, B=@n7.SpeedX)
  n10 = Add(
   A=TextureSample(UVs=Panner(Coordinate=@n4, Time=@n5, Speed=@n6) {bFractionalPart=true, speedY=1}) {texture=/Game/Global/Textures/Default/Tileable_Patterns/T_Smoke_OUT}.R
   B=TextureSample(UVs=Panner(Coordinate=@n8, Time=@n5, Speed=@n9) {bFractionalPart=true, speedY=1}) {texture=&T_Smoke_OUT}.R
  )
  n11 = Add(A=ComponentMask(None=TextureCoordinate()) {g=true}, B=ScalarParameter() {parameterName=p})
  n12 = SmoothStep(Min=@n11, Max=Add(A=@n11, B=ScalarParameter() {parameterName=w}), Value=OneMinus(None=Saturate(None=@n10)))
  n13 = Multiply(A=ParticleColor().A, B=Multiply(A=Multiply(A=@n10, B=@n12), B=VertexColor().R))
  n14 = DepthFade(Opacity=@n13, FadeDistance=ScalarParameter() {parameterName=DepthDistance})
  QualitySwitch(Default=@n13, Low=@n13, High=@n14, Medium=@n13, Epic=@n14)
 ]
}
