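# Light function material for the lobby island prop FX, named as a light wash-over, that builds scrolling, broken-up light masks.
# Two separate chains multiply a panned Mash_M mask by squared panned noise to drive a lerp with a low floor.
# One chain tiles vertically and samples with explicit UV derivatives; the other scrolls faster sideways and sums two offset, square-rooted samples.
# Neither chain is shown wired to a material output.
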Material /Game/Environments/Asteria/Props/LobbyIslandSpace/FX/M_LF_LightWashOver
graph {
 properties {materialDomain=MD_LightFunction, pixelDepthOffsetMode=PDOM_Legacy, usedShadingModels=Default Lit}
 nodes [
  n1 = TextureCoordinate() {vTiling=3}
  n2 = Panner(Coordinate=@n1) {speedY=-0.23}
  n3 = DDX(Value=@n1)
  n4 = DDY(Value=@n1)
  n5 = TextureSample!(UVs=Panner(Coordinate=@n1) {speedX=0.05, speedY=0.1}, Tex, "DDX(UVs)"=DDX(Value=@n1), "DDY(UVs)"=DDY(Value=@n1), Apply View MipBias) {mipValueMode=TMVM_Derivative, samplerType=SAMPLERTYPE_LinearColor, texture=&T_NoisePack_LinearColor}
  LinearInterpolate(Alpha=Multiply(
    A=TextureSample(UVs=@n2, "DDX(UVs)"=@n3, "DDY(UVs)"=@n4) {mipValueMode=TMVM_Derivative, texture=/Game/Weapons/WeaponSkins/Wraps/Textures/Mash/Mash_M}.R
    B=Multiply(A=@n5.G, B=@n5.G)
   )) {constA=0.1}
  n6 = TextureCoordinate()
  n7 = SquareRoot(None=TextureSample!(UVs=Panner(Coordinate=@n6) {speedX=-1, speedY=0.02}, Tex, Apply View MipBias) {texture=&Mash_M}.R)
  n8 = Add(A=@n6) {constB=0.5}
  n9 = SquareRoot(None=TextureSample(UVs=Panner(Coordinate=@n8) {speedX=-1, speedY=0.02}) {texture=&Mash_M}.R)
  n10 = TextureSample(UVs=Panner(Coordinate=@n6) {speedX=-1.3, speedY=-0.1}) {samplerType=SAMPLERTYPE_LinearColor, texture=&T_NoisePack_LinearColor}
  LinearInterpolate(Alpha=Multiply(A=Add(A=@n7, B=@n9), B=Multiply(A=@n10.G, B=@n10.G))) {constA=0.03}
  MaterialFunctionCall(Float2=TextureCoordinate()) {materialFunction=&BreakOutFloat2Components}
 ]
}
