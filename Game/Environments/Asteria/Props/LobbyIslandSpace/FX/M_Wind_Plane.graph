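# Translucent, unlit, two-sided wind plane material for Niagara mesh particles in the Asteria lobby island props.
# Two oppositely drifting samples of a storm cloud texture multiply into streaky noise, with vertex red offsetting the pan time.
# Particle color tints the noise, while opacity combines it with particle alpha, UV gradient edge masks and depth fade.
# It also includes a world position offset function borrowed from the party bus animation.
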Material /Game/Environments/Asteria/Props/LobbyIslandSpace/FX/M_Wind_Plane
graph {
 properties {
  bUsedWithNiagaraMeshParticles=true
  blendMode=BLEND_Translucent
  pixelDepthOffsetMode=PDOM_Legacy
  shadingModel=MSM_Unlit
  twoSided=true
  usedShadingModels=Unlit
 }
 nodes [
  n1 = ParticleColor()
  n2 = MaterialFunctionCall(XYZ=TextureCoordinate()) {materialFunction=&Swizzle}
  n3 = VectorParameter() {parameterName=UV, defaultValue=<rgba 1 0.2 0 0>}
  n4 = AppendVector(A=@n3.R, B=@n3.R)
  n5 = Multiply(A=@n2.YX, B=@n4)
  n6 = Add(A=VertexColor().R, B=Time())
  n7 = Multiply(A=@n2.YX, B=Divide(A=@n4))
  n8 = Multiply(
   A=TextureSample(UVs=Panner(Coordinate=@n5, Time=@n6) {speedX=0.5, speedY=-1}) {samplerType=SAMPLERTYPE_LinearGrayscale, texture=/Game/Global/Textures/Default/Tileable_Patterns/T_StormCloudDetail01_M}.R
   B=TextureSample(UVs=Panner(Coordinate=@n7, Time=@n6) {speedX=-0.5, speedY=-0.7}) {samplerType=SAMPLERTYPE_LinearGrayscale, texture=&T_StormCloudDetail01_M}.R
  )
  n9 = MaterialFunctionCall(Float2=MaterialFunctionCall(XY=TextureCoordinate()) {materialFunction=&Swizzle}.YX) {materialFunction=&BreakOutFloat2Components}
  n10 = OneMinus(None=@n9.R)
  n11 = OneMinus(None=@n9.G)
  n12 = Multiply(A=@n8, B=Power(Base=Multiply(A=@n10, B=@n11)) {constExponent=3})
  n13 = LinearInterpolate(A=Saturate(None=Multiply(A=@n9.R) {constB=5}), Alpha=@n9.G) {constA=1, constB=0}
  DepthFade(Opacity=Multiply(A=@n1.A, B=Multiply(A=@n12, B=@n13)), FadeDistance=ScalarParameter() {defaultValue=100, parameterName=DepthFade}) {fadeDistanceDefault=200}
  MaterialFunctionCall() {materialFunction=/Game/Athena/Aircraft/PartyBus_RT/Materials/MF_PartyBusAnimationWPO}
  Multiply(A=@n1.RGB, B=@n8)
 ]
}
