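# Niagara system for the lobby island Battle Bus fire-up effect, a blue flare build-up that ends in a flash and flame burst.
# ZeroPointFlare spawns short-lived particles, each drawn by five layered sprite renderers with per-module attributes and modulated by emitter loop-age curves.
# At 0.8 seconds, FinalFlash fires a stretching glow with light and component renderers.
# Simultaneously, Flames bursts drag-slowed, SubUV-animated sprites outward from a point below.
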NiagaraSystem /Game/Environments/Asteria/Props/LobbyIslandSpace/FX/NS_BattleBus_FireUp
dependencies {
 usedDataInterfaces [&NiagaraDataInterfaceCurve, &NiagaraDataInterfaceVector2DCurve, &NiagaraDataInterfaceColorCurve]
 usedDynamicInputs []
 usedModules [&SystemState]
}
emitters {
 FinalFlash {
  emitterUpdateScript [
   &EmitterState()
   &SpawnBurst_Instantaneous(Spawn Time=0.8, Use Spawn Probability=false, UseLoopCountLimit=false)
  ]
  particleSpawnScript [
   &InitializeParticle(
    Color=<rgba 0.146 0.5729 1 1>
    Lifetime=0.5
    Sprite Size=<600 1500>
    Sprite Size Mode=::NewEnumerator4
    UsePositionOffset=false
    Write Material Random=true
   )
   /SetVariables(Particles.GlowSpriteAlignment:Vector3f=<0 0 1>) {isSetParameters=true}
   &SystemLocation(MaskSpawnGroup=false, Offset=<0 0 50>)
  ]
  particleUpdateScript [
   /SetVariables(Particles.LightExponent=3) {isSetParameters=true}
   /SetVariables(Particles.LightRadius=3) {isSetParameters=true}
   /SetVariables(Particles.LightColorv2:LinearColor=&MultiplyLinearColor(
     Linear Color=<rgba 0.7432 3.9734 10 1>
     Scale Factor=&MakeVector4_FromFloat(Float=&FloatFromCurve(
       FloatCurve=DataInterface() {propertyValues {Curve {keys [
           {arriveTangent=-2.1707, interpMode=RCIM_Cubic, leaveTangent=-2.1707, tangentMode=RCTM_User, time=0, value=1}
           {arriveTangent=-0.747, interpMode=RCIM_Cubic, leaveTangent=-0.747, tangentMode=RCTM_User, time=0.3, value=0.2}
           {tangentMode=RCTM_User, time=1, value=0}
          ]}}}
       Scale Curve=10
      ))
    )) {isSetParameters=true}
   &ParticleState()
   &ScaleColor(
    Scale Alpha=&Multiply_Float(
     A=1
     B=&FloatFromCurve(FloatCurve=DataInterface() {propertyValues {Curve {keys [<at 0 1 RCIM_Cubic>, {tangentMode=RCTM_User, time=1, value=0}]}}})
    )
    Scale RGB=&VectorFromFloat(Value=&FloatFromCurve(
      FloatCurve=DataInterface() {propertyValues {Curve {keys [
          {arriveTangent=-2.1707, interpMode=RCIM_Cubic, leaveTangent=-2.1707, tangentMode=RCTM_User, time=0, value=1}
          {tangentMode=RCTM_User, time=1, value=0}
         ]}}}
      Scale Curve=10
     ))
    ScaleA=true
    ScaleRGB=true
   )
   &ScaleSpriteSize(
    Non-Uniform Curve Sprite Scale=DataInterface() {propertyValues {
      XCurve {keys [{arriveTangent=-1.1225, interpMode=RCIM_Cubic, leaveTangent=-1.1225, tangentMode=RCTM_User, time=0, value=1}, <at 1 0.25 RCIM_Cubic>]}
      YCurve {keys [{arriveTangent=1.8991, interpMode=RCIM_Cubic, leaveTangent=1.8991, tangentMode=RCTM_User, time=0, value=0.4}, <at 1 1 RCIM_Cubic>]}
     }}
    Scale Sprite Size Mode=::NewEnumerator3
   )
  ]
  rendererClasses [
   &NiagaraSpriteRendererProperties
   &NiagaraLightRendererProperties
   &NiagaraComponentRendererProperties
  ]
  renderers [
   {rendererClass=&NiagaraSpriteRendererProperties}
   {rendererClass=&NiagaraLightRendererProperties, rendererIndex=1}
   {rendererClass=&NiagaraSpriteRendererProperties, rendererIndex=2}
   {rendererClass=&NiagaraComponentRendererProperties, rendererIndex=3}
  ]
 }
 Flames {
  emitterUpdateScript [
   &EmitterState()
   &SpawnBurst_Instantaneous(Spawn Count=10, Spawn Time=0.8, Use Spawn Probability=false, UseLoopCountLimit=false)
  ]
  particleSpawnScript [
   &InitializeParticle(
    Color=<rgba 0 1.098 2 1>
    Lifetime Max=0.5
    Lifetime Min=0.3
    Lifetime Mode=::NewEnumerator1
    Sprite Rotation Mode=::NewEnumerator2
    Sprite Size Mode=::NewEnumerator1
    Sprite UV Mode=::NewEnumerator3
    Uniform Sprite Size Max=250
    Uniform Sprite Size Min=120
    UsePositionOffset=false
    Write Material Random=true
   )
   &DynamicMaterialParameters(
    Index 0 Param 2=&UniformRangedFloat(Minimum=0, Maximum=1)
    Index 0 Param 4=&UniformRangedFloat(Minimum=3, Maximum=5)
    Module.Param0WriteEnabled=false
    Module.Param1WriteEnabled=true
    Module.Param2WriteEnabled=false
    Module.Param3WriteEnabled=true
   )
   /Niagara/Modules/Spawn/Location/SphereLocation(Hemisphere Z=true, MaskSpawnGroup=false, Module.Override Randomness=false, Offset=<0 0 5>, Sphere Radius=50, Surface Only=false)
   &AddVelocityFromPoint(
    Module.Use Falloff Distance=false
    Origin Offset=<0 0 -150>
    Origin Offset Coordinate Space=::World
    Velocity Strength=&Multiply_Float(A=&RandomRangeFloat(Minimum=50, Maximum=400), B=4)
   )
  ]
  particleUpdateScript [
   &ParticleState()
   &Drag(Drag=2, Use Linear Drag=true, Use Rotational Drag=false)
   &ScaleColor(
    Scale Mode=::NewEnumerator1
    Scale RGBA=&MakeVector4_FromLinearColor(LinearColor=&ColorFromCurve(ColorCurve=DataInterface() {propertyValues {
        AlphaCurve {keys [
          <at 0.0133 0>
          {arriveTangent=-3.3426, interpMode=RCIM_Cubic, leaveTangent=-3.3426, tangentMode=RCTM_User, time=0.1255, value=1}
          <at 0.9977 -0.003 RCIM_Cubic>
         ]}
        BlueCurve {keys [<at 0.1486 32 RCIM_Cubic>]}
        GreenCurve {keys [<at 0.1486 32 RCIM_Cubic>]}
        RedCurve {keys [<at 0.1486 32 RCIM_Cubic>]}
       }}))
    ScaleRGBA=true
   )
   &ScaleColor(ScaleA=true, ScaleRGB=true) {module=ScaleColor001}
   &SolveForcesAndVelocity(Clamp Velocity=false, Limit Acceleration=false)
   &DynamicMaterialParameters(
    Index 0 Param 1=&FloatFromCurve(FloatCurve=DataInterface() {propertyValues {Curve {keys [<at 0.0017 0.9908 RCIM_Cubic>, {arriveTangent=0.0101, interpMode=RCIM_Cubic, leaveTangent=0.0101, time=0.3821, value=0.8044}, <at 0.9074 1>]}}})
    Index 0 Param 3=&FloatFromCurve(FloatCurve=DataInterface() {propertyValues {Curve {keys [
         <at 0.0123 1.3769 RCIM_Cubic>
         {arriveTangent=-0.0921, interpMode=RCIM_Cubic, leaveTangent=-0.0921, time=0.367, value=1.1881}
         <at 0.9928 1.2866>
        ]}}})
    Module.Param0WriteEnabled=true
    Module.Param1WriteEnabled=false
    Module.Param2WriteEnabled=true
    Module.Param3WriteEnabled=false
   ) {module=DynamicMaterialParameters001}
   &ScaleSpriteSize(Uniform Curve Sprite Scale=DataInterface() {propertyValues {Curve {keys [
        {arriveTangent=1.571, interpMode=RCIM_Cubic, leaveTangent=1.571, tangentMode=RCTM_User, time=0, value=0.3}
        {tangentMode=RCTM_User, time=1, value=1}
       ]}}})
   /Niagara/Modules/Update/SubUV/SubUVAnimation(Number Of Frames=&FloatFromCurve(FloatCurve=DataInterface() {propertyValues {Curve {keys [{arriveTangent=38.9383, interpMode=RCIM_Cubic, leaveTangent=38.9383, tangentMode=RCTM_User, time=0.0156, value=7.4642}, <at 1.0068 39.7662>]}}}))
  ]
  rendererClasses [&NiagaraSpriteRendererProperties]
  renderers [{rendererClass=&NiagaraSpriteRendererProperties}]
 }
 ZeroPointFlare {
  emitterUpdateScript [
   &EmitterState(Life Cycle Mode=::NewEnumerator1, Loop Behavior=::NewEnumerator1, Loop Duration=0.95, UseLoopDelay=true)
   &SpawnRate(SpawnRate=7, Use Spawn Probability=false)
  ]
  particleSpawnScript [
   /Niagara/Modules/Spawn/Initialization/InitializeParticle(
    Lifetime=&Multiply_Float(
     A=&RandomRangeFloat(Minimum=0.5, Maximum=0.8)
     B=&FloatFromCurve(
      FloatCurve=DataInterface() {propertyValues {Curve {keys [{arriveTangent=-1.5125, interpMode=RCIM_Cubic, leaveTangent=-1.5125, tangentMode=RCTM_User, time=0, value=1}, <at 0.9949 0.2248>]}}}
      CurveIndex=$Emitter.NormalizedLoopAge
     )
    )
    Module.Write Color=true
    Module.Write Lifetime=true
    Module.Write Mass=true
    Module.Write Material Random=true
    Module.Write Position=false
    Module.Write Scale=false
    Module.Write SpriteRotation=true
    Module.Write SpriteSize=true
    Sprite Rotation=&RandomRangeFloat(Minimum=0, Maximum=360)
    Sprite Size=<3000 3000>
   )
   &Color(Color=&MakeLinearColorFromVectorAndFloat("Vector (RGB)"=<0.0175 0.0243 0.03>, "Float (Alpha)"=&ClampFloat(Float=0.6)))
   &SystemLocation(MaskSpawnGroup=false, Offset=<0 0 30>)
  ]
  particleUpdateScript [
   &ParticleState()
   &ScaleColor(
    Scale Alpha=&FloatFromCurve(FloatCurve=DataInterface() {propertyValues {Curve {keys [<at 0 1 RCIM_Cubic>,<at 0.25 1 RCIM_Cubic>,<at 0.75 1 RCIM_Cubic>,<at 1 1>]}}})
    ScaleA=true
    ScaleRGB=true
   )
   /SetVariables(Particles.Size:NiagaraFloat=&FloatFromCurve(
     FloatCurve=DataInterface() {propertyValues {Curve {keys [
         <at -0.0039 0.0032 RCIM_Cubic>
         {arriveTangent=0.0358, interpMode=RCIM_Cubic, leaveTangent=0.0358, tangentMode=RCTM_User, time=0.4967, value=0.9107}
         <at 1 0>
        ]}}}
     Scale Curve=&FloatFromCurve(
      FloatCurve=DataInterface() {propertyValues {Curve {keys [
          <at -0.0005 0.984 RCIM_Cubic>
          {arriveTangent=0.3225, interpMode=RCIM_Cubic, leaveTangent=0.3225, tangentMode=RCTM_User, time=0.4833, value=3.7073}
          <at 1.0108 1.1471>
         ]}}}
      CurveIndex=$Emitter.NormalizedLoopAge
      Scale Curve=0.5
     )
    )) {isSetParameters=true}
   &FN_Sprite_Attributes(
    Camera Offset=-250
    Create Camera Offset Attribute=true
    Create Color Attribute=true
    Create SpriteSize Attribute=true
    SpriteSize=&Vector2DFromFloat(Value=&Multiply_Float(A=500, B=$Particles.Size))
    Write to Module Namespace=true
   )
   &FN_Sprite_Attributes(
    Camera Offset=75
    Color=&MakeLinearColorFromVectorAndFloat("Vector (RGB)"=<0.5625 1.8964 3>, "Float (Alpha)"=0.1)
    Create Camera Offset Attribute=true
    Create Color Attribute=true
    Create SpriteSize Attribute=true
    SpriteSize=&Vector2DFromFloat(Value=&Multiply_Float(A=500, B=$Particles.Size))
    Write to Module Namespace=true
   ) {module=FN_Sprite_Attributes001}
   &FN_Sprite_Attributes(
    Camera Offset=50
    Color=&MakeLinearColorFromVectorAndFloat(
     "Vector (RGB)"=<0.8602 1.4152 2>
     "Float (Alpha)"=&FloatFromCurve(
      FloatCurve=DataInterface() {propertyValues {Curve {keys [
          <at 0.0048 -0.0024 RCIM_Cubic>
          {arriveTangent=2.9233, interpMode=RCIM_Cubic, leaveTangent=2.9233, tangentMode=RCTM_User, time=0.7236, value=0.9786}
          <at 0.9988 -0.0024>
         ]}}}
      Scale Curve=&FloatFromCurve(
       FloatCurve=DataInterface() {propertyValues {Curve {keys [<at 0.0654 -0.0428>,<at 0.5237 0.7921 RCIM_Cubic>,<at 0.9993 -0.0428>]}}}
       CurveIndex=$Emitter.NormalizedLoopAge
       Scale Curve=0.3
      )
     )
    )
    Create Camera Offset Attribute=true
    Create Color Attribute=true
    Create SpriteSize Attribute=true
    SpriteSize=&Vector2DFromFloat(Value=&Multiply_Float(
      A=200
      B=&FloatFromCurve(
       FloatCurve=DataInterface() {propertyValues {Curve {keys [
           <at 0.004 1.0518 RCIM_Cubic>
           {arriveTangent=-0.8729, interpMode=RCIM_Cubic, leaveTangent=-0.8729, tangentMode=RCTM_User, time=0.9881, value=0.7637}
          ]}}}
       Scale Curve=2
      )
     ))
    Write to Module Namespace=true
   ) {module=FN_Sprite_Attributes002}
   &FN_Sprite_Attributes(
    Camera Offset=10
    Color=&MakeLinearColorFromVectorAndFloat(
     "Vector (RGB)"=&Multiply_VectorByFloat(Vector=<1.5665 2.3369 6>, Float=0.5)
     "Float (Alpha)"=&FloatFromCurve(
      FloatCurve=DataInterface() {propertyValues {Curve {keys [
          <at 0.0048 -0.0024 RCIM_Cubic>
          {arriveTangent=2.9233, interpMode=RCIM_Cubic, leaveTangent=2.9233, tangentMode=RCTM_User, time=0.7236, value=0.9786}
          <at 0.9988 -0.0024>
         ]}}}
      Scale Curve=0.5
     )
    )
    Create Camera Offset Attribute=true
    Create Color Attribute=true
    Create Sprite Rotation Attribute=true
    Create SpriteSize Attribute=true
    SpriteSize=&Multiply_Vector2(
     A=<1000 20>
     B=&Vector2DFromFloat(Value=&RandomRangeFloat(Minimum=0.2, Maximum=1))
    )
    Write to Module Namespace=true
   ) {module=FN_Sprite_Attributes003}
   &FN_Sprite_Attributes(
    Camera Offset=45
    Color=&MakeLinearColorFromVectorAndFloat("Vector (RGB)"=<0.2398 1.0892 1.5>, "Float (Alpha)"=&ClampFloat(Float=0.5))
    Create Camera Offset Attribute=true
    Create Color Attribute=true
    Create Sprite Rotation Attribute=true
    Create SpriteSize Attribute=true
    SpriteSize=&Multiply_Vector2(
     A=&RandomRangeVector2D(Minimum=<120 80>, Maximum=<200 140>, Evaluation Type=::NewEnumerator1)
     B=&Vector2DFromFloat(Value=&FloatFromCurve(
       FloatCurve=DataInterface() {propertyValues {Curve {keys [
           <at 0 0 RCIM_Cubic>
           {arriveTangent=1.4286, interpMode=RCIM_Cubic, leaveTangent=1.4286, time=0.5754, value=0.3493}
           {arriveTangent=-0.8228, interpMode=RCIM_Cubic, leaveTangent=-0.8228, time=0.7, value=1}
           <at 1 0>
          ]}}}
       Scale Curve=5
      ))
    )
    Write to Module Namespace=true
   ) {module=FN_Sprite_Attributes004}
   &FN_Light_Attributes(
    Color=<rgba 1 2 10 1>
    Create Color Attribute=true
    Create Position Attribute=true
    Enable Visibility Tag=true
    Position=&Add_Vector(
     A=&Add_Vector(
      A=Linked() {linkedVariable {name=Particles.Position, type=NiagaraPosition}}
      B=&RandomVector(Vector Scale=20, Evaluation Type=::NewEnumerator1)
     )
     B=&TransformVector(Vector=<0 0 50>, Source Space=::World, Destination Space=::Local)
    )
    Scale Color=&FloatFromCurve(
     FloatCurve=DataInterface() {propertyValues {Curve {keys [<at -0.0008 -0.006 RCIM_Cubic>, {arriveTangent=0.006, interpMode=RCIM_Cubic, leaveTangent=0.006, time=0.4887, value=0.9748}, <at 1 0>]}}}
     Scale Curve=10
    )
    VisibilityTag=HlslExpression() {hlslExpression="Particles.UniqueID && 1"}
   )
   &FN_Gen5_Light(Color:LinearColor=<rgba 1 2 10 1>, Intensity=1000, Position Offset:Vector3f=<0 0 400>)
  ]
  rendererClasses [&NiagaraSpriteRendererProperties]
  renderers [
   {rendererClass=&NiagaraSpriteRendererProperties}
   {rendererClass=&NiagaraSpriteRendererProperties, rendererIndex=1}
   {rendererClass=&NiagaraSpriteRendererProperties, rendererIndex=2}
   {rendererClass=&NiagaraSpriteRendererProperties, rendererIndex=3}
   {rendererClass=&NiagaraSpriteRendererProperties, rendererIndex=4}
  ]
 }
}
summary {
 systemName=NS_BattleBus_FireUp
 emitters [
  {emitterName=ZeroPointFlare, rendererClasses [&NiagaraSpriteRendererProperties]}
  {
   emitterName=FinalFlash
   rendererClasses [
    &NiagaraSpriteRendererProperties
    &NiagaraLightRendererProperties
    &NiagaraComponentRendererProperties
   ]
  }
  {emitterName=Flames, rendererClasses [&NiagaraSpriteRendererProperties]}
 ]
 userVariables []
}
