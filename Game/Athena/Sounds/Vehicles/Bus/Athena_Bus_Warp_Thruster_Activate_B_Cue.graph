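# Sound cue for the Battle Bus warp thruster activation, variant B.
# A single wave player plays the 3D activation wave through an attenuation node using the lobby warp thruster attenuation asset.
# It routes to the Battle Bus sound class, keeps playing when silent, and overrides concurrency to allow one instance at a time.
# Priority is slightly raised and bypasses volume scaling, with a long max audible distance.
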SoundCue /Game/Athena/Sounds/Vehicles/Bus/Athena_Bus_Warp_Thruster_Activate_B_Cue
root=Attenuation_1
nodes [
 {
  id=Attenuation_1
  children [WavePlayer_2]
  properties {attenuationSettings=/Game/Sounds/Attenuation/Fort_Bus_Warp_Lobby_Thrusters_Athena_Att}
 }
 {id=WavePlayer_2, properties {soundWaveAssetPtr=&/Bus_Warp_Thruster_Activate_B_3D}}
]
properties {
 bBypassVolumeScaleForPriority=true
 bOverrideConcurrency=true
 maxDistance=50000
 priority=1.2
 soundClassObject=/Game/Sounds/FortSoundClassesAndModes/Fort_BattleBus
 virtualizationMode=PlayWhenSilent
 volumeMultiplier=0.85
 concurrencyOverrides {maxCount=1}
}
