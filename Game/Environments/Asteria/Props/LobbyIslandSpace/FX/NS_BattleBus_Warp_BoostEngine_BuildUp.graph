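# Niagara system for the Battle Bus warp boost engine build-up in the Asteria lobby island space scene.
# Warmup meshes, flame elements, ground smoke, GPU blown particles and an animated light evolve on System.Age curves, mostly tinted by a shared engine color.
# Emitters delayed by MainEnginesTurnOnTime fire the ignition burst, streaks, flash meshes and colliding GPU confetti.
# Material parameters animate the meshes, while sputter and User opacity and flicker parameters modulate the light.
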NiagaraSystem /Game/Environments/Asteria/Props/LobbyIslandSpace/FX/NS_BattleBus_Warp_BoostEngine_BuildUp
dependencies {
 usedDataInterfaces [
  &NiagaraDataInterfaceColorCurve
  &NiagaraDataInterfaceCurve
  &NiagaraDataInterfaceArrayFloat
  &NiagaraDataInterfaceAudioPlayer
  &NiagaraDataInterfaceVectorCurve
  &NiagaraDataInterfaceMeshRendererInfo
  &NiagaraDataInterfaceVector2DCurve
  &NiagaraDataInterfaceArrayFloat3
 ]
 usedDynamicInputs []
 usedModules [
  &self:SystemSpawnScript.SystemScriptSource.SystemScriptGraph.NiagaraNodeAssignment_0.SetVariables_#1
  &self:SystemSpawnScript.SystemScriptSource.SystemScriptGraph.NiagaraNodeAssignment_5.SetVariables_#2
  &self:SystemSpawnScript.SystemScriptSource.SystemScriptGraph.NiagaraNodeAssignment_7.SetVariables_#3
  &SystemState
  &self:SystemSpawnScript.SystemScriptSource.SystemScriptGraph.NiagaraNodeAssignment_6.SetVariables_#4
  &self:SystemSpawnScript.SystemScriptSource.SystemScriptGraph.NiagaraNodeAssignment_1.SetVariables_#5
  &self:SystemSpawnScript.SystemScriptSource.SystemScriptGraph.NiagaraNodeAssignment_4.SetVariables_#6
  &self:SystemSpawnScript.SystemScriptSource.SystemScriptGraph.NiagaraNodeAssignment_2.SetVariables_#7
  &self:SystemSpawnScript.SystemScriptSource.SystemScriptGraph.NiagaraNodeAssignment_3.SetVariables_#8
 ]
}
emitters {
 AnimatedLight {
  emitterSpawnScript [
   /SetVariables(Emitter.Distortion_Input=1, Emitter.LightConeOverallSize=2.2) {isSetParameters=true, module=SetVariables_#9}
   /SetVariables(Emitter.FlareSizeScalar=0.14) {isSetParameters=true, module=SetVariables_#10}
  ]
  emitterUpdateScript [
   &EmitterState(Life Cycle Mode=::NewEnumerator1, Loop Behavior=::NewEnumerator1, Loop Duration Mode=::NewEnumerator1, UseLoopDelay=false)
   /SetVariables(Emitter.LightBasePos:NiagaraPosition=<0 -350 0>) {isSetParameters=true}
   &SpawnBurst_Instantaneous(Spawn Count=1, Use Spawn Probability=false)
  ]
  particleSpawnScript [
   &InitializeParticle(
    Color=<rgba 0.2707 0.9572 2 1>
    Lifetime=5
    Sprite Size=&Multiply_Vector2D_ByFloat(Float=$Emitter.FlareSizeScalar, Vector2D=<30000 10000>)
    Sprite Size Mode=::NewEnumerator4
    UsePositionOffset=false
    Write Material Random=true
   )
   &CameraOffset(Camera Offset Amount=200)
   &SystemLocation(MaskSpawnGroup=false, Offset=<0 0 50>)
   /SetVariables(Particles.LightPos:NiagaraPosition=<0 -450 0>) {isSetParameters=true, module=SetVariables_#11}
   /SetVariables() {isSetParameters=true}
  ]
  particleUpdateScript [
   &ParticleState(Kill Particles When Lifetime Has Elapsed=false, Let Infinitely Lived Particles Die When Emitter Deactivates=true)
   /SetVariables(
    Particles.LensColor:LinearColor=&MultiplyLinearColor(
     Linear Color=<rgba 0.1086 0.1286 0.2 1>
     Scale Factor=&MakeVector4_FromFloat(Float=&Multiply_Float(A=$User.OverallOpacity, B=$System.SputterFloat))
    )
    Particles.LensSize:Vector2f=&Multiply_Vector2D_ByFloat(Float=$Emitter.FlareSizeScalar, Vector2D=<20000 4000>)
   ) {isSetParameters=true, module=SetVariables_#12}
   /SetVariables(
    Particles.DarkRadialColor:LinearColor=&MultiplyLinearColor(
     Linear Color=<rgba 0.0902 0.0318 0.1615 0.5>
     Scale Factor=&MakeVector4_FromFloat(Float=&Multiply_Float(A=$User.OverallOpacity, B=$System.SputterFloat))
    )
    Particles.DarkRadialSize:Vector2f=&Multiply_Vector2D_ByFloat(Float=$Emitter.FlareSizeScalar, Vector2D=<15000 10000>)
   ) {isSetParameters=true, module=SetVariables_#13}
   /SetVariables(Particles.LightColor=&MultiplyLinearColor(
     Linear Color=$System.EnginesColorOveLife
     Scale Factor=&MakeVector4_FromFloat(Float=&FloatFromCurve(
       FloatCurve=DataInterface() {propertyValues {Curve {keys [<at 6 0.5 RCIM_Cubic>,<at 7.1 1.25 RCIM_Cubic>]}}}
       CurveIndex=$System.Age
       Scale Curve=10
      ))
    )) {isSetParameters=true, module=SetVariables_#14}
   &ScaleSpriteSize(Scale Factor=&Vector2DFromFloat(Value=&Lerp_Float(
      A=1
      B=&RandomRangeFloat(Minimum=0.8, Maximum=1, Evaluation Type=::NewEnumerator1)
      Alpha=$User.FlareFlicker
     )))
   /SetVariables(Particles.LightOffset:Vector3f=/Niagara/DynamicInputs/Lerp/SmoothLerpOverTime_Vector(
     Convergence Rate:NiagaraFloat=5
     Smooth Value:Vector3f=$Particles.LightOffset
     Target Value:Vector3f=&RandomVector(
      Vector Scale=&FloatFromCurve(
       FloatCurve=DataInterface() {propertyValues {Curve {keys [<at 0 0.55 RCIM_Cubic>, {arriveTangent=0.45, interpMode=RCIM_Cubic, leaveTangent=0.45, time=0.8, value=0.65}, <at 1 1 RCIM_Cubic>]}}}
       Scale Curve=250
      )
      Constrain to Hemisphere=true
      Hemisphere Constraint Vector=<0 1 0>
      Evaluation Type=::NewEnumerator1
     )
    )) {isSetParameters=true}
   /SetVariables(Particles.LightPos:NiagaraPosition=&AddVectorToPosition(Position=$Emitter.LightBasePos, Vector=$Particles.LightOffset)) {isSetParameters=true}
   &ScaleColor(Scale Alpha=&Multiply_Float(A=$User.OverallOpacity, B=$System.SputterFloat), ScaleA=true, ScaleRGB=true)
  ]
  rendererClasses [&NiagaraLightRendererProperties]
  renderers [{rendererClass=&NiagaraLightRendererProperties, rendererIndex=0}]
 }
 BlowingParticles {
  simTarget=GPUComputeSim
  emitterUpdateScript [
   &EmitterState(Life Cycle Mode=::NewEnumerator1, Loop Behavior=::NewEnumerator1, Loop Duration Mode=::NewEnumerator1, UseLoopDelay=false)
   &SpawnRate(SpawnRate=25, Use Spawn Probability=false)
  ]
  particleSpawnScript [
   &InitializeParticle(
    Color=$System.EnginesColorOveLife
    Lifetime Max=3
    Lifetime Mode=::NewEnumerator1
    Mass Max=1.5
    Mass Mode=::NewEnumerator3
    Sprite Rotation Mode=::NewEnumerator2
    Sprite Size Mode=::NewEnumerator1
    Uniform Sprite Size Max=3
    Uniform Sprite Size Min=2
    UsePositionOffset=false
    Write Material Random=true
   )
   &ShapeLocation(MaskSpawnGroup=false, Override Seed=false, Sphere Radius=150, UseOwnerScale=true)
   &AddVelocity(Velocity=<0 -500 50>)
   &InitialMeshOrientation(Use Rotation Vector=false)
  ]
  particleUpdateScript [
   &ParticleState()
   &GravityForce(Gravity=<0 0 -1>)
   &Drag(Drag=2, Use Linear Drag=true, Use Rotational Drag=true)
   /SetVariables(Particles.DistFromSource:NiagaraFloat=&VectorLength(Vector=Linked() {linkedVariable {name=Particles.Position, type=NiagaraPosition}})) {isSetParameters=true}
   /SetVariables(Particles.Velocity=&MakeVector(
     X=&RandomRangeFloat(Minimum=-100, Maximum=100)
     Y=&Multiply_Float(
      A=&FloatFromCurve(
       FloatCurve=DataInterface() {propertyValues {Curve {keys [
           <at 0.1 0 RCIM_Cubic>
           <at 1.1 -0.25 RCIM_Cubic>
           {arriveTangent=0.3571, interpMode=RCIM_Cubic, leaveTangent=0.3571, time=2.6, value=0}
           {arriveTangent=0.8333, interpMode=RCIM_Cubic, leaveTangent=0.8333, time=3.2, value=0.5}
           {arriveTangent=0.4808, interpMode=RCIM_Cubic, leaveTangent=0.4808, time=3.8, value=1}
           {arriveTangent=1.1842, interpMode=RCIM_Cubic, leaveTangent=1.1842, time=4.5, value=1.125}
           <at 5.7 3.25>
           <at 6.2 2.125 RCIM_Cubic>
          ]}}}
       CurveIndex=$System.Age
       Scale Curve=-1000
      )
      B=&RemapRange(Input Value To Remap=$Particles.DistFromSource, Input Min=0, Input Max=1250, Output Min=1, Output Max=0.05)
     )
     Z=&RandomRangeFloat(Minimum=-100, Maximum=100)
    )) {isSetParameters=true}
   &CurlNoiseForce(Noise Frequency=2, Noise Strength=1827.0829, Pan Noise=true, Pan Noise Field=<0 -2.6188 0>)
   &SolveForcesAndVelocity(Clamp Velocity=false, Limit Acceleration=false)
   &ScaleSpriteSizeBySpeed(Max Scale Factor=<3 12>)
   &CalculateAccurateVelocity()
   &ScaleSpriteSize(Uniform Curve Sprite Scale=DataInterface() {propertyValues {Curve {keys [
        {arriveTangent=14.5835, interpMode=RCIM_Cubic, leaveTangent=14.5835, tangentMode=RCTM_User, time=0, value=0}
        {arriveTangent=1.1111, interpMode=RCIM_Cubic, leaveTangent=1.1111, time=0.1, value=1}
        {arriveTangent=-1.1111, interpMode=RCIM_Cubic, leaveTangent=-1.1111, time=0.9, value=1}
        {arriveTangent=-8.7113, interpMode=RCIM_Cubic, leaveTangent=-8.7113, tangentMode=RCTM_User, time=1, value=0}
       ]}}})
   &ScaleColor(
    Color Value To Scale=$Particles.Color
    Scale Alpha=0.5
    Scale RGB=&VectorFromFloat(Value=&FloatFromCurve(FloatCurve=DataInterface() {propertyValues {Curve {keys [<at 0 1 RCIM_Cubic>,<at 0.1 1.1 RCIM_Cubic>,<at 1 1 RCIM_Cubic>]}}}))
    ScaleA=true
    ScaleRGB=true
   )
  ]
  rendererClasses [&NiagaraSpriteRendererProperties]
  renderers [{rendererClass=&NiagaraSpriteRendererProperties}]
 }
 ConfettiBurst {
  simTarget=GPUComputeSim
  emitterUpdateScript [
   &EmitterState(
    Life Cycle Mode=::NewEnumerator1
    Loop Behavior=::NewEnumerator1
    Loop Delay=$System.MainEnginesTurnOnTime
    Loop Duration=1
    UseLoopDelay=true
   )
   &SpawnBurst_Instantaneous(Spawn Count=50, Use Spawn Probability=false, UseLoopCountLimit=false)
  ]
  particleSpawnScript [
   &InitializeParticle(
    Lifetime Max=1.25
    Lifetime Min=0.75
    Lifetime Mode=::NewEnumerator1
    Mass Max=1.5
    Mass Mode=::NewEnumerator3
    Mesh Scale Mode=::NewEnumerator1
    Mesh Uniform Scale Max=0.12
    Mesh Uniform Scale Min=0.075
    Sprite Rotation Mode=::NewEnumerator2
    Sprite Size Mode=::NewEnumerator1
    Uniform Sprite Size Max=15
    Uniform Sprite Size Min=2
    UsePositionOffset=false
    Write Material Random=true
   )
   &ShapeLocation(
    Handle Radius=20
    MaskSpawnGroup=false
    Non Uniform Scale=<1 1 0.2>
    Offset=<0 -200 -100>
    Offset Mode=::NewEnumerator0
    Override Seed=false
    Shape Primitive=::NewEnumerator3
    U Distribution=0.5
    UseOwnerScale=true
   )
   &AddVelocity(
    Module.Use Falloff Distance=false
    Origin Offset=<0 0 -44.2601>
    Override Seed=false
    Velocity Mode=::NewEnumerator1
    Velocity Speed=&RandomRangeFloat(Minimum=1200, Maximum=5500)
   )
   /SetVariables(Particles.Velocity=&Multiply_Vector(A=$Particles.Velocity, B=<0.5 1 0.1>)) {isSetParameters=true}
   &InitialMeshOrientation(Use Rotation Vector=false)
  ]
  particleUpdateScript [
   &ParticleState()
   &ScaleMeshSize(Scale Factor=&VectorFromFloat(Value=&FloatFromCurve(FloatCurve=DataInterface() {propertyValues {Curve {keys [
          <at 0 0.025 RCIM_Cubic>
          {arriveTangent=1.3929, interpMode=RCIM_Cubic, leaveTangent=1.3929, time=0.2, value=1.1}
          {arriveTangent=-1.375, interpMode=RCIM_Cubic, leaveTangent=-1.375, time=0.7, value=1}
          <at 1 0 RCIM_Cubic>
         ]}}})))
   &GravityForce()
   /SetVariables(Particles.Velocity=<0 -100 0>) {enabled=false, isSetParameters=true}
   &Collision(
    Age Colliding Particles=false
    Analytical Collision Normal 1:Vector3f=&TransformVector(Vector=<0 0 1>, Source Space=::World, Destination Space=::Simulation)
    Analytical Collision Plane Position 1:Vector3f=<0 0 -180>
    Control Roll On Collision=true
    Friction:NiagaraFloat=&RandomRangeFloat(Minimum=0.5, Maximum=1)
    Friction During a Bounce=0.5636
    GPU Collision Type=::NewEnumerator3
    Module.Control Rest State Via Surface Normal=false
    Module.EnableMaxCollisionCount=false
    Module.Randomize Collision Normal Vector=false
    Particle Radius Scale=3
    Restitution=0.8
    ScaleRateOfRotation=false
   )
   &AccelerationForce(Acceleration=<0 800 0>)
   &Drag(Drag=&RandomRangeFloat(Minimum=1, Maximum=1.5), Use Linear Drag=true, Use Rotational Drag=true)
   &SolveForcesAndVelocity(Clamp Velocity=false, Limit Acceleration=false)
   &ScaleSpriteSize(Uniform Curve Sprite Scale=DataInterface() {propertyValues {Curve {keys [
        {arriveTangent=14.5835, interpMode=RCIM_Cubic, time=0, value=0}
        {arriveTangent=3.25, interpMode=RCIM_Cubic, leaveTangent=3.25, time=0.2, value=0.3}
        {arriveTangent=1, interpMode=RCIM_Cubic, leaveTangent=1, time=0.3, value=0.975}
        {arriveTangent=-1.3929, interpMode=RCIM_Cubic, leaveTangent=-1.3929, time=0.9, value=1}
        {interpMode=RCIM_Cubic, leaveTangent=-20.5076, time=1, value=0}
       ]}}})
   &AlignSpriteToMeshOrientation(SetAlignment=true, SetFacing=true)
  ]
  rendererClasses [&NiagaraMeshRendererProperties]
  renderers [{rendererClass=&NiagaraMeshRendererProperties, rendererIndex=0}]
 }
 EngineBurst {
  emitterUpdateScript [
   &EmitterState(
    Life Cycle Mode=::NewEnumerator1
    Loop Behavior=::NewEnumerator1
    Loop Delay=$System.MainEnginesTurnOnTime
    Loop Duration=0.5
    UseLoopDelay=true
   )
   &SpawnBurst_Instantaneous(
    Loop Count Limit:NiagaraInt32=&MakeIntFromFloat(Float=1)
    Use Spawn Probability=false
    UseLoopCountLimit=false
   )
  ]
  particleSpawnScript [
   &InitializeParticle(
    Color=<rgba 4 1 1 1>
    Lifetime=0.2
    Mesh Scale Mode=::NewEnumerator3
    Position Offset=<0 -40 60>
    Sprite Size Mode=::NewEnumerator3
    UsePositionOffset=true
    Write Material Random=true
   )
   &InitialMeshOrientation(Mesh Orientation Mode=::NewEnumerator2, Use Rotation Vector=true)
  ]
  particleUpdateScript [
   &ParticleState()
   &ScaleMeshSize(Scale Factor=&MakeVector(
     X=1
     Y=&FloatFromCurve(FloatCurve=DataInterface() {propertyValues {Curve {keys [{interpMode=RCIM_Cubic, tangentMode=RCTM_User, time=0, value=0}, {interpMode=RCIM_Cubic, tangentMode=RCTM_User, time=1, value=1}]}}})
     Z=1
    ))
   &SolveForcesAndVelocity(Clamp Velocity=false, Limit Acceleration=false)
   &Color(Color=&ColorFromCurve(
     ColorCurve=DataInterface() {propertyValues {
       bOptimizeLUT=false
       AlphaCurve {keys [<at 0 1 RCIM_Cubic>,<at 0.0271 0.1>,<at 0.5909 1>,<at 1 1 RCIM_Cubic>]}
       BlueCurve {keys [<at 0 0.9541>,<at 0.0254 6>,<at 0.1043 1>,<at 0.9511 0.4286>]}
       GreenCurve {keys [<at 0 0.9017>,<at 0.0254 1.4994>,<at 0.1043 0.1527>,<at 0.9511 0.1025>]}
       RedCurve {keys [<at 0 1>,<at 0.0254 0.3369>,<at 0.1043 0.2814>,<at 0.9511 1>]}
      }}
     CurveIndex=&Multiply_Float(A=$Particles.NormalizedAge, B=1)
    ))
   &ScaleColor(
    Color Value To Scale=$StackContext.Color
    Scale Alpha=&Multiply_Float(
     A=1
     B=&FloatFromCurve(FloatCurve=DataInterface() {propertyValues {Curve {keys [<at 0 1 RCIM_Cubic>, {arriveTangent=-1, interpMode=RCIM_Cubic, leaveTangent=-1, time=0.4, value=0.55}, <at 1 0 RCIM_Cubic>]}}})
    )
    Scale RGB=&VectorFromFloat(Value=&FloatFromCurve(
      FloatCurve=DataInterface() {propertyValues {Curve {keys [
          {arriveTangent=-1.15, interpMode=RCIM_Cubic, time=0, value=0}
          {arriveTangent=1, interpMode=RCIM_Cubic, leaveTangent=1, time=0.0462, value=10}
          {interpMode=RCIM_Cubic, leaveTangent=-1.0093, time=1, value=1}
         ]}}}
      CurveIndex=$System.LoopedAge
      Scale Curve=10
     ))
    ScaleA=true
    ScaleRGB=true
   )
   &CameraOffset(Camera Offset Amount=30)
   &DynamicMaterialParameters(
    Index 0 Param 1=&FloatFromCurve(
     FloatCurve=DataInterface() {propertyValues {Curve {keys [<at 0.1 0.5>, {interpMode=RCIM_Cubic, leaveTangent=-0.175, time=0.9, value=1.125}]}}}
     Scale Curve=200
    )
    Index 0 Param 2=&Lerp_Float(
     A=0
     B=1.25
     Alpha=&FloatFromCurve(FloatCurve=DataInterface() {propertyValues {Curve {keys [{interpMode=RCIM_Cubic, tangentMode=RCTM_User, time=0, value=0}, {interpMode=RCIM_Cubic, tangentMode=RCTM_User, time=1, value=1}]}}})
    )
    Index 0 Param 3=&Add_Float(
     A=&FloatFromCurve(FloatCurve=DataInterface() {propertyValues {Curve {keys [<at 0 0 RCIM_Cubic>,<at 1 0.625 RCIM_Cubic>]}}}, Scale Curve=0.75)
     B=&RandomRangeFloat(Minimum=0, Maximum=1)
    )
    Module.Param0WriteEnabled=true
    Module.Param1WriteEnabled=true
    Module.Param2WriteEnabled=true
    Module.Param3WriteEnabled=true
   )
  ]
  rendererClasses [&NiagaraMeshRendererProperties]
  renderers [{rendererClass=&NiagaraMeshRendererProperties, rendererIndex=0}]
 }
 EngineInteriorWarmup {
  emitterSpawnScript [/SetVariables() {isSetParameters=true}]
  emitterUpdateScript [
   &EmitterState(Life Cycle Mode=::NewEnumerator1, Loop Behavior=::NewEnumerator1, Loop Duration Mode=::NewEnumerator1, UseLoopDelay=false)
   &SpawnBurst_Instantaneous(Use Spawn Probability=false, UseLoopCountLimit=false)
   /SetVariables(Emitter.FireHeatColor:LinearColor=&ColorFromCurve(
     ColorCurve=DataInterface() {propertyValues {
       AlphaCurve {keys [<at 0 1 RCIM_Cubic>,<at 1 0 RCIM_Cubic>]}
       BlueCurve {keys [<at 0.1835 1 RCIM_Cubic>]}
       GreenCurve {keys [<at 0.1835 0.828 RCIM_Cubic>]}
       RedCurve {keys [<at 0.1835 0.0365 RCIM_Cubic>]}
      }}
     CurveIndex=$Emitter.NormalizedLoopAge
    )) {isSetParameters=true}
  ]
  particleSpawnScript [
   &InitializeParticle(
    Color=<rgba 4 1 1 1>
    Lifetime=10
    Mesh Scale Mode=::NewEnumerator3
    Position Offset=<0 0 60>
    Sprite Size Mode=::NewEnumerator3
    UsePositionOffset=true
    Write Material Random=true
   )
   &InitialMeshOrientation(Mesh Orientation Mode=::NewEnumerator2, Use Rotation Vector=true)
  ]
  particleUpdateScript [
   &ParticleState(Kill Particles When Lifetime Has Elapsed=false, Let Infinitely Lived Particles Die When Emitter Deactivates=true)
   &ScaleMeshSize(Scale Factor=&MakeVector(
     X=1
     Y=&FloatFromCurve(
      FloatCurve=DataInterface() {propertyValues {Curve {keys [
          <at 0 0 RCIM_Cubic>
          {arriveTangent=-1.9692, interpMode=RCIM_Cubic, leaveTangent=-1.9692, tangentMode=RCTM_User, time=0.0391, value=6}
          {arriveTangent=-0.1405, interpMode=RCIM_Cubic, leaveTangent=-0.1405, tangentMode=RCTM_User, time=2.1, value=0.5}
          <at 9.4 1.5 RCIM_Cubic>
         ]}}}
      CurveIndex=&Multiply_Float(A=$System.Age, B=1.75)
     )
     Z=1
    ))
   &SolveForcesAndVelocity(Clamp Velocity=false, Limit Acceleration=false)
   &Color(Color=&ColorFromCurve(
     ColorCurve=DataInterface() {propertyValues {
       bOptimizeLUT=false
       AlphaCurve {keys [<at 0 1 RCIM_Cubic>,<at 0.0271 0.1>,<at 0.5909 1>,<at 1 1 RCIM_Cubic>]}
       BlueCurve {keys [<at 0 0.9541>,<at 0.0254 6>,<at 0.6092 1>,<at 0.9511 0.4286>]}
       GreenCurve {keys [<at 0 0.9017>,<at 0.0254 1.4994>,<at 0.6092 0.1527>,<at 0.9511 0.1025>]}
       RedCurve {keys [<at 0 1>,<at 0.0254 0.3369>,<at 0.6092 0.2814>,<at 0.9511 1>]}
      }}
     CurveIndex=&Multiply_Float(A=$System.Age, B=0.275)
    )) {enabled=false}
   &Color(Color=$System.EnginesColorOveLife) {module=Color001}
   &ScaleColor(
    Color Value To Scale=$StackContext.Color
    Scale Alpha=0.03
    Scale RGB=&VectorFromFloat(Value=&FloatFromCurve(
      FloatCurve=DataInterface() {propertyValues {Curve {keys [
          {arriveTangent=-1.15, interpMode=RCIM_Cubic, time=0, value=0}
          {arriveTangent=1.25, interpMode=RCIM_Cubic, leaveTangent=1.25, time=0.0462, value=10}
          {arriveTangent=-1.0093, interpMode=RCIM_Cubic, leaveTangent=-1.0093, time=1, value=1.25}
          <at 5 5>
         ]}}}
      CurveIndex=$System.LoopedAge
      Scale Curve=10
     ))
    ScaleA=true
    ScaleRGB=true
   )
   &CameraOffset(Camera Offset Amount=10)
   &DynamicMaterialParameters(
    Index 0 Param 1=&Lerp_Float(
     A=&FloatFromCurve(
      FloatCurve=DataInterface() {propertyValues {Curve {keys [
          {interpMode=RCIM_Cubic, tangentMode=RCTM_User, time=0, value=0}
          {interpMode=RCIM_Cubic, tangentMode=RCTM_User, time=0.5, value=2}
          <at 0.9 0.3>
          <at 6 0.5>
         ]}}}
      CurveIndex=$System.Age
     )
     B=&Waveform(
      "Waveform [1]"=::NewEnumerator12
      "[1] Frequency"=15
      "[1] Amplitude Min/Max"=<0 1>
      "[1] Scale"=0.5
      "[1] Bias"=0.35
      "Waveform [2] Blend"=::NewEnumerator5
     )
     Alpha=&FloatFromCurve(
      FloatCurve=DataInterface() {propertyValues {Curve {keys [{interpMode=RCIM_Cubic, tangentMode=RCTM_User, time=0.75, value=0}, {interpMode=RCIM_Cubic, tangentMode=RCTM_User, time=2, value=1}]}}}
      CurveIndex=$System.Age
     )
    )
    Index 0 Param 2=&FloatFromCurve(FloatCurve=DataInterface() {propertyValues {Curve {keys [<at 4.2944 0.15 RCIM_Cubic>,<at 6 0 RCIM_Cubic>]}}}, CurveIndex=$System.Age)
    Index 0 Param 4=&FloatFromCurve(FloatCurve=DataInterface() {propertyValues {Curve {keys [<at 0 5>,<at 1 2>]}}}, CurveIndex=$System.Age)
    Index 1 Param 1=&Lerp_Float(
     A=80
     B=100
     Alpha=&Waveform("Waveform [1]"=::NewEnumerator12, "[1] Frequency"=25, "[1] Amplitude Min/Max"=<0 1>)
    )
    Index 1 Param 2=&Lerp_Float(
     A=2
     B=-2
     Alpha=&FloatFromCurve(
      FloatCurve=DataInterface() {propertyValues {Curve {keys [
          {arriveTangent=0.16, interpMode=RCIM_Cubic, time=0, value=0}
          {arriveTangent=0.2353, interpMode=RCIM_Cubic, leaveTangent=0.2353, time=0.75, value=0.3}
          <at 4.25 1 RCIM_Cubic>
         ]}}}
      CurveIndex=$System.Age
     )
    )
    Module.Param0WriteEnabled=true
    Module.Param0WriteEnabled1=true
    Module.Param1WriteEnabled=true
    Module.Param1WriteEnabled1=true
    Module.Param2WriteEnabled=true
    Module.Param2WriteEnabled1=true
    Module.Param3WriteEnabled=true
    Module.Param3WriteEnabled1=true
    Write Parameter Index 1=true
   )
  ]
  rendererClasses [&NiagaraMeshRendererProperties]
  renderers [{rendererClass=&NiagaraMeshRendererProperties, rendererIndex=0}]
 }
 EngineInteriorWarmup001 {
  emitterUpdateScript [
   &EmitterState(
    Life Cycle Mode=::NewEnumerator1
    Loop Behavior=::NewEnumerator1
    Loop Delay=$System.MainEnginesTurnOnTime
    Loop Duration=1
    UseLoopDelay=true
   )
   &SpawnBurst_Instantaneous(
    Loop Count Limit:NiagaraInt32=&MakeIntFromFloat(Float=1)
    Use Spawn Probability=false
    UseLoopCountLimit=false
   )
  ]
  particleSpawnScript [
   &InitializeParticle(
    Color=<rgba 4 1 1 1>
    Mesh Scale Mode=::NewEnumerator3
    Position Offset=<0 0 60>
    Sprite Size Mode=::NewEnumerator3
    UsePositionOffset=true
    Write Material Random=true
   )
   &InitialMeshOrientation(Mesh Orientation Mode=::NewEnumerator2, Use Rotation Vector=true)
  ]
  particleUpdateScript [
   &ParticleState()
   &ScaleMeshSize(Scale Factor=&MakeVector(
     X=1
     Y=&FloatFromCurve(
      FloatCurve=DataInterface() {propertyValues {Curve {keys [
          <at 0 0 RCIM_Cubic>
          {arriveTangent=-1.9692, interpMode=RCIM_Cubic, leaveTangent=-1.9692, tangentMode=RCTM_User, time=0.0391, value=6}
          {arriveTangent=-0.1405, interpMode=RCIM_Cubic, leaveTangent=-0.1405, tangentMode=RCTM_User, time=2.1, value=0.5}
          <at 9.4 1.5 RCIM_Cubic>
         ]}}}
      CurveIndex=&Multiply_Float(A=$Particles.NormalizedAge, B=1)
      Scale Curve=0.75
     )
     Z=1
    ))
   &SolveForcesAndVelocity(Clamp Velocity=false, Limit Acceleration=false)
   &Color(Color=&ColorFromCurve(
     ColorCurve=DataInterface() {propertyValues {
       bOptimizeLUT=false
       AlphaCurve {keys [<at 0 1 RCIM_Cubic>,<at 0.0271 0.1>,<at 0.5909 1>,<at 1 1 RCIM_Cubic>]}
       BlueCurve {keys [<at 0 0.9541>,<at 0.0254 6>,<at 0.1043 1>,<at 0.9511 0.4286>]}
       GreenCurve {keys [<at 0 0.9017>,<at 0.0254 1.4994>,<at 0.1043 0.1527>,<at 0.9511 0.1025>]}
       RedCurve {keys [<at 0 1>,<at 0.0254 0.3369>,<at 0.1043 0.2814>,<at 0.9511 1>]}
      }}
     CurveIndex=&Multiply_Float(A=$Particles.NormalizedAge, B=1)
    ))
   &ScaleColor(
    Color Value To Scale=$StackContext.Color
    Scale Alpha=&Multiply_Float(
     A=0.03
     B=&FloatFromCurve(FloatCurve=DataInterface() {propertyValues {Curve {keys [<at 0 1 RCIM_Cubic>, {arriveTangent=-1, interpMode=RCIM_Cubic, leaveTangent=-1, time=0.7, value=0.925}, <at 1 0 RCIM_Cubic>]}}})
    )
    Scale RGB=&VectorFromFloat(Value=&FloatFromCurve(
      FloatCurve=DataInterface() {propertyValues {Curve {keys [
          {arriveTangent=-1.15, interpMode=RCIM_Cubic, time=0, value=0}
          {arriveTangent=1.25, interpMode=RCIM_Cubic, leaveTangent=1.25, time=0.0462, value=10}
          {arriveTangent=-1.0093, interpMode=RCIM_Cubic, leaveTangent=-1.0093, time=1, value=1.25}
          <at 5 5>
         ]}}}
      CurveIndex=$System.LoopedAge
      Scale Curve=10
     ))
    ScaleA=true
    ScaleRGB=true
   )
   &CameraOffset(Camera Offset Amount=10)
   &DynamicMaterialParameters(
    Index 0 Param 1=&Lerp_Float(
     A=&FloatFromCurve(
      FloatCurve=DataInterface() {propertyValues {Curve {keys [
          {interpMode=RCIM_Cubic, tangentMode=RCTM_User, time=0, value=0}
          {interpMode=RCIM_Cubic, tangentMode=RCTM_User, time=0.5, value=2}
          <at 0.9 0.3>
          <at 6 0.5>
         ]}}}
      CurveIndex=$System.Age
     )
     B=&Waveform(
      "Waveform [1]"=::NewEnumerator12
      "[1] Frequency"=15
      "[1] Amplitude Min/Max"=<0 1>
      "[1] Scale"=0.5
      "[1] Bias"=0.35
      "Waveform [2] Blend"=::NewEnumerator5
     )
     Alpha=&FloatFromCurve(
      FloatCurve=DataInterface() {propertyValues {Curve {keys [{interpMode=RCIM_Cubic, tangentMode=RCTM_User, time=0.75, value=0}, {interpMode=RCIM_Cubic, tangentMode=RCTM_User, time=2, value=1}]}}}
      CurveIndex=$System.Age
     )
    )
    Index 0 Param 2=0.15
    Index 0 Param 4=&FloatFromCurve(FloatCurve=DataInterface() {propertyValues {Curve {keys [<at 0 5>,<at 1 2>]}}}, CurveIndex=$System.Age)
    Index 1 Param 1=&Lerp_Float(
     A=80
     B=100
     Alpha=&Waveform("Waveform [1]"=::NewEnumerator12, "[1] Frequency"=25, "[1] Amplitude Min/Max"=<0 1>)
    )
    Index 1 Param 2=&Lerp_Float(
     A=2
     B=-2
     Alpha=&FloatFromCurve(FloatCurve=DataInterface() {propertyValues {Curve {keys [<at 1 0>,<at 2 0>,<at 4 1>]}}}, CurveIndex=$System.Age)
    )
    Module.Param0WriteEnabled=true
    Module.Param0WriteEnabled1=true
    Module.Param1WriteEnabled=true
    Module.Param1WriteEnabled1=true
    Module.Param2WriteEnabled=true
    Module.Param2WriteEnabled1=true
    Module.Param3WriteEnabled=true
    Module.Param3WriteEnabled1=true
    Write Parameter Index 1=true
   )
  ]
  rendererClasses [&NiagaraMeshRendererProperties]
  renderers [{rendererClass=&NiagaraMeshRendererProperties, rendererIndex=0}]
 }
 FN_Mesh {
  emitterUpdateScript [
   &EmitterState(
    Life Cycle Mode=::NewEnumerator1
    Loop Behavior=::NewEnumerator1
    Loop Delay=$System.MainEnginesTurnOnTime
    Loop Duration=1
    Scalability Mode=::NewEnumerator1
    UseLoopDelay=true
   )
   &SpawnBurst_Instantaneous(Spawn Count=1, Use Spawn Probability=false)
  ]
  particleSpawnScript [
   /Niagara/Modules/Spawn/Initialization/InitializeParticle(
    Color=<rgba 35 10 1 0.25>
    Lifetime=&UniformRangedFloat(Minimum=0.5, Maximum=0.5)
    Mesh Scale=&Multiply_VectorByFloat(Vector=<20 20 20>, Float=0.6364)
    Module.Write Color=true
    Module.Write Lifetime=true
    Module.Write Mass=true
    Module.Write Material Random=true
    Module.Write Position=false
    Module.Write Scale=true
    Module.Write SpriteRotation=false
    Module.Write SpriteSize=false
   ) {enabled=false}
   /Niagara/Modules/Spawn/Initialization/V2/InitializeParticle(
    Color=<rgba 1.7429 10.3033 35 0.1>
    Lifetime=0.25
    Mesh Scale Mode=::NewEnumerator3
    Mesh Uniform Scale=10
    Position Offset=<0 50 0>
    UsePositionOffset=true
    Write Material Random=true
   ) {module=InitializeParticle001}
   /SetVariables(
    Particles.MeshOrientationNew:Quat4f=/Niagara/DynamicInputs/Rotation/RandomQuaternion()
    Particles.Scale2:Vector3f=&VectorFromFloat(Value=4)
   ) {isSetParameters=true}
   &DynamicMaterialParameters(
    Index 0 Param 1=&UniformRangedFloat(Minimum=0, Maximum=10)
    Index 0 Param 2=&UniformRangedFloat(Minimum=2, Maximum=5)
    Index 0 Param 3=&UniformRangedFloat(Minimum=0.5, Maximum=2)
    Index 0 Param 4=1
    Module.Param0WriteEnabled=true
    Module.Param1WriteEnabled=true
    Module.Param2WriteEnabled=true
    Module.Param3WriteEnabled=true
   )
   /Niagara/Modules/Spawn/Orientation/V2/InitialMeshOrientation(Use Orientation Vector=true, Use Rotation Vector=true) {enabled=false}
  ]
  particleUpdateScript [
   &ParticleState()
   &CameraOffset(Camera Offset Amount=50) {module=CameraOffset001}
   &ScaleColor(
    Scale Alpha=&FloatFromCurve(FloatCurve=DataInterface() {propertyValues {Curve {keys [<at 0.6 1.0113>,<at 1.0011 0.002>]}}})
    Scale RGB=&VectorFromCurve(VectorCurve=DataInterface() {propertyValues {
       XCurve {keys [<at -0.0011 1.0067>,<at 0.9988 0>]}
       YCurve {keys [<at -0.0011 1.0067>,<at 0.9988 0>]}
       ZCurve {keys [<at -0.0011 1.0067>,<at 0.9988 0>]}
      }})
    ScaleA=true
    ScaleRGB=true
   )
   /SetVariables(
    Particles.DynamicMaterialParameter:Vector4f=&MakeFloat4FromVectorAndFloat("Vector (RGB)"=&VectorFromFloat(Value=1), "Float (Alpha)"=1)
    Particles.DynamicMaterialParameter1:Vector4f=&MakeFloat4FromVectorAndFloat(
     "Vector (RGB)"=&VectorFromFloat(Value=1)
     "Float (Alpha)"=&UniformRangedFloat(Minimum=0, Maximum=1)
    )
   ) {isSetParameters=true}
   &ScaleMeshSize(Scale Factor=&ScaleFloatByCurve(
     InitialValue:Vector3f=<1 1 1>
     ScaleCurve:NiagaraDataInterfaceVectorCurve=DataInterface() {propertyValues {XCurve {keys [<at 0 0.25>,<at 0.25 1>]}, YCurve {keys [<at 0 0.25>,<at 0.25 1>]}, ZCurve {keys [<at 0 0.25>,<at 0.25 1>]}}}
     CurveIndex:NiagaraFloat=$Particles.NormalizedAge
    ))
  ]
  rendererClasses [&NiagaraMeshRendererProperties]
  renderers [{rendererClass=&NiagaraMeshRendererProperties, rendererIndex=0}]
 }
 FireElements-Warmups {
  emitterUpdateScript [
   &EmitterState(Life Cycle Mode=::NewEnumerator1, Loop Behavior=::NewEnumerator1, Loop Duration Mode=::NewEnumerator1, UseLoopDelay=true)
   &SpawnBurst_Instantaneous(Use Spawn Probability=false, UseLoopCountLimit=false)
   &SpawnBurst_Instantaneous(Spawn Group=1, Spawn Time=1.5326, Use Spawn Probability=false, UseLoopCountLimit=false) {module=SpawnBurst_Instantaneous001}
   &SpawnBurst_Instantaneous(Spawn Group=3, Spawn Time=1.75, Use Spawn Probability=false, UseLoopCountLimit=false) {module=SpawnBurst_Instantaneous002}
  ]
  particleSpawnScript [
   &InitializeParticle(
    Color=<rgba 5 0.799 3.809 1>
    Lifetime=4
    Mesh Scale Mode=::NewEnumerator3
    Sprite Size Mode=::NewEnumerator3
    Uniform Sprite Size=79.4604
    UsePositionOffset=true
    Write Material Random=true
   )
   /Niagara/Modules/Spawn/Orientation/V3/InitialMeshOrientation(
    Mesh Orientation Mode=::NewEnumerator6
    Orientation Quaternion:Quat4f=&MultiplyQuaternion(
     Quaternion A=&MakeQuatFromAxisAngle(Axis=<1 0 0>, AngleInDegrees=0)
     Quaternion B=&MakeQuatFromAxisAngle(
      Axis=<0 1 0>
      AngleInDegrees=&SelectFloatFromArray(
       Direct Array Index=&SpawnGroup()
       Float Selection Array=DataInterface() {propertyValues {FloatData [0,120,240]}}
      )
     )
    )
    Use Rotation Vector=true
   )
   /SetVariables(Particles.MaterialRandom:NiagaraFloat=&RandomRangeFloat(Minimum=0, Maximum=1)) {isSetParameters=true}
  ]
  particleUpdateScript [
   &ParticleState(Kill Particles When Lifetime Has Elapsed=false, Let Infinitely Lived Particles Die When Emitter Deactivates=true)
   /SetVariables(Particles.FlameMaturity:NiagaraFloat=&FloatFromCurve(FloatCurve=DataInterface() {propertyValues {Curve {keys [
         <at 0 0 RCIM_Cubic>
         {arriveTangent=0.0117, interpMode=RCIM_Cubic, leaveTangent=0.0117, tangentMode=RCTM_User, time=0.5, value=0.05}
         {arriveTangent=1.9, interpMode=RCIM_Cubic, leaveTangent=1.9, time=0.8, value=0.2}
         <at 1 1 RCIM_Cubic>
        ]}}})) {isSetParameters=true}
   /SetVariables() {enabled=false, isSetParameters=true}
   &Color(Color=&ColorFromCurve(
     ColorCurve=DataInterface() {propertyValues {
       AlphaCurve {keys [<at 0 1 RCIM_Cubic>,<at 1 0 RCIM_Cubic>]}
       BlueCurve {keys [<at 0 10 RCIM_Cubic>,<at 0.2346 9.9711>,<at 0.619 9.9711>,<at 0.6984 249.341>,<at 0.9763 4.9855>]}
       GreenCurve {keys [<at 0 6.9538 RCIM_Cubic>,<at 0.2346 0>,<at 0.619 0>,<at 0.6984 22.25>,<at 0.9763 0>]}
       RedCurve {keys [<at 0 0.9046 RCIM_Cubic>,<at 0.2346 10>,<at 0.619 10>,<at 0.6984 250>,<at 0.9763 5>]}
      }}
     CurveIndex=$Particles.FlameMaturity
    )) {enabled=false}
   &Color(Color=$System.EnginesColorOveLife) {module=Color001}
   &ScaleMeshSize(Scale Factor=&MakeVector(
     X=1
     Y=&Multiply_Float(
      A=&FloatFromCurve(
       FloatCurve=DataInterface() {propertyValues {Curve {keys [
           <at 0 1 RCIM_Cubic>
           {arriveTangent=-0.3571, interpMode=RCIM_Cubic, leaveTangent=-0.3571, time=0.4, value=1.125}
           {arriveTangent=4.19, interpMode=RCIM_Cubic, leaveTangent=4.19, time=0.7, value=0.75}
           {arriveTangent=4.1667, interpMode=RCIM_Cubic, leaveTangent=4.1667, time=0.9, value=3.22}
           <at 1 2>
          ]}}}
       CurveIndex=$Particles.FlameMaturity
      )
      B=&FloatFromCurve(FloatCurve=DataInterface() {propertyValues {Curve {keys [<at 7.2 1>,<at 7.5 1.5>]}}}, CurveIndex=$System.Age)
     )
     Z=1
    ))
   &ScaleMeshSize(
    Initial Mesh Scale:Vector3f=$Particles.Scale
    Scale Factor=&MakeVector(
     X=1
     Y=&FloatFromCurve(FloatCurve=DataInterface() {propertyValues {Curve {keys [<at 7.6 1 RCIM_Cubic>,<at 7.9 2>]}}})
     Z=1
    )
   ) {enabled=false, module=ScaleMeshSize001}
   &SpriteFacingAndAlignment(Sprite Alignment=<0 -1 0>, WriteAlignment=true, WriteFacing=true)
   &DynamicMaterialParameters(
    Index 0 Param 1=&FloatFromCurve(FloatCurve=DataInterface() {propertyValues {Curve {keys [
         <at 0 0 RCIM_Cubic>
         {arriveTangent=0.1566, interpMode=RCIM_Cubic, leaveTangent=0.1566, tangentMode=RCTM_User, time=0.2, value=0.25}
         {arriveTangent=-4.3723, interpMode=RCIM_Cubic, leaveTangent=-4.3723, tangentMode=RCTM_User, time=0.5, value=0.5}
         {arriveTangent=9, interpMode=RCIM_Cubic, leaveTangent=9, time=0.9, value=1.5}
         <at 1 5>
        ]}}})
    Index 0 Param 2=&FloatFromCurve(
     FloatCurve=DataInterface() {propertyValues {Curve {keys [
         <at 0 1 RCIM_Cubic>
         <at 0.7 1.125 RCIM_Cubic>
         {arriveTangent=-3.75, interpMode=RCIM_Cubic, leaveTangent=-3.75, time=0.9, value=1}
         <at 1 0 RCIM_Cubic>
        ]}}}
     CurveIndex=$Particles.FlameMaturity
    )
    Index 0 Param 3=$Particles.FlameMaturity
    Index 0 Param 4=-0.05
    Index 1 Param 1=&FloatFromCurve(
     FloatCurve=DataInterface() {propertyValues {Curve {keys [{arriveTangent=-1.501, interpMode=RCIM_Cubic, time=7.5, value=1}, <at 7.7 0.65 RCIM_Cubic>]}}}
     CurveIndex=$System.Age
    )
    Index 1 Param 2=0.05
    Index 1 Param 3=3
    Module.Param0WriteEnabled=true
    Module.Param0WriteEnabled1=true
    Module.Param1WriteEnabled=true
    Module.Param1WriteEnabled1=true
    Module.Param2WriteEnabled=true
    Module.Param2WriteEnabled1=true
    Module.Param3WriteEnabled=true
    Module.Param3WriteEnabled1=true
    Write Parameter Index 1=true
   )
   &ScaleColor(
    Color Value To Scale=$StackContext.Color
    Scale RGB=&VectorFromFloat(Value=20)
    ScaleA=true
    ScaleRGB=true
   )
  ]
  rendererClasses [&NiagaraMeshRendererProperties]
  renderers [{rendererClass=&NiagaraMeshRendererProperties, rendererIndex=0}]
 }
 GroundSmokeSheet {
  emitterUpdateScript [
   &EmitterState(Life Cycle Mode=::NewEnumerator1, Loop Behavior=::NewEnumerator1, Loop Duration Mode=::NewEnumerator1, UseLoopDelay=false)
   &SpawnBurst_Instantaneous(Spawn Time=3, Use Spawn Probability=false, UseLoopCountLimit=false)
  ]
  particleSpawnScript [&InitializeParticle(
    Color=<rgba 5 0.1607 2.9056 0.371>
    Mesh Scale=<1.5 1 1>
    Mesh Scale Mode=::NewEnumerator4
    Position Offset=<0 -96.672 -160>
    Sprite Size Mode=::NewEnumerator3
    UsePositionOffset=true
    Write Material Random=true
   )]
  particleUpdateScript [
   &ParticleState(Kill Particles When Lifetime Has Elapsed=false, Let Infinitely Lived Particles Die When Emitter Deactivates=true)
   /SetVariables(Particles.Color=&MultiplyLinearColor(
     Linear Color=$System.EnginesColorOveLife
     Scale Factor=&MakeVector4(
      X=1
      Y=1
      Z=1
      W=&FloatFromCurve(
       FloatCurve=DataInterface() {propertyValues {Curve {keys [<at 4 0 RCIM_Cubic>, {arriveTangent=0.25, interpMode=RCIM_Cubic, leaveTangent=0.25, time=6.4, value=0.25}, <at 8 1 RCIM_Cubic>]}}}
       CurveIndex=$System.Age
       Scale Curve=0.2
      )
     )
    )) {isSetParameters=true, module=SetVariables_#15}
   &SolveForcesAndVelocity(Clamp Velocity=false, Limit Acceleration=false)
   &DynamicMaterialParameters(
    Index 0 Param 1=800
    Index 0 Param 2=&Multiply_Float(
     A=&Waveform("Waveform [1]"=::NewEnumerator0, "[1] Frequency"=4, "[1] Amplitude Min/Max"=<0 1>, "[1] Scale"=1.5, "[1] Bias"=0.75)
     B=20
    )
    Module.Param0WriteEnabled=true
    Module.Param1WriteEnabled=true
    Module.Param2WriteEnabled=true
    Module.Param3WriteEnabled=true
   )
  ]
  rendererClasses [&NiagaraMeshRendererProperties]
  renderers [{rendererClass=&NiagaraMeshRendererProperties, rendererIndex=0}]
 }
 Impact_Streaks {
  emitterUpdateScript [
   &EmitterState(
    Life Cycle Mode=::NewEnumerator1
    Loop Behavior=::NewEnumerator1
    Loop Delay=$System.MainEnginesTurnOnTime
    Loop Duration=0.25
    UseLoopDelay=true
   )
   &SpawnBurst_Instantaneous(
    Spawn Count=&RandomRangeInt(Minimum=18, Maximum=36)
    Use Spawn Probability=false
    UseLoopCountLimit=false
   )
  ]
  particleSpawnScript [
   &InitializeParticle(
    Color=<rgba 10 10 10 1>
    Lifetime Max=0.35
    Lifetime Min=0.25
    Lifetime Mode=::NewEnumerator1
    Sprite Size Max=<100 225>
    Sprite Size Min=<25 80>
    Sprite Size Mode=::NewEnumerator2
    UsePositionOffset=false
    Write Material Random=true
   )
   &SubUVAnimation(End Frame=3, SubUV Animation Mode:ENiagara_SubUVLookupMode=::NewEnumerator2)
   &ShapeLocation(
    Cylinder Distribution=::NewEnumerator2
    Cylinder Height=0
    Cylinder Radius=20
    MaskSpawnGroup=false
    Offset=<0 50 50>
    Offset Mode=::NewEnumerator0
    Rotation Axis=<0 1 0>
    Rotation Mode=::NewEnumerator1
    Shape Primitive=::NewEnumerator1
    U Position=&ReturnNormalizedExecIndex()
    UseOwnerScale=true
   )
   &AddVelocity(
    Velocity=&Add_Vector(
     A=&Add_Vector(A=$Particles.ShapeLocation.ShapeNormal, B=&RandomRangeVector(Minimum=<0 0 0>, Maximum=<1 0 0>))
     B=<0 -1 0>
    )
    Velocity Speed Scale=50
   )
  ]
  particleUpdateScript [
   &SolveForcesAndVelocity(Clamp Velocity=false, Limit Acceleration=false)
   &ParticleState()
   &ScaleColor(
    Scale Alpha=&FloatFromCurve(FloatCurve=DataInterface() {propertyValues {Curve {keys [{arriveTangent=-3.3102, interpMode=RCIM_Cubic, leaveTangent=-3.3102, tangentMode=RCTM_User, time=0, value=1}, <at 1 0 RCIM_Cubic>]}}})
    Scale RGB=&MakeVectorFromLinearColorRGB(Color=&ColorFromCurve(ColorCurve=DataInterface() {propertyValues {
        AlphaCurve {keys []}
        BlueCurve {keys [<at 0 9.1864 RCIM_Cubic>,<at 0.125 1>,<at 1 0.1878 RCIM_Cubic>]}
        GreenCurve {keys [<at 0 0.8091 RCIM_Cubic>,<at 0.125 0.084>,<at 1 0.0313 RCIM_Cubic>]}
        RedCurve {keys [<at 0 10 RCIM_Cubic>,<at 0.125 0.8046>,<at 1 0 RCIM_Cubic>]}
       }}))
    ScaleA=true
    ScaleRGB=true
   )
   &ScaleSpriteSize(
    Non-Uniform Curve Sprite Scale=DataInterface() {propertyValues {
      XCurve {keys [
        {arriveTangent=0.9067, interpMode=RCIM_Cubic, leaveTangent=0.9067, tangentMode=RCTM_User, time=0, value=0.8}
        {arriveTangentWeight=0.2034, interpMode=RCIM_Cubic, leaveTangentWeight=0.2034, tangentMode=RCTM_User, time=0.5, value=1}
        {arriveTangent=-3.1509, interpMode=RCIM_Cubic, leaveTangent=-3.1509, tangentMode=RCTM_User, time=1, value=0.1}
       ]}
      YCurve {keys [
        {arriveTangent=3.7985, interpMode=RCIM_Cubic, leaveTangent=3.7985, tangentMode=RCTM_User, time=0, value=0}
        {arriveTangent=2, interpMode=RCIM_Cubic, leaveTangent=2, time=0.5, value=1.6}
        <at 1 2 RCIM_Cubic>
       ]}
     }}
    Scale Sprite Size Mode=::NewEnumerator3
   )
  ]
  rendererClasses [&NiagaraSpriteRendererProperties]
  renderers [{rendererClass=&NiagaraSpriteRendererProperties}]
 }
 InnerNoiseSphere {
  emitterUpdateScript [
   &EmitterState(
    Life Cycle Mode=::NewEnumerator1
    Loop Behavior=::NewEnumerator1
    Loop Delay=$System.MainEnginesTurnOnTime
    Loop Duration=0.2
    UseLoopDelay=true
   )
   &SpawnBurst_Instantaneous(Spawn Count=1, Use Spawn Probability=false)
  ]
  particleSpawnScript [&InitializeParticle(
    Color=<rgba 4 1.0055 3.9041 3>
    Lifetime=0.3
    Mesh Scale Mode=::NewEnumerator3
    Mesh Uniform Scale=7
    UsePositionOffset=false
    Write Material Random=true
   )]
  particleUpdateScript [
   &ParticleState()
   &ScaleColor(
    Scale Alpha=&FloatFromCurve(FloatCurve=DataInterface() {propertyValues {Curve {keys [<at 0 1>, {interpMode=RCIM_Cubic, tangentMode=RCTM_User, time=0.1, value=1}, {interpMode=RCIM_Cubic, tangentMode=RCTM_User, time=1, value=0}]}}})
    ScaleA=true
    ScaleRGB=true
   )
   &DynamicMaterialParameters(
    Index 0 Param 1=1
    Index 0 Param 3=1
    Index 0 Param 4=1
    Module.Param0WriteEnabled=true
    Module.Param1WriteEnabled=true
    Module.Param2WriteEnabled=true
    Module.Param3WriteEnabled=true
   )
   &ScaleMeshSize(Scale Factor=&VectorFromFloat(Value=&FloatFromCurve(FloatCurve=DataInterface() {propertyValues {Curve {keys [
          {arriveTangent=4.1785, interpMode=RCIM_Cubic, leaveTangent=4.1785, tangentMode=RCTM_User, time=0, value=0.05}
          {arriveTangent=0.9995, interpMode=RCIM_Cubic, leaveTangent=0.9995, tangentMode=RCTM_User, time=0.3, value=0.6}
          {arriveTangent=-0.0405, interpMode=RCIM_Cubic, leaveTangent=-0.0405, tangentMode=RCTM_User, time=1, value=1}
         ]}}})))
   &CameraOffset(Camera Offset Amount=250)
  ]
  rendererClasses [&NiagaraMeshRendererProperties]
  renderers [{rendererClass=&NiagaraMeshRendererProperties, rendererIndex=0}]
 }
 MainEngine_LightingFX {
  emitterSpawnScript [
   /SetVariables(Emitter.Distortion_Input=1, Emitter.LightConeOverallSize=2.2) {isSetParameters=true, module=SetVariables_#9}
   /SetVariables(Emitter.FlareSizeScalar=0.14) {isSetParameters=true, module=SetVariables_#10}
  ]
  emitterUpdateScript [
   &EmitterState(
    Life Cycle Mode=::NewEnumerator1
    Loop Behavior=::NewEnumerator1
    Loop Delay=$System.MainEnginesTurnOnTime
    Loop Duration Mode=::NewEnumerator1
    UseLoopDelay=true
   )
   &SpawnBurst_Instantaneous(Spawn Count=1, Use Spawn Probability=false)
  ]
  particleSpawnScript [
   &InitializeParticle(
    Color=<rgba 0.955 0.2919 1 1>
    Lifetime=5
    Sprite Size=&Multiply_Vector2D_ByFloat(Float=$Emitter.FlareSizeScalar, Vector2D=<30000 10000>)
    Sprite Size Mode=::NewEnumerator4
    UsePositionOffset=false
    Write Material Random=true
   )
   &CameraOffset(Camera Offset Amount=200) {enabled=false}
   &SystemLocation(MaskSpawnGroup=false, Offset=<0 0 50>)
   /SetVariables(Particles.LightPos:NiagaraPosition=<0 -500 0>) {isSetParameters=true, module=SetVariables_#11}
  ]
  particleUpdateScript [
   /SetVariables(
    Particles.LensColor:LinearColor=&MultiplyLinearColor(
     Linear Color=<rgba 0.1086 0.1286 0.2 1>
     Scale Factor=&MakeVector4_FromFloat(Float=&Multiply_Float(A=$User.OverallOpacity, B=$System.SputterFloat))
    )
    Particles.LensSize:Vector2f=&Multiply_Vector2D_ByFloat(Float=$Emitter.FlareSizeScalar, Vector2D=<20000 4000>)
   ) {isSetParameters=true, module=SetVariables_#12}
   /SetVariables(
    Particles.DarkRadialColor:LinearColor=&MultiplyLinearColor(
     Linear Color=<rgba 0.0902 0.0318 0.1615 0.5>
     Scale Factor=&MakeVector4_FromFloat(Float=&Multiply_Float(A=$User.OverallOpacity, B=$System.SputterFloat))
    )
    Particles.DarkRadialSize:Vector2f=&Multiply_Vector2D_ByFloat(Float=$Emitter.FlareSizeScalar, Vector2D=<15000 10000>)
   ) {isSetParameters=true, module=SetVariables_#13}
   /SetVariables(Particles.LightColor=&MultiplyLinearColor(
     Linear Color=<rgba 0.4496 0.0786 1 1>
     Scale Factor=&MakeVector4_FromFloat(Float=&Multiply_Float(A=8, B=&Multiply_Float(A=$User.OverallOpacity, B=$System.SputterFloat)))
    )) {isSetParameters=true, module=SetVariables_#14}
   &ParticleState(Kill Particles When Lifetime Has Elapsed=false, Let Infinitely Lived Particles Die When Emitter Deactivates=true)
   &ScaleSpriteSize(Scale Factor=&Vector2DFromFloat(Value=&Lerp_Float(
      A=1
      B=&RandomRangeFloat(Minimum=0.8, Maximum=1, Evaluation Type=::NewEnumerator1)
      Alpha=$User.FlareFlicker
     )))
   &ScaleColor(Scale Alpha=&Multiply_Float(A=$User.OverallOpacity, B=$System.SputterFloat), ScaleA=true, ScaleRGB=true)
   &CameraOffset(Camera Offset Amount=350) {module=CameraOffset001}
  ]
  rendererClasses [&NiagaraSpriteRendererProperties, &NiagaraLightRendererProperties]
  renderers [
   {rendererClass=&NiagaraSpriteRendererProperties}
   {rendererClass=&NiagaraSpriteRendererProperties, rendererIndex=1}
   {rendererClass=&NiagaraSpriteRendererProperties, rendererIndex=2}
   {rendererClass=&NiagaraLightRendererProperties, rendererIndex=3}
  ]
 }
 PlayAudioOnSpawnBurst {
  emitterSpawnScript [/SetVariables(Emitter.AudioPlayer=DataInterface() {propertyValues {
      Attenuation=null
      Concurrency=null
      MaxPlaysPerTick=10
      SoundToPlay=/Game/Athena/Sounds/Vehicles/Bus/Athena_Bus_Warp_Thruster_Activate_B_Cue
      bAllowLoopingOneShotSounds=false
      bLimitPlaysPerTick=true
      bOnlyActiveDuringGameplay=false
      bStopWhenComponentIsDestroyed=true
      ConfigurationUserParameter {parameter {name=None, typeDefHandle {registeredTypeIndex=1}, varData []}}
      ParameterNames []
     }}) {isSetParameters=true}]
  emitterUpdateScript [
   &EmitterState()
   &SpawnBurst_Instantaneous(Spawn Time=3.5083, Use Spawn Probability=false, UseLoopCountLimit=false)
  ]
  particleSpawnScript [
   &InitializeParticle(Lifetime=6, UsePositionOffset=false, Write Material Random=true)
   /Niagara/Modules/Audio/PlayPersistentAudio(
    Audio Player:NiagaraDataInterfaceAudioPlayer=$Emitter.AudioPlayer
    Play Audio:NiagaraBool=&SetBoolByIntComparison(
     A=&ReturnParticleIDAttributes(Particle ID Attribute=::NewEnumerator1)
     B=0
     Comparison Type=::NewEnumerator4
    )
   )
  ]
  particleUpdateScript [
   &SolveForcesAndVelocity(Clamp Velocity=false, Limit Acceleration=false) {enabled=false}
   /Niagara/Modules/Audio/UpdatePersistentAudio(Audio Player:NiagaraDataInterfaceAudioPlayer=$Emitter.AudioPlayer)
  ]
  rendererClasses []
  renderers []
 }
 Shockwave {
  emitterUpdateScript [
   &EmitterScalability(
    Module.bApplySpawnCountScale=false
    Module.bApplySpawnCountScaleByDistanceFraction=false
    Module.bUseMaxDistance=false
    Module.bUseMinDistance=false
   ) {enabled=false}
   &EmitterLifeCycle(NextLoopDuration=4.4208) {enabled=false}
   &EmitterState(
    Life Cycle Mode=::NewEnumerator1
    Loop Behavior=::NewEnumerator1
    Loop Delay=$System.MainEnginesTurnOnTime
    Loop Duration=1
    UseLoopDelay=true
   )
   &SpawnBurst_Instantaneous(Spawn Count=1, Use Spawn Probability=false)
  ]
  particleSpawnScript [
   /SetVariables() {isSetParameters=true}
   /Niagara/Modules/Spawn/Initialization/InitializeParticle(
    Color=<rgba 1835.5947 474.7406 5000 0.5>
    Lifetime=&UniformRangedFloat(Minimum=0.2, Maximum=0.2)
    Mesh Scale=&Multiply_Vector(
     A=&UniformRangedVector(Minimum=<3 3 3>, Maximum=<3 3 3>)
     B=&VectorFromFloat(Value=1.25)
    )
    Module.Write Color=true
    Module.Write Lifetime=true
    Module.Write Mass=true
    Module.Write Material Random=true
    Module.Write Position=false
    Module.Write Scale=true
    Module.Write SpriteRotation=false
    Module.Write SpriteSize=true
    Position=&SimulationPosition()
    Sprite Size=<1024 1024>
   )
   /Niagara/Modules/Spawn/Location/SphereLocation(MaskSpawnGroup=false, Module.Override Randomness=false, Offset=<20 0 0>, Sphere Radius=0.01, Surface Only=false) {enabled=false}
   /SetVariables(
    Particles.NewLightRadious=250
    Particles.NewLinearColor:LinearColor=<rgba 25 5 0 1>
    Particles.NewVector:Vector3f=&Add_Vector(A=/Niagara/DynamicInputs/Helpers/SimulationPosition(), B=<0 0 500>)
   ) {isSetParameters=true}
  ]
  particleUpdateScript [
   &ParticleState()
   &Color(Color=<rgba 5000 0 2160.8997 0.1>)
   &DynamicMaterialParameters(
    Index 0 Param 1=&FloatFromCurve(FloatCurve=DataInterface() {propertyValues {Curve {keys [
         {arriveTangent=2.0031, interpMode=RCIM_Cubic, leaveTangent=2.0031, tangentMode=RCTM_User, time=-0.0072, value=-0.0078}
         {arriveTangent=1.9873, interpMode=RCIM_Cubic, leaveTangent=1.9873, tangentMode=RCTM_User, time=0.6336, value=1.3072}
         <at 1.0021 1.9784>
         <at 1.61 1.9784>
        ]}}})
    Index 0 Param 2=&FloatFromCurve(FloatCurve=DataInterface() {propertyValues {Curve {keys [<at -0.0005 15>,<at 0.1612 13.3874>,<at 1.0021 5>,<at 1.61 5>]}}})
    Index 0 Param 3=-1
    Module.Param0WriteEnabled=true
    Module.Param1WriteEnabled=true
    Module.Param2WriteEnabled=true
    Module.Param3WriteEnabled=true
   )
   &ScaleMeshSize(Scale Factor=&VectorFromCurve(VectorCurve=DataInterface() {propertyValues {
       XCurve {keys [
         {arriveTangent=9.7102, interpMode=RCIM_Cubic, leaveTangent=9.7102, tangentMode=RCTM_User, time=0, value=2}
         {arriveTangent=6.8631, interpMode=RCIM_Cubic, leaveTangent=6.8631, tangentMode=RCTM_User, time=0.3596, value=6.4332}
         <at 0.9996 7.8749>
         <at 1.61 7.8749>
        ]}
       YCurve {keys [
         {arriveTangent=10.0711, interpMode=RCIM_Cubic, leaveTangent=10.0711, tangentMode=RCTM_User, time=0, value=2}
         {arriveTangent=6.8631, interpMode=RCIM_Cubic, leaveTangent=6.8631, tangentMode=RCTM_User, time=0.3582, value=6.4808}
         <at 0.9996 7.8749>
         <at 1.61 7.8749>
        ]}
       ZCurve {keys [
         {arriveTangent=10.3854, interpMode=RCIM_Cubic, leaveTangent=10.3853, tangentMode=RCTM_User, time=0, value=2}
         {arriveTangent=6.8631, interpMode=RCIM_Cubic, leaveTangent=6.8631, tangentMode=RCTM_User, time=0.3615, value=6.5528}
         <at 0.9996 7.8749>
         <at 1.61 7.8749>
        ]}
      }}))
   &ScaleColor(
    Scale Alpha=&FloatFromCurve(
     FloatCurve=DataInterface() {propertyValues {Curve {keys [
         {interpMode=RCIM_Cubic, tangentMode=RCTM_User, time=-0.0034, value=0.1203}
         {arriveTangent=-0.1606, interpMode=RCIM_Cubic, leaveTangent=-0.1606, tangentMode=RCTM_User, time=0.3345, value=0.0495}
         {interpMode=RCIM_Cubic, tangentMode=RCTM_User, time=0.9972, value=-0.0001}
        ]}}}
     Scale Curve=0.85
    )
    Scale RGB=&ScaleFloatByCurve(
     InitialValue:Vector3f=<2.5 2.5 2.5>
     ScaleCurve:NiagaraDataInterfaceVectorCurve=DataInterface() {propertyValues {
       XCurve {keys [<at 0.0013 1.998>,<at 1 1>]}
       YCurve {keys [<at 0.0012 1.4636>,<at 1.0017 0.264>]}
       ZCurve {keys [<at 0.0012 1.0048>,<at 1.0017 0.264>]}
      }}
     CurveIndex:NiagaraFloat=$Particles.NormalizedAge
    )
    ScaleA=true
    ScaleRGB=true
   )
   /SetVariables(Particles.FlareColor=&ColorFromCurve(ColorCurve=DataInterface() {propertyValues {
       AlphaCurve {keys [<at 0 1 RCIM_Cubic>,<at 1 0 RCIM_Cubic>]}
       BlueCurve {keys [<at 0 7.3344 RCIM_Cubic>]}
       GreenCurve {keys [<at 0 0 RCIM_Cubic>]}
       RedCurve {keys [<at 0 8 RCIM_Cubic>]}
      }})) {isSetParameters=true}
   /SetVariables(
    Particles.NewLightRadious:NiagaraFloat=&FloatFromCurve(FloatCurve=DataInterface() {propertyValues {Curve {keys [
         <at -0.0284 132.9428>
         {arriveTangent=65.2061, interpMode=RCIM_Cubic, leaveTangent=65.2061, tangentMode=RCTM_User, time=0.2563, value=219.2211}
         <at 1 200>
        ]}}})
    Particles.NewLinearColor:LinearColor=&MakeLinearColorFromVectorAndFloat(
     "Vector (RGB)"=&VectorFromCurve(VectorCurve=DataInterface() {propertyValues {XCurve {keys [<at 0.0062 50>,<at 0.75 0>]}, YCurve {keys [<at 0 10>,<at 0.75 0>]}, ZCurve {keys [<at 0 1>,<at 0.75 0>]}}})
     "Float (Alpha)"=1
    )
   ) {isSetParameters=true}
   &CameraOffset(Camera Offset Amount=250)
  ]
  rendererClasses [
   &NiagaraLightRendererProperties
   &NiagaraSpriteRendererProperties
   &NiagaraMeshRendererProperties
  ]
  renderers [
   {rendererClass=&NiagaraLightRendererProperties, rendererIndex=0}
   {rendererClass=&NiagaraSpriteRendererProperties, rendererIndex=1}
   {rendererClass=&NiagaraMeshRendererProperties, rendererIndex=2}
  ]
 }
 SmokeLicks {
  emitterUpdateScript [
   &EmitterState(Life Cycle Mode=::NewEnumerator1, Loop Behavior=::NewEnumerator1, Loop Duration Mode=::NewEnumerator1, UseLoopDelay=false)
   &SpawnBurst_Instantaneous(Spawn Count=2, Spawn Time=3, Use Spawn Probability=false, UseLoopCountLimit=false)
  ]
  particleSpawnScript [/Niagara/Modules/Spawn/Initialization/V2/InitializeParticle(
    Color=<rgba 5 0.1607 2.9056 0.35>
    Mesh Scale=&Multiply_Vector(
     A=<1 2 0.5>
     B=&MakeVector(
      X=&SelectFloatFromArray(
       Direct Array Index=&ReturnParticleIDAttributes(Particle ID Attribute=::NewEnumerator1)
       Float Selection Array=DataInterface() {propertyValues {FloatData [1,-1]}}
      )
      Y=1
      Z=1
     )
    )
    Mesh Scale Mode=::NewEnumerator4
    Position Offset=<0 -200 -52.453>
    Sprite Size Mode=::NewEnumerator3
    UsePositionOffset=true
    Write Material Random=true
   )]
  particleUpdateScript [
   &ParticleState(Kill Particles When Lifetime Has Elapsed=false, Let Infinitely Lived Particles Die When Emitter Deactivates=true)
   /SetVariables(Particles.Color=&MultiplyLinearColor(
     Linear Color=$System.EnginesColorOveLife
     Scale Factor=&MakeVector4(
      X=1
      Y=1
      Z=1
      W=&FloatFromCurve(
       FloatCurve=DataInterface() {propertyValues {Curve {keys [<at 4 0 RCIM_Cubic>, {arriveTangent=0.25, interpMode=RCIM_Cubic, leaveTangent=0.25, time=6.4, value=0.25}, <at 8 1 RCIM_Cubic>]}}}
       CurveIndex=$System.Age
       Scale Curve=0.1
      )
     )
    )) {isSetParameters=true, module=SetVariables_#15}
   &SolveForcesAndVelocity(Clamp Velocity=false, Limit Acceleration=false)
   &DynamicMaterialParameters(
    Index 0 Param 1=646.6827
    Index 0 Param 2=&Multiply_Float(
     A=&Waveform("Waveform [1]"=::NewEnumerator0, "[1] Frequency"=4, "[1] Amplitude Min/Max"=<0 1>, "[1] Scale"=1.5, "[1] Bias"=0.75)
     B=20
    )
    Module.Param0WriteEnabled=true
    Module.Param1WriteEnabled=true
    Module.Param2WriteEnabled=true
    Module.Param3WriteEnabled=true
   )
  ]
  rendererClasses [&NiagaraMeshRendererProperties]
  renderers [{rendererClass=&NiagaraMeshRendererProperties, rendererIndex=0}]
 }
 SmokeWorld {
  emitterUpdateScript [
   &EmitterState(
    Life Cycle Mode=::NewEnumerator1
    Loop Behavior=::NewEnumerator1
    Loop Delay=$System.MainEnginesTurnOnTime
    Loop Duration=1
    UseLoopDelay=true
   )
   &SpawnBurst_Instantaneous(Spawn Count=5, Use Spawn Probability=false)
  ]
  particleSpawnScript [
   &InitializeParticle(
    Color=<rgba 0 1 0 1>
    Lifetime Max=0.75
    Lifetime Min=0.5
    Lifetime Mode=::NewEnumerator1
    Position Offset=<0 0 5>
    Sprite Rotation Mode=::NewEnumerator2
    Sprite Size Mode=::NewEnumerator1
    Uniform Sprite Size Max=300
    Uniform Sprite Size Min=180
    UsePositionOffset=true
    Write Material Random=true
   )
   &ShapeLocation(
    Cylinder Distribution=::NewEnumerator2
    Cylinder Height=0
    Cylinder Radius=2
    MaskSpawnGroup=false
    Rotation Axis=<0 1 0>
    Rotation Mode=::NewEnumerator1
    Shape Primitive=::NewEnumerator1
    U Position=&ReturnNormalizedExecIndex(Normalized Index Scale=-0.5)
    UseOwnerScale=true
   )
   &AddVelocity(
    Velocity=$Particles.ShapeLocation.ShapeNormal
    Velocity Speed Scale=&RandomRangeFloat(Minimum=1400, Maximum=1900)
   )
   /SetVariables(Particles.RandomFloat=&RandomRangeFloat(Minimum=0, Maximum=1)) {isSetParameters=true}
  ]
  particleUpdateScript [
   &ParticleState()
   &SpriteRotationRate(Rotation Rate=&Multiply_Float(
     A=1
     B=&FloatFromCurve(
      FloatCurve=DataInterface() {propertyValues {Curve {keys [
          {arriveTangent=-4.0107, interpMode=RCIM_Cubic, leaveTangent=-4.0107, tangentMode=RCTM_User, time=-0.0046, value=1.0113}
          {arriveTangent=-0.2496, interpMode=RCIM_Cubic, leaveTangent=-0.2496, tangentMode=RCTM_User, time=0.5, value=0.16}
         ]}}}
      CurveIndex=$Emitter.Age
      Scale Curve=150
     )
    ))
   &Drag(
    Drag=&FloatFromCurve(
     FloatCurve=DataInterface() {propertyValues {Curve {keys [<at 0 2 RCIM_Cubic>,<at 0.5 1 RCIM_Cubic>]}}}
     Scale Curve=&RandomRangeFloat(Minimum=4, Maximum=4)
    )
    Use Linear Drag=true
    Use Rotational Drag=false
   )
   &AccelerationForce(Acceleration=&Multiply_VectorByFloat(Vector=<0 10000 0>, Float=&RandomRangeFloat(Minimum=0.3, Maximum=0.5)))
   &SolveForcesAndVelocity(Clamp Velocity=false, Limit Acceleration=false)
   &ScaleSpriteSize(Scale Factor=&Vector2DFromFloat(Value=&FloatFromCurve(FloatCurve=DataInterface() {propertyValues {Curve {keys [{arriveTangent=1.7458, interpMode=RCIM_Cubic, leaveTangent=1.7458, tangentMode=RCTM_User, time=0, value=0.5}, <at 1 1.25 RCIM_Cubic>]}}})))
   &DynamicMaterialParameters(
    Index 0 Param 1=&Add_Float(
     A=$Particles.RandomFloat
     B=&FloatFromCurve(FloatCurve=DataInterface() {propertyValues {Curve {keys [
          {arriveTangent=0.9088, interpMode=RCIM_Cubic, leaveTangent=0.9088, tangentMode=RCTM_User, time=0, value=1}
          {arriveTangent=0.5473, interpMode=RCIM_Cubic, leaveTangent=0.5473, tangentMode=RCTM_User, time=1, value=2}
         ]}}})
    )
    Index 0 Param 3=&FloatFromCurve(
     FloatCurve=DataInterface() {propertyValues {Curve {keys [<at 0 1 RCIM_Cubic>, {arriveTangent=-1.3333, interpMode=RCIM_Cubic, leaveTangent=-1.3333, time=0.3, value=0.2}, <at 0.75 0 RCIM_Cubic>]}}}
     Scale Curve=5
    )
    Index 1 Param 1=&FloatFromCurve(
     FloatCurve=DataInterface() {propertyValues {Curve {keys [<at 0 0>,<at 0.25 1>]}}}
     Scale Curve=&RandomRangeFloat(Minimum=0, Maximum=1)
    )
    Index 1 Param 3=4.4703
    Index 2 Param 1=&FloatFromCurve(
     FloatCurve=DataInterface() {propertyValues {Curve {keys [{interpMode=RCIM_Cubic, leaveTangent=3.2309, tangentMode=RCTM_Break, time=0.1, value=0}, <at 1 1>]}}}
     Scale Curve=3
    )
    Index 2 Param 2=&FloatFromCurve(
     FloatCurve=DataInterface() {propertyValues {Curve {keys [<at 0 0.2 RCIM_Cubic>,<at 1 0.8 RCIM_Cubic>]}}}
     Scale Curve=&RandomRangeFloat(Minimum=0.2, Maximum=1.2)
    )
    Index 2 Param 3=&FloatFromCurve(FloatCurve=DataInterface() {propertyValues {Curve {keys [
         {interpMode=RCIM_Cubic, tangentMode=RCTM_Break, time=0, value=1}
         {interpMode=RCIM_Cubic, leaveTangent=11.1233, tangentMode=RCTM_Break, time=1, value=2}
        ]}}})
    Index 2 Param 4=&FloatFromCurve(FloatCurve=DataInterface() {propertyValues {Curve {keys [<at 0 1 RCIM_Cubic>]}}}, Scale Curve=0.75)
    Module.Param0WriteEnabled=true
    Module.Param0WriteEnabled1=true
    Module.Param0WriteEnabled2=true
    Module.Param1WriteEnabled=false
    Module.Param1WriteEnabled1=true
    Module.Param1WriteEnabled2=true
    Module.Param2WriteEnabled=true
    Module.Param2WriteEnabled1=true
    Module.Param2WriteEnabled2=true
    Module.Param3WriteEnabled=true
    Module.Param3WriteEnabled1=false
    Module.Param3WriteEnabled2=true
    Write Parameter Index 1=true
    Write Parameter Index 2=true
   )
   &Color(Color=&ColorFromCurve(ColorCurve=DataInterface() {propertyValues {
       AlphaCurve {keys [<at 0 1 RCIM_Cubic>,<at 1 0 RCIM_Cubic>]}
       BlueCurve {keys [<at 0 20 RCIM_Cubic>,<at 0.5389 0.4149 RCIM_Cubic>]}
       GreenCurve {keys [<at 0 14.0625 RCIM_Cubic>,<at 0.5389 0.4149 RCIM_Cubic>]}
       RedCurve {keys [<at 0 4.9812 RCIM_Cubic>,<at 0.5389 0.4149 RCIM_Cubic>]}
      }}))
   &CameraOffset(Camera Offset Amount=&RandomRangeFloat(Minimum=30, Maximum=50))
  ]
  rendererClasses [&NiagaraSpriteRendererProperties]
  renderers [
   {rendererClass=&NiagaraSpriteRendererProperties}
   {rendererClass=&NiagaraSpriteRendererProperties, rendererIndex=1}
  ]
 }
 StartupAirglow {
  emitterUpdateScript [
   &EmitterState(Life Cycle Mode=::NewEnumerator1, Loop Behavior=::NewEnumerator1, Loop Duration=2, UseLoopDelay=false)
   &SpawnBurst_Instantaneous(Use Spawn Probability=false, UseLoopCountLimit=false)
  ]
  particleSpawnScript [&InitializeParticle(Sprite Size Mode=::NewEnumerator3, Uniform Sprite Size=750, UsePositionOffset=false, Write Material Random=true)]
  particleUpdateScript [
   &ParticleState()
   &SolveForcesAndVelocity(Clamp Velocity=false, Limit Acceleration=false)
   &Color(Color=&ColorFromCurve(ColorCurve=DataInterface() {propertyValues {
       AlphaCurve {keys [<at 0 1 RCIM_Cubic>]}
       BlueCurve {keys [<at 0 1 RCIM_Cubic>,<at 0.0459 1>,<at 0.6611 0.7954 RCIM_Cubic>]}
       GreenCurve {keys [<at 0 1 RCIM_Cubic>,<at 0.0459 0.6753>,<at 0.6611 0 RCIM_Cubic>]}
       RedCurve {keys [<at 0 1 RCIM_Cubic>,<at 0.0459 0.0148>,<at 0.6611 1 RCIM_Cubic>]}
      }}))
   &ScaleColor(
    Color Value To Scale=$StackContext.Color
    Scale Alpha=&FloatFromCurve(FloatCurve=DataInterface() {propertyValues {Curve {keys [<at -0.0046 1.0113 RCIM_Cubic>,<at 1.0011 0.002 RCIM_Cubic>]}}}, Scale Curve=0.1)
    Scale RGB=&VectorFromFloat(Value=&FloatFromCurve(
      FloatCurve=DataInterface() {propertyValues {Curve {keys [
          {arriveTangent=-1.025, interpMode=RCIM_Cubic, time=0, value=0}
          {arriveTangent=0.3798, interpMode=RCIM_Cubic, leaveTangent=0.3798, time=0.068, value=1.025}
          <at 0.9769 0.371>
         ]}}}
      Scale Curve=10
     ))
    ScaleA=true
    ScaleRGB=true
   )
   &ScaleSpriteSize(Scale Factor=&Vector2DFromFloat(Value=&Lerp_Float(A=1, B=&RandomRangeFloat(Minimum=0.8, Maximum=1, Evaluation Type=::NewEnumerator1), Alpha=0.125)))
   &CameraOffset(Camera Offset Amount=500)
  ]
  rendererClasses [&NiagaraSpriteRendererProperties]
  renderers [{rendererClass=&NiagaraSpriteRendererProperties}]
 }
 StartupBurst {
  emitterSpawnScript [
   /SetVariables(Emitter.Distortion_Input=1, Emitter.LightConeOverallSize=2.2) {isSetParameters=true, module=SetVariables_#9}
   /SetVariables(Emitter.FlareSizeScalar=0.14) {isSetParameters=true, module=SetVariables_#10}
  ]
  emitterUpdateScript [
   &EmitterState(Life Cycle Mode=::NewEnumerator1, Loop Behavior=::NewEnumerator1, Loop Duration Mode=::NewEnumerator1, UseLoopDelay=false)
   &SpawnBurst_Instantaneous(Spawn Count=1, Use Spawn Probability=false)
  ]
  particleSpawnScript [
   &InitializeParticle(
    Color=<rgba 0.0364 0.0844 0.7812 1>
    Sprite Size=&Multiply_Vector2D_ByFloat(Float=$Emitter.FlareSizeScalar, Vector2D=&Multiply_Vector2D_ByFloat(Float=18500, Vector2D=<1 0.5>))
    Sprite Size Mode=::NewEnumerator4
    UsePositionOffset=false
    Write Material Random=true
   )
   &CameraOffset(Camera Offset Amount=500)
   &SystemLocation(MaskSpawnGroup=false, Offset=<0 0 25.0808>)
   /SetVariables(Particles.LightPos:NiagaraPosition=<0 -500 0>) {enabled=false, isSetParameters=true, module=SetVariables_#11}
  ]
  particleUpdateScript [
   /SetVariables(
    Particles.LensColor:LinearColor=&MultiplyLinearColor(
     Linear Color=<rgba 0.1086 0.1286 0.2 1>
     Scale Factor=&MakeVector4_FromFloat(Float=&Multiply_Float(A=$User.OverallOpacity, B=$System.SputterFloat))
    )
    Particles.LensSize:Vector2f=&Multiply_Vector2D_ByFloat(Float=$Emitter.FlareSizeScalar, Vector2D=<20000 4000>)
   ) {enabled=false, isSetParameters=true, module=SetVariables_#12}
   /SetVariables(
    Particles.DarkRadialColor:LinearColor=&MultiplyLinearColor(
     Linear Color=<rgba 0.0902 0.0318 0.1615 0.5>
     Scale Factor=&MakeVector4_FromFloat(Float=&Multiply_Float(A=$User.OverallOpacity, B=$System.SputterFloat))
    )
    Particles.DarkRadialSize:Vector2f=&Multiply_Vector2D_ByFloat(Float=$Emitter.FlareSizeScalar, Vector2D=<15000 10000>)
   ) {enabled=false, isSetParameters=true, module=SetVariables_#13}
   /SetVariables(Particles.LightColor=&MultiplyLinearColor(
     Linear Color=<rgba 0.1379 0.6405 1 1>
     Scale Factor=&MakeVector4_FromFloat(Float=&FloatFromCurve(FloatCurve=DataInterface() {propertyValues {Curve {keys [<at 0 0.2>,<at 7 0.2>,<at 8 1.5>]}}}, CurveIndex=$System.Age, Scale Curve=10))
    )) {isSetParameters=true, module=SetVariables_#14}
   &ParticleState(Kill Particles When Lifetime Has Elapsed=false, Let Infinitely Lived Particles Die When Emitter Deactivates=true)
   &ScaleSpriteSize(Scale Factor=&Vector2DFromFloat(Value=&FloatFromCurve(FloatCurve=DataInterface() {propertyValues {Curve {keys [
          {arriveTangent=0.225, interpMode=RCIM_Cubic, time=0, value=0.025}
          {arriveTangent=1.2, interpMode=RCIM_Cubic, leaveTangent=1.2, time=0.1, value=0.575}
          <at 1 1.225 RCIM_Cubic>
         ]}}})))
   &ScaleColor(
    Scale Alpha=&FloatFromCurve(FloatCurve=DataInterface() {propertyValues {Curve {keys [
         <at 0 1 RCIM_Cubic>
         {arriveTangent=-2.5, interpMode=RCIM_Cubic, leaveTangent=-2.5, time=0.1, value=1.25}
         {arriveTangent=-1.3889, interpMode=RCIM_Cubic, leaveTangent=-1.3889, time=0.3, value=0.25}
         <at 1 0 RCIM_Cubic>
        ]}}})
    Scale RGB=&VectorFromFloat(Value=&FloatFromCurve(FloatCurve=DataInterface() {propertyValues {Curve {keys [
          {arriveTangent=47.5885, interpMode=RCIM_Cubic, time=0, value=0}
          {arriveTangent=0.7143, interpMode=RCIM_Cubic, leaveTangent=0.7143, time=0.1, value=5}
          <at 0.7 0.5>
          <at 1 0 RCIM_Cubic>
         ]}}}))
    ScaleA=true
    ScaleRGB=true
   )
  ]
  rendererClasses [&NiagaraSpriteRendererProperties, &NiagaraLightRendererProperties]
  renderers [
   {rendererClass=&NiagaraSpriteRendererProperties}
   {rendererClass=&NiagaraSpriteRendererProperties, rendererIndex=1}
   {rendererClass=&NiagaraSpriteRendererProperties, rendererIndex=2}
   {rendererClass=&NiagaraLightRendererProperties, rendererIndex=3}
   {rendererClass=&NiagaraSpriteRendererProperties, rendererIndex=4}
  ]
 }
 StartupEmbers {
  simTarget=GPUComputeSim
  emitterUpdateScript [
   &EmitterState(Life Cycle Mode=::NewEnumerator1, Loop Behavior=::NewEnumerator1, Loop Duration=10, UseLoopDelay=false)
   &SpawnBurst_Instantaneous(Spawn Count=100, Use Spawn Probability=false, UseLoopCountLimit=false)
   &SpawnBurst_Instantaneous(Spawn Count=50, Spawn Time=3.75, Use Spawn Probability=false, UseLoopCountLimit=false) {module=SpawnBurst_Instantaneous001}
   &SpawnBurst_Instantaneous(Spawn Count=50, Spawn Time=5.5, Use Spawn Probability=false, UseLoopCountLimit=false) {module=SpawnBurst_Instantaneous002}
  ]
  particleSpawnScript [
   &InitializeParticle(
    Lifetime=&RandomRangeFloat(Minimum=0.2, Maximum=1.5)
    Sprite Size Mode=::NewEnumerator3
    Uniform Sprite Size=7
    UsePositionOffset=false
    Write Material Random=true
   )
   &ShapeLocation(
    Cylinder Height=1
    Cylinder Radius=90
    MaskSpawnGroup=false
    Override Seed=false
    Rotation Axis=<0 1 0>
    Rotation Mode=::NewEnumerator1
    Shape Primitive=::NewEnumerator1
    Surface Only=false
    UseOwnerScale=true
   )
   &AddVelocity(
    Cone Angle=70
    Cone Axis=<0 -1 0>
    Module.UseConeFalloff=false
    Module.UseVelDistribution=false
    Override Seed=false
    Velocity Mode=::NewEnumerator2
    Velocity Speed=&RandomRangeFloat(Minimum=512, Maximum=1024)
   )
  ]
  particleUpdateScript [
   &ParticleState()
   &ScaleColor(
    Linear Color Curve=DataInterface() {propertyValues {
      AlphaCurve {keys [<at 0 0.13>,<at 1 0>]}
      BlueCurve {keys [<at 0 5>,<at 0.9891 4.1629>]}
      GreenCurve {keys [<at 0 3.4283>,<at 0.9891 0.3845>]}
      RedCurve {keys [<at 0 0.1406>,<at 0.9891 5>]}
     }}
    Scale Mode=::NewEnumerator2
   )
   &Drag(Drag=&RandomRangeFloat(Minimum=1, Maximum=3), Use Linear Drag=true, Use Rotational Drag=true)
   &CurlNoiseForce(Noise Frequency=10, Noise Strength=500, Pan Noise=true, Pan Noise Field=<1 1069.1512 0>)
   &SolveForcesAndVelocity(Clamp Velocity=false, Limit Acceleration=false)
   &ScaleSpriteSizeBySpeed(Max Scale Factor=<2 4>)
  ]
  rendererClasses [&NiagaraSpriteRendererProperties]
  renderers [{rendererClass=&NiagaraSpriteRendererProperties}]
 }
 StartupMainBurst {
  emitterSpawnScript [
   /SetVariables(Emitter.Distortion_Input=1, Emitter.LightConeOverallSize=2.2) {isSetParameters=true, module=SetVariables_#9}
   /SetVariables(Emitter.FlareSizeScalar=0.14) {isSetParameters=true, module=SetVariables_#10}
  ]
  emitterUpdateScript [
   &EmitterState(
    Life Cycle Mode=::NewEnumerator1
    Loop Behavior=::NewEnumerator1
    Loop Delay=$System.MainEnginesTurnOnTime
    Loop Duration=1
    UseLoopDelay=true
   )
   &SpawnBurst_Instantaneous(Spawn Count=1, Use Spawn Probability=false)
  ]
  particleSpawnScript [
   &InitializeParticle(
    Color=<rgba 0.7812 0.0892 0.6522 1>
    Sprite Size=&Multiply_Vector2D_ByFloat(Float=$Emitter.FlareSizeScalar, Vector2D=&Multiply_Vector2D_ByFloat(Float=18500, Vector2D=<1 0.5>))
    Sprite Size Mode=::NewEnumerator4
    UsePositionOffset=false
    Write Material Random=true
   )
   &CameraOffset(Camera Offset Amount=500)
   &SystemLocation(MaskSpawnGroup=false, Offset=<0 0 25.0808>)
   /SetVariables(Particles.LightPos:NiagaraPosition=<0 -500 0>) {enabled=false, isSetParameters=true, module=SetVariables_#11}
   /Niagara/Modules/Audio/PlayAudio(Audio Player=DataInterface() {propertyValues {
      Attenuation=null
      Concurrency=null
      MaxPlaysPerTick=10
      SoundToPlay=/Game/Athena/Sounds/Vehicles/Bus/Athena_Bus_Warp_Thruster_Activate_C_Cue
      bAllowLoopingOneShotSounds=false
      bLimitPlaysPerTick=true
      bOnlyActiveDuringGameplay=false
      bStopWhenComponentIsDestroyed=true
      ConfigurationUserParameter {parameter {name=None, typeDefHandle {registeredTypeIndex=1}, varData []}}
      ParameterNames []
     }})
  ]
  particleUpdateScript [
   /SetVariables(
    Particles.LensColor:LinearColor=&MultiplyLinearColor(
     Linear Color=<rgba 0.1086 0.1286 0.2 1>
     Scale Factor=&MakeVector4_FromFloat(Float=&Multiply_Float(A=$User.OverallOpacity, B=$System.SputterFloat))
    )
    Particles.LensSize:Vector2f=&Multiply_Vector2D_ByFloat(Float=$Emitter.FlareSizeScalar, Vector2D=<20000 4000>)
   ) {enabled=false, isSetParameters=true, module=SetVariables_#12}
   /SetVariables(
    Particles.DarkRadialColor:LinearColor=&MultiplyLinearColor(
     Linear Color=<rgba 0.0902 0.0318 0.1615 0.5>
     Scale Factor=&MakeVector4_FromFloat(Float=&Multiply_Float(A=$User.OverallOpacity, B=$System.SputterFloat))
    )
    Particles.DarkRadialSize:Vector2f=&Multiply_Vector2D_ByFloat(Float=$Emitter.FlareSizeScalar, Vector2D=<15000 10000>)
   ) {enabled=false, isSetParameters=true, module=SetVariables_#13}
   /SetVariables(Particles.LightColor=&MultiplyLinearColor(
     Linear Color=<rgba 0.1379 0.6405 1 1>
     Scale Factor=&MakeVector4_FromFloat(Float=&Multiply_Float(
       A=3
       B=&Multiply_Float(
        A=&FloatFromCurve(FloatCurve=DataInterface() {propertyValues {Curve {keys [<at 0 0 RCIM_Cubic>, {arriveTangent=1.25, interpMode=RCIM_Cubic, leaveTangent=1.25, time=0.1, value=3.75}, <at 0.4 0.5>, <at 1 0 RCIM_Cubic>]}}})
        B=$System.SputterFloat
       )
      ))
    )) {isSetParameters=true, module=SetVariables_#14}
   &ParticleState(Kill Particles When Lifetime Has Elapsed=false, Let Infinitely Lived Particles Die When Emitter Deactivates=true)
   &ScaleSpriteSize(Scale Factor=&Vector2DFromFloat(Value=&FloatFromCurve(FloatCurve=DataInterface() {propertyValues {Curve {keys [
          {arriveTangent=0.225, interpMode=RCIM_Cubic, time=0, value=0.025}
          {arriveTangent=1.2, interpMode=RCIM_Cubic, leaveTangent=1.2, time=0.1, value=0.575}
          <at 1 1.225 RCIM_Cubic>
         ]}}})))
   &ScaleColor(
    Scale Alpha=&FloatFromCurve(FloatCurve=DataInterface() {propertyValues {Curve {keys [
         <at 0 1 RCIM_Cubic>
         {arriveTangent=-2.5, interpMode=RCIM_Cubic, leaveTangent=-2.5, time=0.1, value=1.25}
         {arriveTangent=-1.3889, interpMode=RCIM_Cubic, leaveTangent=-1.3889, time=0.3, value=0.25}
         <at 1 0 RCIM_Cubic>
        ]}}})
    Scale RGB=&VectorFromFloat(Value=&FloatFromCurve(
      FloatCurve=DataInterface() {propertyValues {Curve {keys [
          {arriveTangent=47.5885, interpMode=RCIM_Cubic, time=0, value=0}
          {arriveTangent=0.7143, interpMode=RCIM_Cubic, leaveTangent=0.7143, time=0.1, value=5}
          <at 0.7 0.5>
          <at 1 0 RCIM_Cubic>
         ]}}}
      Scale Curve=8
     ))
    ScaleA=true
    ScaleRGB=true
   )
  ]
  rendererClasses [&NiagaraSpriteRendererProperties]
  renderers [{rendererClass=&NiagaraSpriteRendererProperties}]
 }
 Streaks_Bursts {
  emitterUpdateScript [
   &EmitterState(Life Cycle Mode=::NewEnumerator1, Loop Behavior=::NewEnumerator1, Loop Delay=0.0353, Loop Duration Mode=::NewEnumerator1, UseLoopDelay=false)
   &SpawnBurst_Instantaneous(
    Spawn Time=&SelectFloatFromArray(Direct Array Index=0, Float Selection Array=$System.BurstTimings)
    Use Spawn Probability=false
    UseLoopCountLimit=false
   )
   &SpawnBurst_Instantaneous(
    Spawn Group=1
    Spawn Time=&SelectFloatFromArray(Direct Array Index=1, Float Selection Array=$System.BurstTimings)
    Use Spawn Probability=false
    UseLoopCountLimit=false
   ) {module=SpawnBurst_Instantaneous001}
   &SpawnBurst_Instantaneous(
    Spawn Group=2
    Spawn Time=&SelectFloatFromArray(Direct Array Index=2, Float Selection Array=$System.BurstTimings)
    Use Spawn Probability=false
    UseLoopCountLimit=false
   ) {module=SpawnBurst_Instantaneous002}
  ]
  particleSpawnScript [
   &InitializeParticle(
    Color=<rgba 2.9306 14.315 20 0.12>
    Lifetime Max=1.2
    Lifetime Min=0.758
    Lifetime Mode=::NewEnumerator1
    Position Offset=&SelectVectorFromArray(
     Direct Array Index=&SpawnGroup()
     Vector Selection Array=DataInterface() {propertyValues {FloatData [<50 -70 75>, <0 -70 -20>, <-100 -70 75>]}}
    )
    Sprite Size=&Multiply_Vector2D_ByFloat(Float=1, Vector2D=&UniformRangedVector2D(Minimum=<150 250>, Maximum=<150 350>))
    Sprite Size Mode=::NewEnumerator4
    UsePositionOffset=true
    Write Material Random=true
   )
   &SphereLocation(
    Hemisphere Y=true
    MaskSpawnGroup=false
    Module.Override Randomness=false
    Offset=<0 0 0.01>
    Sphere Radius=&Multiply_Float(A=5, B=1)
    Surface Only=true
   ) {enabled=false}
   &AddVelocityFromPoint(
    Module.Use Falloff Distance=false
    Origin Offset=<0 0 -3>
    Velocity Strength=&UniformRangedFloat(Minimum=400, Maximum=500)
   ) {enabled=false}
   &AddVelocity(
    Cone Angle=2
    Cone Axis=<0 -1 0>
    Inner Cone Angle=2
    Module.UseConeFalloff=false
    Module.UseVelDistribution=false
    Override Seed=false
    Velocity Mode=::NewEnumerator2
   )
   &SubUVAnimation(End Frame=3, SubUV Animation Mode:ENiagara_SubUVLookupMode=::NewEnumerator2)
   &CameraOffset(Camera Offset Amount=-50)
  ]
  particleUpdateScript [
   &ParticleState()
   &Drag(
    Drag=&FloatFromCurve(FloatCurve=DataInterface() {propertyValues {Curve {keys [<at 0.2 0.925>,<at 0.25 0>]}}}, Scale Curve=2)
    Use Linear Drag=true
    Use Rotational Drag=false
   )
   &GravityForce(Gravity=&Multiply_VectorByFloat(
     Vector=<0 0 -2.2>
     Float=&FloatFromCurve(
      FloatCurve=DataInterface() {propertyValues {Curve {keys [{arriveTangent=3.1357, interpMode=RCIM_Cubic, leaveTangent=3.1357, tangentMode=RCTM_User, time=0.1, value=0}, <at 0.6 1>]}}}
      CurveIndex=$Emitter.Age
     )
    )) {enabled=false}
   &SolveForcesAndVelocity(Clamp Velocity=false, Limit Acceleration=false)
   &ScaleSpriteSize(Scale Factor=&Vector2DFromCurve(
     Vector2Curve=DataInterface() {propertyValues {
       XCurve {keys [
         <at 0 0.625 RCIM_Cubic>
         {arriveTangent=1.125, interpMode=RCIM_Cubic, leaveTangent=1.125, time=0.4, value=0.75}
         {arriveTangent=2.6138, interpMode=RCIM_Cubic, leaveTangent=2.6138, tangentMode=RCTM_User, time=1, value=1.75}
        ]}
       YCurve {keys [
         {arriveTangent=6.6719, interpMode=RCIM_Cubic, leaveTangent=6.6719, tangentMode=RCTM_User, time=0, value=0.1}
         {arriveTangent=1.775, interpMode=RCIM_Cubic, leaveTangent=1.775, time=0.1, value=1.025}
         <at 1 1.875 RCIM_Cubic>
        ]}
      }}
     ScaleCurve=<1.5 2>
    ))
   &ScaleColor(
    Linear Color Curve=DataInterface() {propertyValues {
      AlphaCurve {keys [
        {arriveTangent=-0.3577, interpMode=RCIM_Cubic, leaveTangent=-0.3577, tangentMode=RCTM_User, time=0.08, value=1}
        {arriveTangent=-1.087, interpMode=RCIM_Cubic, leaveTangent=-1.0164, tangentMode=RCTM_Break, time=0.3702, value=1}
        <at 1 0>
       ]}
      BlueCurve {keys [<at 0.6491 0.3>,<at 1 0.45>]}
      GreenCurve {keys [<at 0.6491 0.3>,<at 1 0.45>]}
      RedCurve {keys [<at 0.6491 0.3>,<at 1 0.45>]}
     }}
    Scale Mode=::NewEnumerator2
   )
   &DynamicMaterialParameters(
    Index 0 Param 1=50
    Index 0 Param 4=&FloatFromCurve(FloatCurve=DataInterface() {propertyValues {Curve {keys [{interpMode=RCIM_Cubic, tangentMode=RCTM_User, time=0, value=0}, {interpMode=RCIM_Cubic, tangentMode=RCTM_User, time=0.4, value=1}]}}})
    Index 1 Param 1=&FloatFromCurve(
     FloatCurve=DataInterface() {propertyValues {Curve {keys [
         {interpMode=RCIM_Cubic, tangentMode=RCTM_User, time=0, value=0}
         {arriveTangent=0.6295, interpMode=RCIM_Cubic, leaveTangent=0.6295, tangentMode=RCTM_User, time=1, value=1}
        ]}}}
     Scale Curve=0.1
    )
    Index 1 Param 2=&FloatFromCurve(
     FloatCurve=DataInterface() {propertyValues {Curve {keys [
         {arriveTangent=-713.2054, interpMode=RCIM_Cubic, leaveTangent=-713.2054, tangentMode=RCTM_User, time=0, value=50}
         {arriveTangent=-0.3869, interpMode=RCIM_Cubic, leaveTangent=-0.3869, tangentMode=RCTM_User, time=0.125, value=0}
        ]}}}
     Scale Curve=2
    )
    Index 1 Param 3=&FloatFromCurve(FloatCurve=DataInterface() {propertyValues {Curve {keys [{interpMode=RCIM_Cubic, tangentMode=RCTM_User, time=0.2, value=0.5}, {interpMode=RCIM_Cubic, tangentMode=RCTM_User, time=1, value=1}]}}})
    Index 1 Param 4=&FloatFromCurve(FloatCurve=DataInterface() {propertyValues {Curve {keys [<at 0 1>,<at 1 0>]}}}, Scale Curve=-0.3)
    Module.Param0WriteEnabled=true
    Module.Param0WriteEnabled1=true
    Module.Param1WriteEnabled=false
    Module.Param1WriteEnabled1=true
    Module.Param2WriteEnabled=false
    Module.Param2WriteEnabled1=true
    Module.Param3WriteEnabled=true
    Module.Param3WriteEnabled1=true
    Write Parameter Index 1=true
   )
  ]
  rendererClasses [&NiagaraSpriteRendererProperties]
  renderers [{rendererClass=&NiagaraSpriteRendererProperties}]
 }
 Streaks_Fat {
  emitterUpdateScript [
   &EmitterState(
    Life Cycle Mode=::NewEnumerator1
    Loop Behavior=::NewEnumerator1
    Loop Delay=&Subtract_Float(A=$System.MainEnginesTurnOnTime, B=0)
    Loop Duration Mode=::NewEnumerator1
    UseLoopDelay=true
   )
   &SpawnBurst_Instantaneous(Spawn Count=6, Use Spawn Probability=false, UseLoopCountLimit=false)
   &SpawnBurst_Instantaneous(Spawn Count=3, Spawn Group=1, Use Spawn Probability=false, UseLoopCountLimit=false) {enabled=false, module=SpawnBurst_Instantaneous001}
  ]
  particleSpawnScript [
   &InitializeParticle(
    Color=<rgba 20 8.6101 18.1386 0.12>
    Lifetime Max=1.8
    Lifetime Min=0.8
    Lifetime Mode=::NewEnumerator1
    Sprite Size=&Multiply_Vector2D_ByFloat(Float=0.75, Vector2D=&UniformRangedVector2D(Minimum=<85 250>, Maximum=<160 500>))
    Sprite Size Mode=::NewEnumerator4
    UsePositionOffset=true
    Write Material Random=true
   )
   &SphereLocation(
    Hemisphere Y=true
    MaskSpawnGroup=false
    Module.Override Randomness=false
    Offset=<0 0 0.01>
    Sphere Radius=&Multiply_Float(A=5, B=1)
    Surface Only=true
   ) {enabled=false}
   &AddVelocityFromPoint(
    Module.Use Falloff Distance=false
    Origin Offset=<0 0 -3>
    Velocity Strength=&UniformRangedFloat(Minimum=400, Maximum=500)
   ) {enabled=false}
   &ShapeLocation(
    Cylinder Radius=250
    MaskSpawnGroup=false
    Offset Mode=::NewEnumerator0
    Override Seed=false
    Rotation Axis=<0 1 0>
    Rotation Mode=::NewEnumerator1
    Shape Primitive=::NewEnumerator1
    Surface Only=false
    UseOwnerScale=true
   )
   &AddVelocity(
    Cone Angle=10
    Cone Axis=<0 -1 0>
    Inner Cone Angle=5
    Module.UseConeFalloff=false
    Module.UseVelDistribution=false
    Override Seed=false
    Velocity Mode=::NewEnumerator2
   )
   &SubUVAnimation(End Frame=3, SubUV Animation Mode:ENiagara_SubUVLookupMode=::NewEnumerator2)
   &CameraOffset(Camera Offset Amount=-100)
  ]
  particleUpdateScript [
   &ParticleState()
   &Drag(
    Drag=&FloatFromCurve(FloatCurve=DataInterface() {propertyValues {Curve {keys [<at 0.2 1>,<at 0.3 0>]}}}, Scale Curve=15)
    Use Linear Drag=true
    Use Rotational Drag=false
   )
   &GravityForce(Gravity=&Multiply_VectorByFloat(
     Vector=<0 0 -2.2>
     Float=&FloatFromCurve(
      FloatCurve=DataInterface() {propertyValues {Curve {keys [{arriveTangent=3.1357, interpMode=RCIM_Cubic, leaveTangent=3.1357, tangentMode=RCTM_User, time=0.1, value=0}, <at 0.6 1>]}}}
      CurveIndex=$Emitter.Age
     )
    )) {enabled=false}
   &SolveForcesAndVelocity(Clamp Velocity=false, Limit Acceleration=false)
   &ScaleSpriteSize(Scale Factor=&Vector2DFromCurve(
     Vector2Curve=DataInterface() {propertyValues {
       XCurve {keys [
         <at 0 0.625 RCIM_Cubic>
         {arriveTangent=1.125, interpMode=RCIM_Cubic, leaveTangent=1.125, time=0.4, value=0.75}
         {arriveTangent=2.6138, interpMode=RCIM_Cubic, leaveTangent=2.6138, tangentMode=RCTM_User, time=1, value=1.75}
        ]}
       YCurve {keys [{arriveTangent=6.6719, interpMode=RCIM_Cubic, leaveTangent=6.6719, tangentMode=RCTM_User, time=0, value=0.1}, <at 1 1.875 RCIM_Cubic>]}
      }}
     ScaleCurve=<1.5 2>
    ))
   &ScaleColor(
    Linear Color Curve=DataInterface() {propertyValues {
      AlphaCurve {keys [
        {arriveTangent=-0.3577, interpMode=RCIM_Cubic, leaveTangent=-0.3577, tangentMode=RCTM_User, time=0.08, value=1}
        {arriveTangent=-1.087, interpMode=RCIM_Cubic, leaveTangent=-1.0164, tangentMode=RCTM_Break, time=0.3702, value=1}
        <at 1 0>
       ]}
      BlueCurve {keys [<at 0.6491 0.3>,<at 1 0.45>]}
      GreenCurve {keys [<at 0.6491 0.3>,<at 1 0.45>]}
      RedCurve {keys [<at 0.6491 0.3>,<at 1 0.45>]}
     }}
    Scale Mode=::NewEnumerator2
   )
   &DynamicMaterialParameters(
    Index 0 Param 1=100
    Index 0 Param 4=&FloatFromCurve(FloatCurve=DataInterface() {propertyValues {Curve {keys [{interpMode=RCIM_Cubic, tangentMode=RCTM_User, time=0, value=0}, {interpMode=RCIM_Cubic, tangentMode=RCTM_User, time=0.4, value=1}]}}})
    Index 1 Param 1=&FloatFromCurve(
     FloatCurve=DataInterface() {propertyValues {Curve {keys [
         {interpMode=RCIM_Cubic, tangentMode=RCTM_User, time=0, value=0}
         {arriveTangent=0.6295, interpMode=RCIM_Cubic, leaveTangent=0.6295, tangentMode=RCTM_User, time=1, value=1}
        ]}}}
     Scale Curve=0.42
    )
    Index 1 Param 2=&FloatFromCurve(FloatCurve=DataInterface() {propertyValues {Curve {keys [
         {arriveTangent=-713.2054, interpMode=RCIM_Cubic, leaveTangent=-713.2054, tangentMode=RCTM_User, time=0, value=50}
         {arriveTangent=-0.3869, interpMode=RCIM_Cubic, leaveTangent=-0.3869, tangentMode=RCTM_User, time=0.125, value=0}
        ]}}})
    Index 1 Param 3=&FloatFromCurve(
     FloatCurve=DataInterface() {propertyValues {Curve {keys [{interpMode=RCIM_Cubic, tangentMode=RCTM_User, time=0.2, value=0.5}, {interpMode=RCIM_Cubic, tangentMode=RCTM_User, time=1, value=1}]}}}
     Scale Curve=2
    )
    Index 1 Param 4=&FloatFromCurve(FloatCurve=DataInterface() {propertyValues {Curve {keys [<at 0 1>,<at 1 0>]}}}, Scale Curve=-0.3)
    Module.Param0WriteEnabled=true
    Module.Param0WriteEnabled1=true
    Module.Param1WriteEnabled=false
    Module.Param1WriteEnabled1=true
    Module.Param2WriteEnabled=false
    Module.Param2WriteEnabled1=true
    Module.Param3WriteEnabled=true
    Module.Param3WriteEnabled1=true
    Write Parameter Index 1=true
   )
  ]
  rendererClasses [&NiagaraSpriteRendererProperties]
  renderers [{rendererClass=&NiagaraSpriteRendererProperties}]
 }
 WarmupAirglow {
  emitterUpdateScript [
   &EmitterState(Life Cycle Mode=::NewEnumerator1, Loop Behavior=::NewEnumerator1, Loop Duration Mode=::NewEnumerator1, UseLoopDelay=false)
   &SpawnBurst_Instantaneous(Use Spawn Probability=false, UseLoopCountLimit=false)
  ]
  particleSpawnScript [&InitializeParticle(Lifetime=3, Sprite Size Mode=::NewEnumerator3, Uniform Sprite Size=650, UsePositionOffset=false, Write Material Random=true)]
  particleUpdateScript [
   &ParticleState(Kill Particles When Lifetime Has Elapsed=false, Let Infinitely Lived Particles Die When Emitter Deactivates=true)
   &SolveForcesAndVelocity(Clamp Velocity=false, Limit Acceleration=false)
   &Color(Color=$System.EnginesColorOveLife)
   &ScaleColor(
    Color Value To Scale=$StackContext.Color
    Scale Alpha=&Lerp_Float(
     A=0.25
     B=0.5
     Alpha=&Waveform(
      Waveform Count=::NewEnumerator1
      "Waveform [1]"=::NewEnumerator12
      "[1] Frequency"=12
      "[1] Amplitude Min/Max"=<0 1>
      "[2] Average"=false
      "Waveform [2]"=::NewEnumerator12
      "[2] Input Position"=0
      "[2] Phase"=$Emitter.Age
      "[2] Phase Scale"=1
      "[2] Compound Frequency"=2
      "[2] Frequency"=10
      "[2] Amplitude Min/Max"=<0 1>
      "[2] Scale"=1
      "[2] Bias"=0
      "[2] Exponent"=1
      "[2] Pulse Duration"=0.1
      "[2] Clip Triangle Min/Max"=<0 0>
      "[2] Random Seed"=0
     )
    )
    ScaleA=true
    ScaleRGB=true
   )
   &ScaleSpriteSize(Scale Factor=&Vector2DFromFloat(Value=&Lerp_Float(
      A=&FloatFromCurve(FloatCurve=DataInterface() {propertyValues {Curve {keys [
           <at 0 0 RCIM_Cubic>
           {arriveTangent=3.125, interpMode=RCIM_Cubic, leaveTangent=3.125, time=0.1, value=1.5}
           {arriveTangent=-0.4167, interpMode=RCIM_Cubic, leaveTangent=-0.4167, time=0.4, value=1.25}
           <at 1 1.125 RCIM_Cubic>
          ]}}})
      B=&RandomRangeFloat(Minimum=0.8, Maximum=1, Evaluation Type=::NewEnumerator1)
      Alpha=0.125
     )))
   &CameraOffset(Camera Offset Amount=350)
  ]
  rendererClasses [&NiagaraSpriteRendererProperties]
  renderers [{rendererClass=&NiagaraSpriteRendererProperties}]
 }
}
summary {
 systemName=NS_BattleBus_Warp_BoostEngine_BuildUp
 emitters [
  {
   emitterName=MainEngine_LightingFX
   rendererClasses [&NiagaraSpriteRendererProperties, &NiagaraLightRendererProperties]
  }
  {emitterName=EngineInteriorWarmup, rendererClasses [&NiagaraMeshRendererProperties]}
  {emitterName=StartupBurst, rendererClasses [&NiagaraSpriteRendererProperties, &NiagaraLightRendererProperties]}
  {emitterName=StartupAirglow, rendererClasses [&NiagaraSpriteRendererProperties]}
  {emitterName=StartupEmbers, simTarget=GPUComputeSim, rendererClasses [&NiagaraSpriteRendererProperties]}
  {emitterName=FireElements-Warmups, rendererClasses [&NiagaraMeshRendererProperties]}
  {emitterName=WarmupAirglow, rendererClasses [&NiagaraSpriteRendererProperties]}
  {emitterName=AnimatedLight, rendererClasses [&NiagaraLightRendererProperties]}
  {emitterName=EngineInteriorWarmup001, rendererClasses [&NiagaraMeshRendererProperties]}
  {emitterName=StartupMainBurst, rendererClasses [&NiagaraSpriteRendererProperties]}
  {
   emitterName=Shockwave
   rendererClasses [
    &NiagaraLightRendererProperties
    &NiagaraSpriteRendererProperties
    &NiagaraMeshRendererProperties
   ]
  }
  {emitterName=FN_Mesh, rendererClasses [&NiagaraMeshRendererProperties]}
  {emitterName=Streaks_Fat, rendererClasses [&NiagaraSpriteRendererProperties]}
  {emitterName=Streaks_Bursts, rendererClasses [&NiagaraSpriteRendererProperties]}
  {emitterName=BlowingParticles, simTarget=GPUComputeSim, rendererClasses [&NiagaraSpriteRendererProperties]}
  {emitterName=GroundSmokeSheet, rendererClasses [&NiagaraMeshRendererProperties]}
  {emitterName=SmokeLicks, rendererClasses [&NiagaraMeshRendererProperties]}
  {emitterName=EngineBurst, rendererClasses [&NiagaraMeshRendererProperties]}
  {emitterName=ConfettiBurst, simTarget=GPUComputeSim, rendererClasses [&NiagaraMeshRendererProperties]}
  {emitterName=InnerNoiseSphere, rendererClasses [&NiagaraMeshRendererProperties]}
  {emitterName=SmokeWorld, rendererClasses [&NiagaraSpriteRendererProperties]}
  {emitterName=Impact_Streaks, rendererClasses [&NiagaraSpriteRendererProperties]}
  {emitterName=PlayAudioOnSpawnBurst, rendererClasses []}
 ]
 userVariables [
  User.FlareFlicker {type=NiagaraFloat, defaultValue=1, description=""}
  User.Global Color {type=LinearColor, description="", defaultValue {struct=LinearColor, value=<rgba 1 1 1 1>}}
  User.Global Scale {type=NiagaraFloat, defaultValue=0, description=""}
  User.OverallOpacity {type=NiagaraFloat, defaultValue=1, description=""}
 ]
}
