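# Niagara effect for a "rift out" burst among the Asteria lobby island space props, in blue-cyan tones.
# A crack mesh animates through scale, color and dynamic material curves, while a layered flare drives glow sprites from per-particle variables.
# Burst sparks stretch with speed under drag, and shards in a camera-oriented ring fly out as tumbling meshes with flickering sprites.
# Glints spawn at a steady rate, their size shrinking with emitter age.
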NiagaraSystem /Game/Environments/Asteria/Props/LobbyIslandSpace/FX/NS_PoolDust_RiftOut
dependencies {
 usedDataInterfaces [
  &NiagaraDataInterfaceMeshRendererInfo
  &NiagaraDataInterfaceCamera
  &NiagaraDataInterfaceCurve
  &NiagaraDataInterfaceVector2DCurve
  &NiagaraDataInterfaceVectorCurve
 ]
 usedDynamicInputs []
 usedModules [&SystemState]
}
emitters {
 BlueSparks {
  emitterUpdateScript [
   &EmitterState(Scalability Mode=::NewEnumerator1)
   &SpawnBurst_Instantaneous(
    Spawn Count=&MakeIntFromFloat(Float=&Lerp_Float(A=9, B=15, Alpha=$User.SizeNormalized), Conversion Mode=::NewEnumerator1)
    Use Spawn Probability=false
    UseLoopCountLimit=false
   ) {enabled=false}
   &SpawnBurst_Instantaneous(Spawn Count=50, Spawn Group=1, Spawn Time=0.05, Use Spawn Probability=false, UseLoopCountLimit=false) {module=SpawnBurst_Instantaneous001}
  ]
  particleSpawnScript [
   /SetVariables(Particles.Distribution:NiagaraFloat=&FloatFromCurve(
     FloatCurve=DataInterface() {propertyValues {Curve {keys [
         {interpMode=RCIM_Cubic, leaveTangent=0.4364, tangentMode=RCTM_Break, time=0, value=0}
         {arriveTangent=1.9884, interpMode=RCIM_Cubic, leaveTangent=1.9884, tangentMode=RCTM_User, time=0.64, value=0.5}
         <at 1 1 RCIM_Cubic>
        ]}}}
     CurveIndex=&RandomRangeFloat(Minimum=0, Maximum=1)
    )) {isSetParameters=true}
   &InitializeParticle(
    Color=&LerpLinearColors(StartColor=<rgba 0.141 0.5013 6 1>, EndColor=<rgba 0.0331 3.0407 8 1>, LerpFactor=$Particles.Distribution)
    Lifetime=&MaskFloatBySpawnGroup(Masked Float=0.2, Passthrough Float=&RandomRangeFloat(Minimum=0.5, Maximum=1.5))
    Sprite Size Mode=::NewEnumerator3
    Uniform Sprite Size=20
    UsePositionOffset=false
    Write Material Random=true
   )
   &ShapeLocation(
    MaskSpawnGroup=false
    Override Seed=false
    Sphere Radius=&MaskFloatBySpawnGroup(Masked Float=3, Passthrough Float=&Lerp_Float(A=15, B=25, Alpha=$User.SizeNormalized))
    Sphere Surface Distribution=1
    UseOwnerScale=true
   )
   &AddVelocity(
    Module.Use Falloff Distance=false
    Override Seed=false
    Velocity Mode=::NewEnumerator1
    Velocity Speed=&Multiply_Float(A=&RandomRangeFloat(Minimum=1501, Maximum=4500), B=4)
   )
  ]
  particleUpdateScript [
   &ParticleState()
   &DynamicMaterialParameters(
    Index 0 Param 1=&FloatFromCurve(FloatCurve=DataInterface() {propertyValues {Curve {keys [{arriveTangent=1.8602, interpMode=RCIM_Cubic, leaveTangent=1.8602, tangentMode=RCTM_User, time=0, value=1}, <at 1 2>]}}})
    Index 0 Param 2=5
    Index 0 Param 3=3
    Index 0 Param 4=&FloatFromCurve(FloatCurve=DataInterface() {propertyValues {Curve {keys [{arriveTangentWeight=0.1944, interpMode=RCIM_Cubic, leaveTangentWeight=0.1944, tangentMode=RCTM_User, time=0.4, value=0}, <at 0.9 1>]}}})
    Module.Param0WriteEnabled=true
    Module.Param1WriteEnabled=true
    Module.Param2WriteEnabled=true
    Module.Param3WriteEnabled=true
   )
   &Drag(
    Drag=&MaskFloatBySpawnGroup(
     Masked Float=12
     Passthrough Float=&FloatFromCurve(
      FloatCurve=DataInterface() {propertyValues {Curve {keys [
          {arriveTangent=4.106, interpMode=RCIM_Cubic, leaveTangent=4.106, tangentMode=RCTM_User, time=0, value=0.1}
          {arriveTangent=0.3, interpMode=RCIM_Cubic, leaveTangent=0.3, time=0.25, value=1}
          {arriveTangent=-6.537, interpMode=RCIM_Cubic, leaveTangent=-6.537, tangentMode=RCTM_User, time=0.5, value=0.25}
         ]}}}
      Scale Curve=9
     )
    )
    Use Linear Drag=true
    Use Rotational Drag=false
   )
   &CurlNoiseForce(
    Noise Frequency=&RandomRangeFloat(Minimum=40, Maximum=60)
    Noise Strength=&FloatFromCurve(
     FloatCurve=DataInterface() {propertyValues {Curve {keys [
         {arriveTangent=2.2956, interpMode=RCIM_Cubic, leaveTangent=2.2956, tangentMode=RCTM_User, time=0, value=0}
         {interpMode=RCIM_Cubic, tangentMode=RCTM_User, time=0.5, value=1}
         {arriveTangent=-2.5698, interpMode=RCIM_Cubic, leaveTangent=-2.5698, tangentMode=RCTM_User, time=1, value=0.4}
        ]}}}
     Scale Curve=&MaskFloatBySpawnGroup(Masked Float=0, Passthrough Float=400)
    )
    Pan Noise=false
   )
   &SolveForcesAndVelocity(Clamp Velocity=false, Limit Acceleration=false)
   &ScaleSpriteSize(Uniform Curve Sprite Scale=DataInterface() {propertyValues {Curve {keys [
        {interpMode=RCIM_Cubic, tangentMode=RCTM_Break, time=0.5, value=1}
        {arriveTangent=-0.1826, interpMode=RCIM_Cubic, leaveTangent=-7.3774, tangentMode=RCTM_Break, time=1, value=0}
       ]}}})
   &ScaleSpriteSizeBySpeed(
    Max Scale Factor=<3 6>
    Min Scale Factor=<1.2 1>
    Sample Scale Factor By Curve=true
    Scale Factor Curve=DataInterface() {propertyValues {Curve {keys [
        {arriveTangent=0.6496, interpMode=RCIM_Cubic, time=0, value=0}
        {arriveTangent=1.4663, interpMode=RCIM_Cubic, leaveTangent=1.4663, tangentMode=RCTM_User, time=1, value=1}
       ]}}}
    Velocity Threshold=700
   )
  ]
  rendererClasses [&NiagaraSpriteRendererProperties]
  renderers [{rendererClass=&NiagaraSpriteRendererProperties}]
 }
 Flare_COmplex001 {
  emitterUpdateScript [
   &EmitterState()
   &SpawnBurst_Instantaneous(Use Spawn Probability=true, UseLoopCountLimit=false)
   &SpawnBurst_Instantaneous(Spawn Group=1, Use Spawn Probability=true, UseLoopCountLimit=false) {module=SpawnBurst_Instantaneous001}
  ]
  particleSpawnScript [
   &InitializeParticle(
    Color=<rgba 0.4977 5.4992 8 0.25>
    Lifetime=&MaskFloatBySpawnGroup(Masked Float=0.5, Passthrough Float=0.7)
    Position Offset=&MaskVectorBySpawnGroup(Masked Vector=<0 0 0>, Passthrough Vector=<0 0 0>)
    Sprite Size Mode=::NewEnumerator3
    Uniform Sprite Size=2000
    UsePositionOffset=true
    Write Material Random=true
   )
   /SetVariables(Particles.DistanceLerp:NiagaraFloat=&FloatFromCurve(FloatCurve=DataInterface() {propertyValues {Curve {keys [<at 500 0>,<at 7000 1>]}}}, CurveIndex=$Engine.Owner.LODDistance)) {isSetParameters=true}
   &FN_SetSpawnGroup()
  ]
  particleUpdateScript [
   &ParticleState()
   &ScaleSpriteSize(Uniform Curve Sprite Scale=DataInterface() {propertyValues {Curve {keys [
        {arriveTangent=8.6837, interpMode=RCIM_Cubic, leaveTangent=15.0616, tangentMode=RCTM_Break, time=0, value=0}
        {arriveTangent=0.2156, interpMode=RCIM_Cubic, leaveTangent=0.2156, tangentMode=RCTM_User, time=0.5, value=3.5}
       ]}}})
   &CameraOffset(Camera Offset Amount=500)
   &DynamicMaterialParameters(
    Index 0 Param 1=1
    Index 0 Param 2=&Lerp_Float(
     A=1.25
     B=0
     Alpha=&FloatFromCurve(FloatCurve=DataInterface() {propertyValues {Curve {keys [
          {arriveTangent=1.3124, interpMode=RCIM_Cubic, leaveTangent=3.2573, tangentMode=RCTM_Break, time=0, value=0}
          {arriveTangent=0.3903, interpMode=RCIM_Cubic, leaveTangent=0.3903, tangentMode=RCTM_User, time=1, value=1}
         ]}}})
    )
    Module.Index0Param0Write=true
    Module.Index0Param1Write=true
    Module.Index0Param2Write=true
    Module.Index0Param3Write=false
   )
   &ScaleColor(
    Scale Alpha=&FloatFromCurve(FloatCurve=DataInterface() {propertyValues {Curve {keys [<at 0 1 RCIM_Cubic>,<at 1 0>]}}})
    ScaleA=true
    ScaleRGB=false
   )
   /SetVariables(
    Particles.Darken Color:LinearColor=&MakeLinearColorFromVectorAndFloat(
     "Vector (RGB)"=&MakeVectorFromLinearColorRGB(Color=<rgba 0.01 0.0143 0.02 1>)
     "Float (Alpha)"=&FloatFromCurve(FloatCurve=DataInterface() {propertyValues {Curve {keys [{interpMode=RCIM_Cubic, tangentMode=RCTM_User, time=0, value=1}, {interpMode=RCIM_Cubic, tangentMode=RCTM_User, time=1, value=0}]}}})
    )
    Particles.Darken Size:Vector2f=&Multiply_Vector2D_ByFloat(Float=1.4, Vector2D=$Particles.SpriteSize)
   ) {enabled=false, isSetParameters=true}
   /SetVariables(
    Particles.Glow Color=&MultiplyLinearColor(
     Linear Color=$Particles.Color
     Scale Factor=&MakeFloat4FromVectorAndFloat(
      "Vector (RGB)"=<1 1 1>
      "Float (Alpha)"=&FloatFromCurve(
       FloatCurve=DataInterface() {propertyValues {Curve {keys [
           {arriveTangent=3.5878, interpMode=RCIM_Cubic, leaveTangent=3.5878, tangentMode=RCTM_User, time=0, value=0.35}
           {arriveTangent=-0.05, interpMode=RCIM_Cubic, leaveTangent=-0.05, time=0.4, value=1}
           {arriveTangent=-0.0405, interpMode=RCIM_Cubic, leaveTangent=-0.0405, tangentMode=RCTM_User, time=1, value=0.3}
          ]}}}
       Scale Curve=0.02
      )
     )
    )
    Particles.Glow Size=&Multiply_Vector2D_ByFloat(Float=2, Vector2D=<4000 4000>)
   ) {isSetParameters=true}
   /SetVariables(Particles.Wide Flare Scale:Vector2f=&Vector2DFromCurve(
     Vector2Curve=DataInterface() {propertyValues {
       XCurve {keys [
         {arriveTangent=4.673, interpMode=RCIM_Cubic, leaveTangent=4.673, tangentMode=RCTM_User, time=0, value=0}
         {arriveTangentWeight=0.1536, interpMode=RCIM_Cubic, leaveTangentWeight=0.1152, tangentMode=RCTM_User, time=0.3, value=1}
         {arriveTangent=-3.1914, interpMode=RCIM_Cubic, leaveTangent=-3.1914, tangentMode=RCTM_User, time=0.5, value=0.15}
         {arriveTangent=-2.3891, interpMode=RCIM_Cubic, leaveTangent=-2.3891, tangentMode=RCTM_User, time=0.7, value=0.25}
         {interpMode=RCIM_Cubic, leaveTangent=-1.231, time=1, value=0}
        ]}
       YCurve {keys [
         {arriveTangent=10.6903, interpMode=RCIM_Cubic, leaveTangent=10.6903, tangentMode=RCTM_User, time=0, value=0}
         {arriveTangent=0.8, interpMode=RCIM_Cubic, leaveTangent=0.8, time=0.25, value=1}
         {arriveTangent=-1.7663, interpMode=RCIM_Cubic, leaveTangent=-1.7663, tangentMode=RCTM_User, time=0.5, value=0.4}
         {arriveTangent=-4.8806, interpMode=RCIM_Cubic, leaveTangent=-4.8806, tangentMode=RCTM_User, time=0.7, value=1}
         {interpMode=RCIM_Cubic, leaveTangent=-2.2148, time=1, value=0}
        ]}
      }}
     ScaleCurve=&Multiply_Vector2D_ByFloat(Float=1.5, Vector2D=<10000 1500>)
    )) {isSetParameters=true}
   /SetVariables(
    Particles.1-x Glow Camera Offset=800
    Particles.1-x Glow Color:LinearColor=&MultiplyLinearColor(
     Linear Color=$Particles.Color
     Scale Factor=&MakeFloat4FromVectorAndFloat(
      "Vector (RGB)"=<1 1 1>
      "Float (Alpha)"=&FloatFromCurve(
       FloatCurve=DataInterface() {propertyValues {Curve {keys [
           {arriveTangent=3.5878, interpMode=RCIM_Cubic, tangentMode=RCTM_Break, time=0, value=0.35}
           {arriveTangent=-0.05, interpMode=RCIM_Cubic, leaveTangent=-0.05, time=0.3, value=1}
           {arriveTangent=-0.0405, interpMode=RCIM_Cubic, leaveTangent=-0.0405, tangentMode=RCTM_User, time=1, value=0.3}
          ]}}}
       Scale Curve=2
      )
     )
    )
    Particles.1-x Glow Size:Vector2f=&Vector2DFromFloat(Value=1500)
   ) {isSetParameters=true}
  ]
  rendererClasses [&NiagaraSpriteRendererProperties]
  renderers [
   {rendererClass=&NiagaraSpriteRendererProperties}
   {rendererClass=&NiagaraSpriteRendererProperties, rendererIndex=1}
   {rendererClass=&NiagaraSpriteRendererProperties, rendererIndex=2}
   {rendererClass=&NiagaraSpriteRendererProperties, rendererIndex=3}
  ]
 }
 Glints {
  emitterUpdateScript [
   &EmitterState(Life Cycle Mode=::NewEnumerator1, Loop Behavior=::NewEnumerator1, Loop Duration=2.3, UseLoopDelay=true)
   &SpawnRate(SpawnRate=12, Use Spawn Probability=true) {module=SpawnRate001}
  ]
  particleSpawnScript [&InitializeParticle(
    Color=<rgba 0 1 1.5 1>
    Lifetime=&RandomRangeFloat(Minimum=0.25, Maximum=0.55)
    Position Offset=<-820 0 0>
    Sprite Size=&Multiply_Vector2D_ByFloat(Float=1.7, Vector2D=&RandomRangeVector2D(Minimum=<1200 1200>, Maximum=<2000 2000>))
    Sprite Size Mode=::NewEnumerator4
    UsePositionOffset=false
    Write Material Random=true
   )]
  particleUpdateScript [
   &ParticleState()
   &ScaleSpriteSize(
    Scale Factor=&Vector2DFromFloat(Value=&FloatFromCurve(
      FloatCurve=DataInterface() {propertyValues {Curve {keys [
          <at 0 0 RCIM_Cubic>
          {arriveTangent=1.4286, interpMode=RCIM_Cubic, leaveTangent=1.4286, time=0.5754, value=0.3493}
          {arriveTangent=-0.8228, interpMode=RCIM_Cubic, leaveTangent=-0.8228, time=0.7, value=1}
          <at 1 0>
         ]}}}
      Scale Curve=&FloatFromCurve(FloatCurve=DataInterface() {propertyValues {Curve {keys [<at 4 2>,<at 6 0>]}}}, CurveIndex=$Emitter.Age)
     ))
    Scale Sprite Size Mode=::NewEnumerator2
   )
   &Drag(Drag=0.2, Use Linear Drag=true, Use Rotational Drag=true)
   &SolveForcesAndVelocity(Clamp Velocity=false, Limit Acceleration=false)
   &SystemLocation(
    MaskSpawnGroup=false
    Offset=&MakeVector(
     X=&FloatFromCurve(FloatCurve=DataInterface() {propertyValues {Curve {keys [<at 4 -670>,<at 5.5 -870>]}}}, CurveIndex=$Emitter.Age)
     Y=0
     Z=0
    )
   ) {enabled=false}
  ]
  rendererClasses [&NiagaraSpriteRendererProperties]
  renderers [{rendererClass=&NiagaraSpriteRendererProperties}]
 }
 LargeCracks {
  emitterUpdateScript [
   &EmitterState(Inactive Response=::NewEnumerator1, Life Cycle Mode=::NewEnumerator1, Loop Behavior=::NewEnumerator1, Loop Duration=6.3, UseLoopDelay=false)
   &SpawnBurst_Instantaneous(Spawn Time=0.02, Use Spawn Probability=true, UseLoopCountLimit=false)
   &SpawnBurst_Instantaneous(Spawn Time=7.0333, Use Spawn Probability=false, UseLoopCountLimit=false) {enabled=false, module=SpawnBurst_Instantaneous001}
  ]
  particleSpawnScript [
   &InitializeParticle(
    Color=<rgba 1 1.5337 2 1>
    Lifetime=3
    Mesh Scale Mode=::NewEnumerator3
    Mesh Uniform Scale=80
    UsePositionOffset=true
    Write Material Random=true
   ) {module=InitializeParticle001}
   &InitialMeshOrientation(
    Mesh Orientation Mode=::NewEnumerator7
    Rotation=&RandomRangeVector(Minimum=<0 0 0>, Maximum=<180 0 0>)
    Use Rotation Vector=true
   )
   /SetVariables(Particles.SplineScale:Vector3f=&Multiply_VectorByFloat(Vector=<25 25 35>, Float=1.5)) {isSetParameters=true}
  ]
  particleUpdateScript [
   &ParticleState()
   &ScaleColor(
    Scale Alpha=&FloatFromCurve(FloatCurve=DataInterface() {propertyValues {Curve {keys [
         <at 0 0 RCIM_Cubic>
         {arriveTangent=1.2987, time=0.1, value=1}
         {arriveTangent=-1.25, interpMode=RCIM_Cubic, leaveTangent=-1.25, tangentMode=RCTM_Break, time=0.3, value=1}
         <at 0.6 1 RCIM_Cubic>
        ]}}})
    Scale RGB=&VectorFromFloat(Value=&FloatFromCurve(FloatCurve=DataInterface() {propertyValues {Curve {keys [
          {arriveTangent=-88.7987, interpMode=RCIM_Cubic, leaveTangent=-232.7645, tangentMode=RCTM_Break, time=0, value=10}
          {interpMode=RCIM_Cubic, tangentMode=RCTM_Break, time=0.07, value=1.3}
          {arriveTangent=-1.3978, interpMode=RCIM_Cubic, leaveTangent=-1.3978, time=0.9, value=1}
          {arriveTangent=-1.1457, interpMode=RCIM_Cubic, leaveTangent=-1.1457, tangentMode=RCTM_User, time=1, value=0}
         ]}}}))
    ScaleA=true
    ScaleRGB=true
   )
   &ScaleMeshSize(Scale Factor=&VectorFromCurve(VectorCurve=DataInterface() {propertyValues {
       XCurve {keys [
         {arriveTangent=2.6567, interpMode=RCIM_Cubic, leaveTangent=2.6567, tangentMode=RCTM_User, time=0, value=0}
         <at 0.02 1>
         {arriveTangentWeight=0.1667, interpMode=RCIM_Cubic, leaveTangentWeight=0.1, tangentMode=RCTM_User, time=0.77, value=1}
         {arriveTangent=-0.4714, tangentMode=RCTM_User, time=1, value=0.95}
        ]}
       YCurve {keys [
         {arriveTangent=2.6567, interpMode=RCIM_Cubic, leaveTangent=2.6567, tangentMode=RCTM_User, time=0, value=0}
         <at 0.02 1>
         {arriveTangentWeight=0.1718, interpMode=RCIM_Cubic, leaveTangentWeight=0.1031, tangentMode=RCTM_User, time=0.77, value=1}
         {arriveTangent=-0.4714, tangentMode=RCTM_User, time=1, value=0.95}
        ]}
       ZCurve {keys [
         {arriveTangent=11.6068, interpMode=RCIM_Cubic, leaveTangent=11.6068, tangentMode=RCTM_User, time=0, value=0.7}
         {arriveTangentWeight=0.0363, interpMode=RCIM_Cubic, leaveTangentWeight=0.2176, tangentMode=RCTM_User, time=0.05, value=1}
         {arriveTangentWeight=0.2049, interpMode=RCIM_Cubic, leaveTangentWeight=0.1024, tangentMode=RCTM_User, time=0.77, value=1}
         {arriveTangent=-0.4714, interpMode=RCIM_Cubic, leaveTangent=-0.4714, tangentMode=RCTM_User, time=1, value=0.95}
        ]}
      }}))
   &DynamicMaterialParameters(
    Index 0 Param 1=&FloatFromCurve(FloatCurve=DataInterface() {propertyValues {Curve {keys [
         {arriveTangent=11.1199, time=0, value=0}
         {arriveTangent=1.7206, tangentMode=RCTM_User, time=0.15, value=1}
         {arriveTangent=-1.6667, leaveTangent=-1.6667, time=0.3, value=1}
         <at 1 0.1>
        ]}}})
    Index 0 Param 2=&FloatFromCurve(FloatCurve=DataInterface() {propertyValues {Curve {keys [
         {arriveTangent=30.9039, time=0, value=0.5}
         {interpMode=RCIM_Cubic, tangentMode=RCTM_Break, time=0.1, value=0.9}
         {arriveTangent=-1.0951, leaveTangent=-3.5455, tangentMode=RCTM_User, time=1, value=0}
        ]}}})
    Index 0 Param 3=&FloatFromCurve(FloatCurve=DataInterface() {propertyValues {Curve {keys [
         {arriveTangent=2.464, interpMode=RCIM_Cubic, leaveTangent=2.464, tangentMode=RCTM_User, time=0, value=-0.1}
         <at 0.1 0>
         {arriveTangentWeight=0.151, interpMode=RCIM_Cubic, leaveTangentWeight=0.1343, tangentMode=RCTM_User, time=0.2, value=-0.02}
         <at 1 -0.18>
        ]}}})
    Index 0 Param 4=&FloatFromCurve(FloatCurve=DataInterface() {propertyValues {Curve {keys [
         {arriveTangent=39.9267, interpMode=RCIM_Cubic, leaveTangent=52.4427, tangentMode=RCTM_Break, time=0, value=0}
         {arriveTangentWeight=0.0737, interpMode=RCIM_Cubic, leaveTangentWeight=0.2743, tangentMode=RCTM_User, time=0.04, value=1}
         {arriveTangentWeight=0.1567, interpMode=RCIM_Cubic, leaveTangentWeight=0.1325, tangentMode=RCTM_User, time=0.3, value=0.95}
         <at 0.6 -0.5>
        ]}}})
    Index 1 Param 1=&FloatFromCurve(FloatCurve=DataInterface() {propertyValues {Curve {keys [{arriveTangent=-23.6397, time=0, value=-0.15}, {interpMode=RCIM_Cubic, tangentMode=RCTM_Break, time=0.15, value=1}, <at 0.9 1>, <at 1 1>]}}})
    Index 1 Param 2=&FloatFromCurve(FloatCurve=DataInterface() {propertyValues {Curve {keys [<at 0.1 1>,<at 0.4 1>,<at 0.65 0>]}}})
    Index 1 Param 3=&FloatFromCurve(FloatCurve=DataInterface() {propertyValues {Curve {keys [<at 0 5>,<at 0.2 0>]}}})
    Module.Index0Param0Write=true
    Module.Index0Param1Write=true
    Module.Index0Param2Write=true
    Module.Index0Param3Write=true
    Module.Index1Param0Write=true
    Module.Index1Param1Write=true
    Module.Index1Param2Write=true
    Module.Index1Param3Write=false
    Write Parameter Index 1=true
   )
  ]
  rendererClasses [&NiagaraMeshRendererProperties]
  renderers [{rendererClass=&NiagaraMeshRendererProperties, rendererIndex=0}]
 }
 Shards {
  emitterUpdateScript [
   &EmitterState()
   &SpawnBurst_Instantaneous(Spawn Count=15, Use Spawn Probability=false, UseLoopCountLimit=false)
   &SpawnBurst_Instantaneous(Spawn Count=7, Spawn Group=1, Use Spawn Probability=false, UseLoopCountLimit=false) {module=SpawnBurst_Instantaneous001}
  ]
  particleSpawnScript [
   &InitializeParticle(
    Color Maximum=<rgba 1.152 8 6.8587 1>
    Color Minimum=<rgba 0.4 0.4 4 1>
    Color Mode=::NewEnumerator2
    Lifetime Max=2.7
    Lifetime Min=1.5
    Lifetime Mode=::NewEnumerator1
    Mesh Renderer Array Visibility Mode=::NewEnumerator3
    Mesh Renderer Info=DataInterface() {propertyValues {MeshRenderer=&self:Shards_2.NiagaraMeshRendererProperties_0}}
    Mesh Scale Mode=::NewEnumerator3
    Mesh Uniform Scale=&MaskFloatBySpawnGroup(Masked Float=&RandomRangeFloat(Minimum=4, Maximum=6), Passthrough Float=11)
    Sprite Size Mode=::NewEnumerator3
    Uniform Sprite Size=&MaskFloatBySpawnGroup(Masked Float=&RandomRangeFloat(Minimum=300, Maximum=600), Passthrough Float=1200)
    UsePositionOffset=false
    Write Material Random=true
   )
   &ShapeLocation(
    MaskSpawnGroup=false
    Rotation Axis=&ReturnCameraVector(Camera Vectors=::NewEnumerator1, Coordinate Space=::Simulation)
    Rotation Mode=::NewEnumerator1
    Sphere Distribution=::NewEnumerator1
    Sphere Radius=600
    U Position=&ReturnNormalizedExecIndex(Normalized Index Includes One=false)
    UseOwnerScale=true
   )
   &AddVelocity(
    Module.Use Falloff Distance=false
    Override Seed=false
    Velocity Mode=::NewEnumerator1
    Velocity Speed=&RandomRangeFloat(Minimum=1500, Maximum=10000)
   )
   &InitialMeshOrientation(Use Rotation Vector=false)
  ]
  particleUpdateScript [
   &ParticleState()
   &ScaleMeshSize(Scale Factor=&VectorFromFloat(Value=&FloatFromCurve(FloatCurve=DataInterface() {propertyValues {Curve {keys [{interpMode=RCIM_Cubic, tangentMode=RCTM_Break, time=0.5, value=1}, <at 1 0>]}}})))
   &ScaleSpriteSize(
    Uniform Curve Index=&Add_Float(
     A=&Multiply_Float(A=$Particles.Age, B=&RandomRangeFloat(Minimum=0.8, Maximum=1.5))
     B=&RandomRangeFloat(Minimum=-0.25, Maximum=0.25)
    )
    Uniform Curve Sprite Scale=DataInterface() {propertyValues {Curve {keys [
        {tangentMode=RCTM_User, time=0.2, value=0}
        <at 0.22 1 RCIM_Cubic>
        {tangentMode=RCTM_User, time=0.5, value=0}
        <at 0.68 0>
        <at 0.7 0.8 RCIM_Cubic>
        <at 0.85 0>
       ]}}}
   )
   &UpdateMeshOrientation(
    Rotation Rate=&FloatFromCurve(
     FloatCurve=DataInterface() {propertyValues {Curve {keys [
         <at 0 1 RCIM_Cubic>
         {arriveTangent=-2.1341, interpMode=RCIM_Cubic, leaveTangent=-2.1341, tangentMode=RCTM_User, time=0.35, value=0.65}
         <at 0.8 0.1 RCIM_Cubic>
        ]}}}
     Scale Curve=&RandomRangeFloat(Minimum=1, Maximum=2)
    )
    Rotation Vector=&RandomVector(Vector Scale=1)
   )
   &Drag(
    Drag=&FloatFromCurve(
     FloatCurve=DataInterface() {propertyValues {Curve {keys [
         {arriveTangent=4.7895, interpMode=RCIM_Cubic, leaveTangent=4.7895, tangentMode=RCTM_User, time=0, value=0.1}
         {arriveTangent=0.6, interpMode=RCIM_Cubic, leaveTangent=0.6, time=0.25, value=1.5}
         {arriveTangent=-4.8446, interpMode=RCIM_Cubic, leaveTangent=-4.8446, tangentMode=RCTM_User, time=0.5, value=0.4}
        ]}}}
     Scale Curve=3
    )
    Use Linear Drag=true
    Use Rotational Drag=false
   )
   &SolveForcesAndVelocity(Clamp Velocity=false, Limit Acceleration=false)
   &ScaleColor(
    Scale Alpha=&FloatFromCurve(FloatCurve=DataInterface() {propertyValues {Curve {keys [<at 0.5 1 RCIM_Cubic>,<at 1 0>]}}})
    Scale RGB=&VectorFromFloat(Value=&FloatFromCurve(FloatCurve=DataInterface() {propertyValues {Curve {keys [{arriveTangent=-13.9762, interpMode=RCIM_Cubic, leaveTangent=-13.9762, tangentMode=RCTM_User, time=0, value=5}, <at 0.5 1>]}}}))
    ScaleA=false
    ScaleRGB=true
   )
  ]
  rendererClasses [&NiagaraMeshRendererProperties, &NiagaraSpriteRendererProperties]
  renderers [
   {rendererClass=&NiagaraMeshRendererProperties, rendererIndex=0}
   {rendererClass=&NiagaraSpriteRendererProperties, rendererIndex=1}
  ]
 }
}
summary {
 systemName=NS_PoolDust_RiftOut
 emitters [
  {emitterName=Shards, rendererClasses [&NiagaraMeshRendererProperties, &NiagaraSpriteRendererProperties]}
  {emitterName=BlueSparks, rendererClasses [&NiagaraSpriteRendererProperties]}
  {emitterName=Flare_COmplex001, rendererClasses [&NiagaraSpriteRendererProperties]}
  {emitterName=Glints, rendererClasses [&NiagaraSpriteRendererProperties]}
  {emitterName=LargeCracks, rendererClasses [&NiagaraMeshRendererProperties]}
 ]
 userVariables [
  User.LocalPostProcessEffect {type=NiagaraBool}
  User.SizeNormalized {type=NiagaraFloat, defaultValue=1, description=""}
 ]
}
