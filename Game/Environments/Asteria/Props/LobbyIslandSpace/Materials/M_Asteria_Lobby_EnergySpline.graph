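# Additive, unlit energy-ribbon material for Asteria lobby island splines, usable on spline meshes and Niagara mesh particles.
# Caustics warp two panning layers of stars and dots, combined with tinted glow and background gradient, scaled by particle color.
# SplineThicken and sine waves drive world position offset, damping waves at the ends.
# Vertex red varies strands, dynamic Random offsets time, and depth fade softens opacity outside forward and mobile paths.
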Material /Game/Environments/Asteria/Props/LobbyIslandSpace/Materials/M_Asteria_Lobby_EnergySpline
graph {
 properties {
  bUsedWithNiagaraMeshParticles=true
  bUsedWithSplineMeshes=true
  blendMode=BLEND_Additive
  pixelDepthOffsetMode=PDOM_Legacy
  shadingModel=MSM_Unlit
  twoSided=true
  usedShadingModels=Unlit
 }
 nodes [
  n1 = SphereMask(
   A=ComponentMask(None=TextureCoordinate()) {g=true}
   B=Constant() {r=0.5}
   Radius=LinearInterpolate(
    A=ScalarParameter() {defaultValue=0.466, parameterName=EmissiveRadiusMax}
    B=ScalarParameter() {defaultValue=0.466, parameterName=EmissiveRadiusMin}
    Alpha=VertexColor().R
   )
   Hardness=ScalarParameter() {defaultValue=0.5601, parameterName=EmissiveFadeHardness}
  ) {attenuationRadius=1, hardnessPercent=50}
  n2 = VectorParameter() {parameterName=Tile1, defaultValue=<rgba 0.4 5 0 0>}
  n3 = Multiply(A=Add(A=VertexColor().R, B=TextureCoordinate()), B=AppendVector(A=@n2.R, B=@n2.R))
  n4 = VectorParameter() {parameterName=Tile2, defaultValue=<rgba 0.5 0.25 0 0>}
  n5 = AppendVector(A=@n4.R, B=@n4.R)
  n6 = VectorParameter() {parameterName=PanDirection2, defaultValue=<rgba 0.04 0.05 0 0>}
  n7 = TextureSample(UVs=Panner(Coordinate=Multiply(A=@n3, B=@n5), Time=NamedRerouteUsage(), Speed=AppendVector(A=@n6.R, B=@n6.R)) {speedX=1, speedY=1}) {texture=/Game/Effects/Fort_Effects/Textures/Tile/T_Caustics_Chromatic}
  n8 = Multiply(A=ComponentMask(None=@n7.RGB) {g=true, r=true}, B=ScalarParameter() {parameterName=WarpAmt}) {constB=0}
  n9 = NamedRerouteUsage()
  n10 = VectorParameter() {parameterName=PanDirection1, defaultValue=<rgba 0.08 1 0 0>}
  n11 = AppendVector(A=@n10.R, B=@n10.R)
  n12 = Panner(Coordinate=Add(A=@n3, B=@n8), Time=@n9, Speed=@n11) {speedX=1, speedY=1}
  n13 = TextureSample(UVs=@n12) {texture=/Game/Environments/Asteria/Props/LobbyIslandSpace/Dots03}
  n14 = VectorParameter() {parameterName=PanDirection1Mult, defaultValue=<rgba -1.7 1.4 0 1>}
  n15 = Panner(Coordinate=Subtract(A=Multiply(A=@n3) {constB=2}, B=@n8), Time=@n9, Speed=Multiply(A=@n11, B=AppendVector(A=@n14.R, B=@n14.R))) {speedX=1, speedY=1}
  n16 = TextureSample(UVs=@n15) {texture=/Game/Environments/Asteria/Props/LobbyIslandSpace/Dots01}
  n17 = TextureSampleParameter2D(UVs=@n12) {parameterName=MainTexture, texture=/Game/Athena/Environments/Nexus/Textures/T_ZeroPoint_Line_Stars}
  n18 = TextureSampleParameter2D(UVs=@n15) {parameterName=MainTexture, texture=&T_ZeroPoint_Line_Stars}
  n19 = VectorParameter() {parameterName=SecondaryTint, defaultValue=<rgba 0.5521 0.1042 1 1>}
  n20 = Add(
   A=Multiply(A=Add(A=@n13.G, B=@n16.G), B=ScalarParameter() {defaultValue=100, parameterName=DotBrightness})
   B=Add(A=@n17.RGB, B=Multiply(A=@n18.RGB, B=@n19.RGB) {constB=0.5})
  )
  n21 = VectorParameter() {parameterName=WaveGlow, defaultValue=<rgba 0 3.0248 5 0>}
  n22 = Multiply(A=@n21.RGB) {constB=3}
  n23 = NamedRerouteUsage()
  n24 = Multiply(A=@n7.RGB, B=Add(A=LinearInterpolate(A=@n22, B=@n21.RGB, Alpha=@n23) {constB=0}) {constB=0.5}) {constB=0.5}
  n25 = LinearInterpolate(
   A=VectorParameter() {parameterName=BackgroundColor1, defaultValue=<rgba 0.112 0.4078 0.5 0>}.RGB
   B=VectorParameter() {parameterName=BackgroundColor2, defaultValue=<rgba 0.0457 0.0378 0.12 0>}.RGB
   Alpha=NamedRerouteUsage()
  )
  n26 = ComponentMask(None=TextureCoordinate()) {r=true}
  n27 = Constant() {r=0.5}
  n28 = Multiply(A=Constant() {r=1}, B=Multiply(A=SphereMask(A=@n26, B=@n27, Radius=@n27) {hardnessPercent=0}))
  n29 = MaterialFunctionCall(In=@n28, Contrast=Constant() {r=0.25}) {materialFunction=&CheapContrast}
  n30 = Multiply(
   A=LinearInterpolate(
    A=VectorParameter() {parameterName=EmissiveMultHigh, defaultValue=<rgba 0.9603 1.3011 1.5 0>}.RGB
    B=VectorParameter() {parameterName=EmissiveMultLow, defaultValue=<rgba 0.9603 1.3011 1.5 0>}.RGB
    Alpha=OneMinus(None=@n1)
   )
   B=Add(A=Add(A=@n20, B=@n24), B=Multiply(A=@n25, B=@n29.Result))
  )
  n31 = ParticleColor()
  MaterialFunctionCall(Emissive=Multiply(A=Multiply(A=@n30, B=@n31.RGB)) {constB=0.2}) {materialFunction=&MF_FX_EmissiveAdjustments}
  n32 = Add(A=@n18.B, B=@n7.B)
  n33 = Multiply(
   A=Add(A=Add(A=@n17.B, B=@n32) {desc=Opacity}, B=Multiply(A=@n29.Result) {constB=0.4, desc=Add opacity for background color}) {constB=0.5}
   B=@n28
  )
  Multiply(A=@n33, B=NamedRerouteUsage())
  n34 = NamedRerouteDeclaration(None=Add(A=DynamicParameter() {defaultValue=<rgba 1 1 0 1>, paramNames [Random, SplineWidthMult, ZipMult, WaveScale]}.Random, B=Time())) {name=Time+Random, nodeColor=<rgba 1 0 0.4922 1>}
  n35 = ScalarParameter() {defaultValue=0.06, parameterName=PanSpeed}
  NamedRerouteDeclaration(None=Multiply(A=VertexColor().R, B=Multiply(A=@n34, B=@n35))) {name=PanTime, nodeColor=<rgba 1 0.0511 0 1>}
  NamedRerouteUsage()
  n36 = ScalarParameter() {defaultValue=0.5, parameterName=SplineEndMinThickness}
  n37 = NamedRerouteUsage()
  n38 = NamedRerouteDeclaration(None=Multiply(A=LinearInterpolate(A=@n36, Alpha=@n37), B=ScalarParameter() {defaultValue=250, parameterName=SplineWidth})) {name=SplineWidth, nodeColor=<rgba 0.6172 1 0 1>}
  n39 = MaterialFunctionCall() {materialFunction=&TangentBasis}
  n40 = Add(A=ComponentMask(None=TextureCoordinate()) {g=true}, B=VertexColor().R)
  n41 = LinearInterpolate(
   A=ScalarParameter() {defaultValue=1, parameterName=WaveLengthMin}
   B=ScalarParameter() {defaultValue=1, parameterName=WaveLengthMax}
   Alpha=VertexColor().R
  )
  n42 = ScalarParameter() {defaultValue=1, parameterName=LastIndex}
  n43 = LinearInterpolate(Alpha=VertexColor().R) {constA=0.87, constB=1.125}
  n44 = ScalarParameter() {defaultValue=-0.02, parameterName=WaveSpeed}
  n45 = Add(A=Divide(A=OneMinus(None=@n40), B=Divide(A=@n41, B=@n42)), B=Multiply(A=NamedRerouteUsage(), B=Multiply(A=@n43, B=@n44)))
  n46 = Add(A=VertexColor().R, B=@n45)
  n47 = NamedRerouteDeclaration(None=ConstantBiasScale(None=Sine(None=@n46) {period=0.7})) {name=Offsets, nodeColor=<rgba 0.4531 0 1 1>}
  n48 = Multiply(
   A=Multiply(A=LinearInterpolate(A=@n39.X U, B=@n39.X U, Alpha=@n47), B=Sine(None=@n45) {desc=Smaller waves for WPO})
   B=ScalarParameter() {defaultValue=350, parameterName=WaveScale}
  )
  n49 = TextureCoordinate()
  n50 = Constant() {r=0.5}
  n51 = NamedRerouteDeclaration(None=SphereMask(A=ComponentMask(None=@n49) {g=true}, B=@n50, Radius=@n50, Hardness=ScalarParameter() {defaultValue=0.64, parameterName=EndFadeHardness}) {attenuationRadius=1, hardnessPercent=50}) {name=EndFade, nodeColor=<rgba 0 1 0.3672 1>}
  Reroute(None=Add(
    A=MaterialFunctionCall(WidthBase=@n38, WidthTip=@n38) {materialFunction=&SplineThicken}.WorldPositionOffset
    B=Multiply(A=@n48, B=@n51) {desc=Reduce WPO at spline ends}
   ))
  n52 = Multiply(A=Multiply(A=@n31.A, B=@n33), B=ScalarParameter() {defaultValue=1, parameterName=OpacityMult})
  n53 = DepthFade(Opacity=@n52, FadeDistance=ScalarParameter() {defaultValue=100000, parameterName=DepthFade})
  ShadingPathSwitch(Default=@n53, Deferred=@n53, Forward=@n52, Mobile=@n52)
 ]
}
