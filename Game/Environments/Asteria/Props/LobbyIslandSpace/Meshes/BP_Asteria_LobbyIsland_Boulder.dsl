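; Boulder prop Blueprint for the Asteria lobby island that dresses a main rock with scattered smaller rocks.
; The construction script applies the chosen boulder mesh, then loops over a clamped rock count to add small static mesh components.
; A seeded random stream gives each rock an offset within a cubic scatter range and a random rotation, with uniform configurable scale.
; Collision is disabled on all these meshes, and the event graph holds only empty events.
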(event EventBeginPlay)

(event Collision|EventActorBeginOverlap
  (params (OtherActor :type Actor%20Object%20Reference)))

(event EventTick
  (params (DeltaSeconds :type Float%20%28single-precision%29)))

(fn ConstructionScript
  (Components|StaticMesh|SetStaticMesh (Variables|Default|GetBoulder) (Variables|Default|GetBoulderMesh))
  (Collision|SetCollisionEnabled (Variables|Default|GetBoulder) "NoCollision")
  (for _ (range 1 (+ (Math|Integer|Clamp(Integer) (Variables|Default|GetSmallRockNum) 1 10) 1))
    (bind _returnvalue (AddComponent|NODE_AddStaticMeshComponent-1 :RelativeTransform (Math|Transform|MakeTransform (Math|Vector|MakeVector (Math|Random|RandomFloatinRangefromStream (Variables|Default|GetRandomSeed) (* (Variables|Default|GetScatterRadius) -1.0) (Variables|Default|GetScatterRadius)) (Math|Random|RandomFloatinRangefromStream (Variables|Default|GetRandomSeed) (* (Variables|Default|GetScatterRadius) -1.0) (Variables|Default|GetScatterRadius)) (Math|Random|RandomFloatinRangefromStream (Variables|Default|GetRandomSeed) (* (Variables|Default|GetScatterRadius) -1.0) (Variables|Default|GetScatterRadius))) (Math|Rotator|MakeRotator (Math|Random|RandomFloatinRangefromStream (Variables|Default|GetRandomSeed) 0.0 360) (Math|Random|RandomFloatinRangefromStream (Variables|Default|GetRandomSeed) 0.0 360) (Math|Random|RandomFloatinRangefromStream (Variables|Default|GetRandomSeed) 0.0 360)) (Math|Conversions|ToVector(Float) (Variables|Default|GetSmallRockSize)))))
    (Collision|SetCollisionEnabled _returnvalue "NoCollision")))
