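; Blueprint actor for a floating island prop in the Asteria lobby island space environment.
; Its construction script assigns a configurable mesh to the FloatingIsland component and disables that component's collision.
; It creates a dynamic instance of the broken island edge material on slot 0, then drives its cliff tint, edge tint and normal parameters from variables.
; BeginPlay, overlap and Tick events are empty, so all setup happens at construction.
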(event EventBeginPlay)

(event Collision|EventActorBeginOverlap
  (params (OtherActor :type Actor%20Object%20Reference)))

(event EventTick
  (params (DeltaSeconds :type Float%20%28single-precision%29)))

(fn ConstructionScript
  (bind _floatingisland (Variables|Default|GetFloatingIsland))
  (bind _floatingislandmid (Variables|Default|GetFloatingIslandMID))
  (Components|StaticMesh|SetStaticMesh _floatingisland (Variables|Default|GetFloatingIslandMesh))
  (Collision|SetCollisionEnabled (Variables|Default|GetFloatingIsland) "NoCollision")
  (bind _returnvalue (Rendering|Material|CreateDynamicMaterialInstance _floatingisland 0 "/Game/Environments/Asteria/Props/LobbyIslandSpace/Materials/M_BrokenIslandEdge_Floating.M_BrokenIslandEdge_Floating" "None" :!class /Script/Engine.PrimitiveComponent))
  (bind _output_get (Variables|Default|SetFloatingIslandMID _returnvalue))
  (Rendering|Material|SetMaterial _floatingisland 0 _output_get)
  (Rendering|Material|SetVectorParameterValue _floatingislandmid "TintCliff" (Math|Conversions|ToLinearColor(Color) (Variables|Default|GetTintCliff)) :!class /Script/Engine.MaterialInstanceDynamic)
  (Rendering|Material|SetVectorParameterValue _floatingislandmid "TintCliffEdge" (Math|Conversions|ToLinearColor(Color) (Variables|Default|GetTintCliffEdge)) :!class /Script/Engine.MaterialInstanceDynamic)
  (Rendering|Material|SetScalarParameterValue _floatingislandmid "CliffNormal" (Variables|Default|GetCliffNormal) :!class /Script/Engine.MaterialInstanceDynamic))
