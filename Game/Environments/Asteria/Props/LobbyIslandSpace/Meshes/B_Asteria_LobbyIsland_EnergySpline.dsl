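; Blueprint for an energy spline prop on the Asteria lobby island, building a deformed mesh along an editable spline.
; The construction script adds one spline mesh component per segment between consecutive spline points and sets its forward axis to X.
; Each component is stored in an array and fitted to the local locations and tangents of its two points.
; BeginPlay, overlap and Tick events are empty, so it has no runtime logic.
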(event EventBeginPlay)

(event Collision|EventActorBeginOverlap
  (params (OtherActor :type Actor%20Object%20Reference)))

(event EventTick
  (params (DeltaSeconds :type Float%20%28single-precision%29)))

(fn ConstructionScript
  (for _index (range (+ (- (Spline|GetNumberOfSplinePoints (Variables|Default|GetSpline)) 2) 1))
    (bind _returnvalue (AddComponent|NODE_AddSplineMeshComponent-0))
    (SplineMesh|SetForwardAxis _returnvalue "X" true)
    (Utilities|Array|AddUnique (Variables|Default|GetSplineMeshes) _returnvalue)
    (SplineMesh|SetStartandEnd _returnvalue (Spline|GetLocationandTangentatSplinePoint (Variables|Default|GetSpline) _index "Local") (.Tangent (Spline|GetLocationandTangentatSplinePoint (Variables|Default|GetSpline) _index "Local")) (Spline|GetLocationandTangentatSplinePoint (Variables|Default|GetSpline) (+ _index 1) "Local") (.Tangent (Spline|GetLocationandTangentatSplinePoint (Variables|Default|GetSpline) (+ _index 1) "Local")) true :!class /Script/Engine.SplineMeshComponent)))
